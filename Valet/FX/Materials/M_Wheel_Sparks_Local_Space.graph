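# Unlit additive, two-sided material for wheel spark effects on Niagara mesh particles.
# Two offset samples of a noise texture pan along U at dynamic-parameter SparksSpeed, are shaped by UV edge falloffs and sharpened by Power.
# Time-of-day color, particle color and depth fade tint and fade the result into an emissive adjustment function.
# The UV falloff, without noise, also scales a vertex-normal vector by the Distortion parameter.
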Material /Valet/FX/Materials/M_Wheel_Sparks_Local_Space
graph {
 properties {
  bUsedWithNiagaraMeshParticles=true
  blendMode=BLEND_Additive
  pixelDepthOffsetMode=PDOM_Legacy
  shadingModel=MSM_Unlit
  twoSided=true
  usedShadingModels=Unlit
 }
 nodes [
  n1 = ParticleColor()
  n2 = ComponentMask(None=TextureCoordinate()) {r=true}
  n3 = Multiply(A=OneMinus(None=@n2)) {constB=2.25}
  n4 = ComponentMask(None=TextureCoordinate()) {g=true}
  n5 = OneMinus(None=@n4)
  n6 = Clamp(None=Multiply(A=Multiply(A=@n5, B=@n4)) {constB=4.5})
  n7 = Multiply(A=Clamp(None=Multiply(A=@n3, B=@n2)), B=@n6)
  n8 = ScalarParameter() {defaultValue=0.065, parameterName=UV X}
  n9 = ScalarParameter() {defaultValue=1, parameterName=UV Y}
  n10 = DynamicParameter() {defaultValue=<rgba 0.25 1 1 1>, paramNames [SparksSpeed,Distortion,Param3,Param4]}
  n11 = Panner(
   Coordinate=Multiply(
    A=TextureCoordinate()
    B=MaterialFunctionCall(R=@n8, G=@n9) {materialFunction=&AppendMany}.RG
   )
   Time=Time()
   Speed=AppendVector(A=Multiply(A=@n10.SparksSpeed) {constB=-1}, B=Constant())
  ) {bFractionalPart=true}
  n12 = Constant2Vector() {g=0.325, r=0.01}
  n13 = Add(
   A=TextureSample(UVs=@n11) {samplerSource=SSM_Wrap_WorldGroupSettings, texture=/Game/Global/Textures/Default/Noise/Noise}.G
   B=TextureSample(UVs=Add(A=@n11, B=@n12)) {samplerSource=SSM_Wrap_WorldGroupSettings, texture=&Noise}.G
  )
  n14 = ScalarParameter() {defaultValue=6.5, parameterName=Power}
  n15 = Multiply(
   A=Power(Base=Multiply(A=@n13, B=@n7), Exp=@n14) {constExponent=4}
   B=MaterialFunctionCall(
    Color_Morning=VectorParameter() {parameterName=MorningColor, defaultValue=<rgba 1 1 1 0>}.RGB
    Color_Day=VectorParameter() {parameterName=DayColor, defaultValue=<rgba 1 1 1 0>}.RGB
    Color_Evening=VectorParameter() {parameterName=EveningColor, defaultValue=<rgba 1 1 1 0>}.RGB
    Color_Night=VectorParameter() {parameterName=NightColor, defaultValue=<rgba 1 1 1 0>}.RGB
   ) {materialFunction=&Color_Time_of_Day}.Result
  )
  n16 = DepthFade(FadeDistance=ScalarParameter() {defaultValue=12, parameterName=Depth Fade})
  MaterialFunctionCall(Emissive=Multiply(A=Multiply(A=@n1.RGB, B=@n15), B=Multiply(A=@n1.A, B=@n16))) {materialFunction=&MF_FX_EmissiveAdjustments}
  n17 = Multiply(A=Multiply(A=@n6, B=Power(Base=@n7, Exp=@n14))) {constB=25}
  Multiply(
   A=Multiply(A=VertexNormalWS(), B=Clamp(None=@n17))
   B=DynamicParameter() {defaultValue=<rgba 0.25 1 1 1>, paramNames [SparksSpeed,Distortion,Param3,Param4]}.Distortion
  )
  Multiply()
 ]
}
