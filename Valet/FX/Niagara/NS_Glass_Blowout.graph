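# Niagara burst effect that blows glass shards out of a window, sized by User.windowSize.
# Large and small sprite bits spawn in that box and fly in an upward-tilted cone along owner rotation, falling and fading.
# GPU bits collide, gated by random and age checks; GlassBits_Large_LOW is a non-GPU copy without collision.
# LW_ emitters for mist and burst lines are enabled, replacing disabled standard versions; LW_GlassBits_Small is disabled.
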NiagaraSystem /Valet/FX/Niagara/NS_Glass_Blowout
dependencies=null
emitters {
 GlassBits_Large {
  simTarget=GPUComputeSim
  emitterUpdateScript [
   &EmitterState(Life Cycle Mode=::NewEnumerator1, Loop Behavior=::NewEnumerator1, Loop Duration=1, UseLoopDelay=false)
   &SpawnBurst_Instantaneous(Spawn Count=20, Use Spawn Probability=false)
  ]
  particleSpawnScript [
   /Niagara/Modules/Spawn/Initialization/InitializeParticle(
    Color=<rgba 0.5208 0.9759 1 1.25>
    Lifetime=&UniformRangedFloat(Minimum=2, Maximum=2.5)
    Mass=&UniformRangedFloat(Minimum=0.1, Maximum=0.05)
    Module.Write Color=true
    Module.Write Lifetime=true
    Module.Write Mass=true
    Module.Write Material Random=true
    Module.Write Position=false
    Module.Write Scale=false
    Module.Write SpriteRotation=true
    Module.Write SpriteSize=true
    Position=&SimulationPosition()
    Sprite Rotation=&UniformRangedFloat(Minimum=-360, Maximum=360)
    Sprite Size=&UniformRangedVector2D(Minimum=<7 7>, Maximum=<15 15>)
   )
   &BoxLocation(
    Box Size=$User.windowSize
    MaskSpawnGroup=false
    Module.Override Seed=false
    Rotate Box=false
    Rotate Box On Axis:Vector3f=<0 1 0>
    Surface Only=false
   )
   &AddVelocityInCone(
    Cone Angle=32
    Cone Axis=&RotateVectorByQuaternion(Vector To Rotate=<1 0 0.325>, Quaternion=$Engine.Owner.Rotation)
    Cone Axis Coordinate Space=::Simulation
    Module.Use Velocity Falloff On Cone Axis=true
    Velocity Falloff Away From Cone Axis=0.333
    Velocity Strength=&UniformRangedFloat(Minimum=300, Maximum=900)
   )
   &AddRotationalVelocity(Rotation Rate=&UniformRangedVector(Minimum=<-150 -150 -150>, Maximum=<150 150 150>))
   &ApplyInitialForces()
  ]
  particleUpdateScript [
   &ParticleState()
   &GravityForce()
   &Drag(Drag=&UniformAOrBFloat(A=0.2), Use Linear Drag=true, Use Rotational Drag=false)
   &Collision(
    Advanced Aging Rate=1
    Age Colliding Particles=false
    Collision Enabled=&Boolean_AndOperation(
     Bool1=&SetBoolByFloatComparison(A=&UniformRangedFloat(Minimum=0, Maximum=1), B=0.65)
     Bool2=&SetBoolByFloatComparison(A=$Particles.Age, B=0.1)
    )
    Friction=0.05
    Kill Offscreen Particles=true
    Module.Control Rest State Via Surface Normal=false
    Module.EnableMaxCollisionCount=false
    Module.Randomize Collision Normal Vector=true
    Particle Radius Scale=0.125
    Restitution=0.5
   )
   &ScaleColor(
    Scale Alpha=&FloatFromCurve(FloatCurve=DataInterface() {propertyValues {Curve {keys [<at -0.0046 1.0113 RCIM_Cubic>,<at 1.0011 0.6898 RCIM_Cubic>]}}})
    ScaleA=true
    ScaleRGB=true
   )
   &SpriteRotationRate(Rotation Rate=&Multiply_Float(A=$Particles.Initial.MaterialRandom, B=1250))
   &SolveForcesAndVelocity(Clamp Velocity=false, Limit Acceleration=false)
   &SolveRotationalForcesAndVelocity(Rotational Inertia:NiagaraFloat=$Particles.RotationalInertia)
   &ScaleSpriteSize(Scale Factor=&Vector2DFromCurve(
     Vector2Curve=DataInterface() {propertyValues {XCurve {keys [<at 0.227 1.0067>,<at 0.9989 0>]}, YCurve {keys [<at 0.227 1.0067>,<at 0.9989 0>]}}}
     ScaleCurve=<1 1>
    ))
  ]
  rendererClasses [&NiagaraSpriteRendererProperties]
  renderers [{rendererClass=&NiagaraSpriteRendererProperties}]
 }
 GlassBits_Large_LOW {
  emitterUpdateScript [
   &EmitterState(Life Cycle Mode=::NewEnumerator1, Loop Behavior=::NewEnumerator1, Loop Duration=1, UseLoopDelay=false)
   &SpawnBurst_Instantaneous(Spawn Count=20, Use Spawn Probability=false)
  ]
  particleSpawnScript [
   &InitializeParticle(
    Color=<rgba 0.5208 0.9759 1 1.25>
    Lifetime=&UniformRangedFloat(Minimum=2, Maximum=2.5)
    Mass=&UniformRangedFloat(Minimum=0.1, Maximum=0.05)
    Module.Write Color=true
    Module.Write Lifetime=true
    Module.Write Mass=true
    Module.Write Material Random=true
    Module.Write Position=false
    Module.Write Scale=false
    Module.Write SpriteRotation=true
    Module.Write SpriteSize=true
    Position=&SimulationPosition()
    Sprite Rotation=&UniformRangedFloat(Minimum=-360, Maximum=360)
    Sprite Size=&UniformRangedVector2D(Minimum=<7 7>, Maximum=<15 15>)
   )
   &BoxLocation(
    Box Size=$User.windowSize
    MaskSpawnGroup=false
    Module.Override Seed=false
    Rotate Box=false
    Rotate Box On Axis:Vector3f=<0 1 0>
    Surface Only=false
   )
   &AddVelocityInCone(
    Cone Angle=32
    Cone Axis=&RotateVectorByQuaternion(Vector To Rotate=<1 0 0.325>, Quaternion=$Engine.Owner.Rotation)
    Cone Axis Coordinate Space=::Simulation
    Module.Use Velocity Falloff On Cone Axis=true
    Velocity Falloff Away From Cone Axis=0.333
    Velocity Strength=&UniformRangedFloat(Minimum=300, Maximum=900)
   )
   &AddRotationalVelocity(Rotation Rate=&UniformRangedVector(Minimum=<-150 -150 -150>, Maximum=<150 150 150>))
   &ApplyInitialForces()
  ]
  particleUpdateScript [
   &ParticleState()
   &GravityForce()
   &Drag(Drag=&UniformAOrBFloat(A=0.2), Use Linear Drag=true, Use Rotational Drag=false)
   &ScaleColor(
    Scale Alpha=&FloatFromCurve(FloatCurve=DataInterface() {propertyValues {Curve {keys [<at -0.0046 1.0113 RCIM_Cubic>,<at 1.0011 0.6898 RCIM_Cubic>]}}})
    ScaleA=true
    ScaleRGB=true
   )
   &SpriteRotationRate(Rotation Rate=&Multiply_Float(A=$Particles.Initial.MaterialRandom, B=1250))
   &SolveForcesAndVelocity(Clamp Velocity=false, Limit Acceleration=false)
   &SolveRotationalForcesAndVelocity(Rotational Inertia:NiagaraFloat=$Particles.RotationalInertia)
   &ScaleSpriteSize(Scale Factor=&Vector2DFromCurve(
     Vector2Curve=DataInterface() {propertyValues {XCurve {keys [<at 0.227 1.0067>,<at 0.9989 0>]}, YCurve {keys [<at 0.227 1.0067>,<at 0.9989 0>]}}}
     ScaleCurve=<1 1>
    ))
  ]
  rendererClasses [&NiagaraSpriteRendererProperties]
  renderers [{rendererClass=&NiagaraSpriteRendererProperties}]
 }
 GlassBits_Small {
  simTarget=GPUComputeSim
  emitterUpdateScript [
   &EmitterState(Life Cycle Mode=::NewEnumerator1, Loop Behavior=::NewEnumerator1, Loop Duration=1, UseLoopDelay=false)
   &SpawnBurst_Instantaneous(Spawn Count=45, Use Spawn Probability=false)
  ]
  particleSpawnScript [
   &InitializeParticle(
    Color=<rgba 0.85 0.9925 1 1.2>
    Lifetime=&UniformRangedFloat(Minimum=0.8, Maximum=1.75)
    Mass=&UniformRangedFloat(Minimum=0.75, Maximum=1.75)
    Module.Write Color=true
    Module.Write Lifetime=true
    Module.Write Mass=true
    Module.Write Material Random=true
    Module.Write Position=false
    Module.Write Scale=false
    Module.Write SpriteRotation=true
    Module.Write SpriteSize=true
    Position=&SimulationPosition()
    Sprite Rotation=&UniformRangedFloat(Minimum=-360, Maximum=360)
    Sprite Size=&UniformRangedVector2D(Minimum=<3 3>, Maximum=<10 10>)
   )
   &AddVelocityInCone(
    Cone Angle=32
    Cone Axis=&RotateVectorByQuaternion(Vector To Rotate=<1 0 0.325>, Quaternion=$Engine.Owner.Rotation)
    Cone Axis Coordinate Space=::Simulation
    Module.Use Velocity Falloff On Cone Axis=true
    Velocity Falloff Away From Cone Axis=0.333
    Velocity Strength=&UniformRangedFloat(Minimum=300, Maximum=750)
   )
   &BoxLocation(
    Box Size=$User.windowSize
    MaskSpawnGroup=false
    Module.Override Seed=false
    Rotate Box=false
    Rotate Box On Axis:Vector3f=<0 1 0>
    Surface Only=false
   )
  ]
  particleUpdateScript [
   &ParticleState()
   &GravityForce()
   &Drag(Use Linear Drag=true, Use Rotational Drag=false)
   &Collision(
    Advanced Aging Rate=1
    Age Colliding Particles=false
    Collision Enabled=&Boolean_AndOperation(
     Bool1=&SetBoolByFloatComparison(A=&UniformRangedFloat(Minimum=0, Maximum=1), B=0.5)
     Bool2=&SetBoolByFloatComparison(A=$Particles.Age, B=0.1)
    )
    Friction=0.05
    Kill Offscreen Particles=true
    Module.Control Rest State Via Surface Normal=false
    Module.EnableMaxCollisionCount=false
    Module.Randomize Collision Normal Vector=true
    Particle Radius Scale=0.125
    Restitution=0.5
   )
   &ScaleColor(
    Scale Alpha=&FloatFromCurve(FloatCurve=DataInterface() {propertyValues {Curve {keys [<at -0.0046 1.0113 RCIM_Cubic>,<at 1.0011 0.3938 RCIM_Cubic>]}}})
    ScaleA=true
    ScaleRGB=true
   )
   &SolveForcesAndVelocity(Clamp Velocity=false, Limit Acceleration=false)
  ]
  rendererClasses [&NiagaraSpriteRendererProperties]
  renderers [{rendererClass=&NiagaraSpriteRendererProperties}]
 }
 LW_DirectionalBurstLines {rendererClasses [], renderers []}
 LW_GlassBits_Small {bEnabled=false, rendererClasses [], renderers []}
 LW_Glass_Mist {rendererClasses [], renderers []}
}
summary {
 systemName=NS_Glass_Blowout
 emitters [
  {emitterName=GlassBits_Large, simTarget=GPUComputeSim, rendererClasses [&NiagaraSpriteRendererProperties]}
  {emitterName=GlassBits_Small, simTarget=GPUComputeSim, rendererClasses [&NiagaraSpriteRendererProperties]}
  {bEnabled=false, emitterName=Glass_Mist, rendererClasses [&NiagaraSpriteRendererProperties]}
  {bEnabled=false, emitterName=DirectionalBurstLines, rendererClasses [&NiagaraSpriteRendererProperties]}
  {emitterName=GlassBits_Large_LOW, rendererClasses [&NiagaraSpriteRendererProperties]}
  {emitterName=LW_DirectionalBurstLines, rendererClasses []}
  {emitterName=LW_Glass_Mist, rendererClasses []}
  {bEnabled=false, emitterName=LW_GlassBits_Small, rendererClasses []}
 ]
 userVariables [User.windowSize {type=Vector3f, description="", defaultValue {struct=Vector3f, value=<0 0 0>}}]
}
