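# One-shot Niagara scrape impact effect that throws directional sparks and mesh debris with a brief warm light flash, timed by a shared System delay.
# GPU sparks burst from a squashed half-ring with gravity, curl noise and collision; CPU debris tumbles and shrinks.
# A randomly gated lead spark feeds position and direction through a Particle Attribute Reader to a secondary trailing spark emitter.
# Paint and several smoke and spark variants are disabled.
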NiagaraSystem /Valet/FX/Niagara/NS_Scrape_Impact
dependencies=null
emitters {
 DirBurstSpark_EPIC {
  simTarget=GPUComputeSim
  emitterUpdateScript [
   &EmitterState(Life Cycle Mode=::NewEnumerator1, Loop Behavior=::NewEnumerator1, Loop Delay=$System.delay, Loop Duration=1, UseLoopDelay=true)
   &SpawnBurst_Instantaneous(Spawn Count=&UniformRangedInt(Minimum=50, Maximum=120), Use Spawn Probability=false)
   &SpawnBurst_Instantaneous(Spawn Count=15, Spawn Group=1, Use Spawn Probability=false) {module=SpawnBurst_Instantaneous001}
  ]
  particleSpawnScript [
   /Niagara/Modules/Spawn/Initialization/InitializeParticle(
    Color=&SelectLinearColorFromArray(
     Array Sampling Mode=::NewEnumerator0
     Direct Array Index=?
     Color Selection Array=DataInterface() {propertyValues {ColorData [<rgba 350 80 0 1>, <rgba 35 15 0 1>]}}
     Random Seed=0
    )
    Lifetime=&MaskFloatBySpawnGroup(
     Masked Float=&UniformRangedFloat(Minimum=0.35, Maximum=1.75)
     Passthrough Float=&UniformRangedFloat(Minimum=1.5, Maximum=2)
    )
    Mass=&UniformRangedFloat(Minimum=25, Maximum=150)
    Module.Write Color=true
    Module.Write Lifetime=true
    Module.Write Mass=true
    Module.Write Material Random=true
    Module.Write Position=false
    Module.Write Scale=false
    Module.Write SpriteRotation=false
    Module.Write SpriteSize=true
    Position=&SimulationPosition()
    Sprite Size=&Vector2DFromFloat(Value=&UniformRangedFloat(Minimum=5, Maximum=10))
   ) {enabled=false}
   /Niagara/Modules/Spawn/Initialization/V2/InitializeParticle(
    Color=&SelectLinearColorFromArray(
     Array Sampling Mode=::NewEnumerator0
     Direct Array Index=?
     Color Selection Array=DataInterface() {propertyValues {ColorData [<rgba 350 80 0 1>, <rgba 35 15 0 1>]}}
     Random Seed=0
    )
    Lifetime=&MaskFloatBySpawnGroup(
     Masked Float=&UniformRangedFloat(Minimum=0.35, Maximum=1.75)
     Passthrough Float=&UniformRangedFloat(Minimum=1.5, Maximum=2)
    )
    Mass=&UniformRangedFloat(Minimum=25, Maximum=150)
    Mass Mode=::NewEnumerator1
    Sprite Size=&Vector2DFromFloat(Value=&UniformRangedFloat(Minimum=5, Maximum=10))
    Sprite Size Mode=::NewEnumerator4
    UsePositionOffset=false
    Write Material Random=true
   ) {module=InitializeParticle001}
   /Niagara/Modules/Update/Mass/CalculateSizeAndRotationalInertiaByMass(Depth=0, Height=2, Mass Modulation=3, Module.Manually Enter Density=false, Width=0.255)
   &AddVelocityInCone(
    Cone Angle=90
    Cone Axis=<1 0 0.5>
    Module.Use Velocity Falloff On Cone Axis=true
    Velocity Falloff Away From Cone Axis=0.333
    Velocity Strength=&UniformRangedFloat(Minimum=350, Maximum=1800)
   ) {enabled=false}
   &CylinderLocation(
    Cylinder Height=0
    Cylinder Radius=60
    Hemisphere X=true
    MaskSpawnGroup=false
    Module.Enable Lathe Profile=false
    Module.Override Randomness=false
    Module.Surface Only=true
    Non Uniform Scale:Vector3f=<1 0.35 1>
   )
   &AddVelocityFromPoint(
    Module.Use Falloff Distance=false
    Origin Offset=&UniformRangedVector(Minimum=<0 0 -4>, Maximum=<1 1 -25>)
    Origin Offset Coordinate Space=::Simulation
    Velocity Strength=&UniformRangedFloat(Minimum=250, Maximum=2000)
   )
  ]
  particleUpdateScript [
   &ParticleState()
   &GravityForce(Gravity=&UniformRangedVector(Minimum=<0 0 -1600>, Maximum=<1 1 -2580>))
   &Drag(Drag=&UniformRangedFloat(Minimum=0.5, Maximum=1), Use Linear Drag=true, Use Rotational Drag=false) {enabled=false}
   &ScaleColor(
    Scale Alpha=&FloatFromCurve(FloatCurve=DataInterface() {propertyValues {Curve {keys [<at -0.0046 1.0113 RCIM_Cubic>,<at 1.0011 0.002 RCIM_Cubic>]}}}, Scale Curve=0.4)
    ScaleA=true
    ScaleRGB=true
   )
   &Collision(
    Advanced Aging Rate=1
    Age Colliding Particles=false
    Collision Enabled=/Niagara/DynamicInputs/Helpers/MaskBoolBySpawnGroup(
     Masked Bool:NiagaraBool=false
     Passthrough Bool:NiagaraBool=&SetBoolByFloatComparison(A=$Particles.Age, B=0.05)
     Spawn Group Mask:NiagaraInt32=0
    )
    Kill Offscreen Particles=true
    Module.Control Rest State Via Surface Normal=false
    Module.EnableMaxCollisionCount=false
    Module.Randomize Collision Normal Vector=true
   )
   /Niagara/Modules/Update/Forces/CurlNoiseForce(Noise Frequency=0.1, Noise Quality / Cost=::NewEnumerator0, Noise Strength=50000, Pan Noise=true, Pan Noise Field=<25 0 0>)
   &SolveForcesAndVelocity(Clamp Velocity=false, Limit Acceleration=false)
   &ScaleSpriteSizeBySpeed(
    Max Scale Factor=<0.7 0.5>
    Min Scale Factor=<0.4 0.1>
    Sample Scale Factor By Curve=true
    Scale Factor Curve=DataInterface() {propertyValues {Curve {keys [
        {arriveTangent=2.865, interpMode=RCIM_Cubic, time=0.1247, value=-0.0006}
        {arriveTangent=2.5675, interpMode=RCIM_Cubic, leaveTangent=2.5675, tangentMode=RCTM_User, time=0.3948, value=0.4893}
        <at 0.9958 0.9974 RCIM_Cubic>
       ]}}}
    Velocity Threshold=1500
   ) {module=SpriteSizeScaleByVelocity}
   &ScaleSpriteSize(Scale Factor=&Vector2DFromFloat(Value=&FloatFromCurve(FloatCurve=DataInterface() {propertyValues {Curve {keys [
          {interpMode=RCIM_Cubic, tangentMode=RCTM_Break, time=0.7172, value=1}
          {arriveTangent=-9.028, interpMode=RCIM_Cubic, leaveTangent=-9.028, tangentMode=RCTM_User, time=1, value=0}
         ]}}})))
  ]
  rendererClasses [&NiagaraSpriteRendererProperties]
  renderers [{rendererClass=&NiagaraSpriteRendererProperties}]
 }
 DirectionalBurstMeshes_01 {
  emitterUpdateScript [
   &EmitterState(Life Cycle Mode=::NewEnumerator1, Loop Behavior=::NewEnumerator1, Loop Delay=$System.delay, Loop Duration=1, UseLoopDelay=true)
   &SpawnBurst_Instantaneous(Spawn Count=&UniformRangedInt(Minimum=6, Maximum=12), Use Spawn Probability=false)
   &SpawnBurst_Instantaneous(Spawn Count=&UniformRangedInt(Minimum=2, Maximum=5), Spawn Group=1, Use Spawn Probability=false) {module=SpawnBurst_Instantaneous001}
  ]
  particleSpawnScript [
   /SetVariables(Particles.RandonInt2:NiagaraInt32=&UniformRangedInt(Minimum=0, Maximum=3)) {isSetParameters=true}
   /Niagara/Modules/Spawn/Initialization/InitializeParticle(
    Lifetime=&MaskFloatBySpawnGroup(Masked Float=&UniformRangedFloat(Minimum=0.5, Maximum=2), Passthrough Float=2)
    Mass=&UniformRangedFloat(Minimum=25, Maximum=100)
    Mesh Scale=&UniformRangedVector(Minimum=<8 8 8>, Maximum=<15 15 15>)
    Module.Write Color=true
    Module.Write Lifetime=true
    Module.Write Mass=true
    Module.Write Material Random=true
    Module.Write Position=false
    Module.Write Scale=true
    Module.Write SpriteRotation=false
    Module.Write SpriteSize=false
    Position=&SimulationPosition()
   )
   &CalculateSizeAndRotationalInertiaByMass(Calculate Mesh Scale=true, Initial Model Dimensions:Vector3f=<12 12 12>, Module.Manually Enter Density=false)
   &AddVelocityInCone(
    Cone Angle=90
    Cone Axis=<1 0 0.8>
    Module.Use Velocity Falloff On Cone Axis=true
    Velocity Falloff Away From Cone Axis=0.333
    Velocity Strength=&UniformRangedFloat(Minimum=175, Maximum=1500)
   ) {enabled=false}
   &CylinderLocation(
    Cylinder Height=0
    Cylinder Radius=60
    Hemisphere X=true
    Lathe Profile=DataInterface() {propertyValues {Curve {keys [<at 0 0.451 RCIM_Cubic>, {arriveTangent=0.049, interpMode=RCIM_Cubic, leaveTangent=0.049, time=0.5, value=1}, <at 1 0.5 RCIM_Cubic>]}}}
    MaskSpawnGroup=false
    Module.Enable Lathe Profile=false
    Module.Override Randomness=false
    Module.Surface Only=true
    Non Uniform Scale:Vector3f=<1 0.35 1>
   )
   &AddVelocityFromPoint(
    Module.Use Falloff Distance=false
    Origin Offset=&UniformRangedVector(Minimum=<0 0 -4>, Maximum=<1 1 -25>)
    Origin Offset Coordinate Space=::Simulation
    Velocity Strength=&UniformRangedFloat(Minimum=250, Maximum=1500)
   )
   &AddRotationalVelocity(Rotation Rate=&UniformRangedVector(Minimum=<-2 -2 -2>, Maximum=<2 2 2>))
   &ApplyInitialForces()
   &InitialMeshOrientation(
    Rotation=&UniformRangedVector(Minimum=<0 0 0>, Maximum=<1 1 1>)
    Use Orientation Vector=true
    Use Rotation Vector=true
   ) {module=InitialMeshOrientation001}
  ]
  particleUpdateScript [
   &ParticleState()
   &GravityForce()
   &Drag(Use Linear Drag=true, Use Rotational Drag=false)
   &ScaleColor(
    Scale Alpha=&FloatFromCurve(FloatCurve=DataInterface() {propertyValues {Curve {keys [<at -0.0046 1.0113 RCIM_Cubic>,<at 1.0011 0.002 RCIM_Cubic>]}}})
    ScaleA=true
    ScaleRGB=true
   )
   &SolveForcesAndVelocity(Clamp Velocity=false, Limit Acceleration=false)
   &SolveRotationalForcesAndVelocity()
   &ScaleMeshSize(Scale Factor=&VectorFromFloat(Value=&FloatFromCurve(FloatCurve=DataInterface() {propertyValues {Curve {keys [<at 0 1>,<at 0.7956 1.0207>,<at 1 0>]}}})))
  ]
  rendererClasses [&NiagaraMeshRendererProperties]
  renderers [
   {rendererClass=&NiagaraMeshRendererProperties, rendererIndex=0}
   {rendererClass=&NiagaraMeshRendererProperties, rendererIndex=1}
   {rendererClass=&NiagaraMeshRendererProperties, rendererIndex=2}
   {rendererClass=&NiagaraMeshRendererProperties, rendererIndex=3}
  ]
 }
 DirectionalBurstPaint {
  bEnabled=false
  emitterUpdateScript [
   &EmitterState(Life Cycle Mode=::NewEnumerator1, Loop Behavior=::NewEnumerator1, Loop Duration=1, UseLoopDelay=false)
   &SpawnBurst_Instantaneous(Spawn Count=4, Use Spawn Probability=false)
  ]
  particleSpawnScript [
   &InitializeParticle(
    Color=&UniformRangedLinearColor(Minimum=<rgba 0.6771 0.6771 0.6771 0.75>, Maximum=<rgba 0.1823 0.1823 0.1823 0.85>)
    Lifetime=&UniformRangedFloat(Minimum=1, Maximum=1.4)
    Mass=&UniformRangedFloat(Minimum=240, Maximum=500)
    Module.Write Color=true
    Module.Write Lifetime=true
    Module.Write Mass=true
    Module.Write Material Random=true
    Module.Write Position=false
    Module.Write Scale=false
    Module.Write SpriteRotation=true
    Module.Write SpriteSize=true
    Position=&SimulationPosition()
    Sprite Rotation=&UniformRangedFloat(Minimum=0, Maximum=360)
    Sprite Size=&Vector2DFromFloat(Value=100)
   )
   &AddVelocityInCone(
    Cone Angle=90
    Cone Axis=<1 0 0.8>
    Module.Use Velocity Falloff On Cone Axis=true
    Velocity Falloff Away From Cone Axis=0.333
    Velocity Strength=&UniformRangedFloat(Minimum=175, Maximum=1200)
   ) {enabled=false}
   &CylinderLocation(
    Cylinder Height=0
    Cylinder Radius=60
    Hemisphere X=true
    Lathe Profile=DataInterface() {propertyValues {Curve {keys [<at 0 0.451 RCIM_Cubic>, {arriveTangent=0.049, interpMode=RCIM_Cubic, leaveTangent=0.049, time=0.5, value=1}, <at 1 0.5 RCIM_Cubic>]}}}
    MaskSpawnGroup=false
    Module.Enable Lathe Profile=false
    Module.Override Randomness=false
    Module.Surface Only=true
    Non Uniform Scale:Vector3f=<1 0.35 1>
   )
   &AddVelocityFromPoint(
    Module.Use Falloff Distance=false
    Origin Offset=&UniformRangedVector(Minimum=<0 0 -4>, Maximum=<1 1 -25>)
    Origin Offset Coordinate Space=::Simulation
    Velocity Strength=&UniformRangedFloat(Minimum=250, Maximum=600)
   )
  ]
  particleUpdateScript [
   &ParticleState()
   &GravityForce(Gravity=<0 0 -1600>)
   &Drag(Use Linear Drag=true, Use Rotational Drag=false) {enabled=false}
   &ScaleColor(
    Scale Alpha=&FloatFromCurve(FloatCurve=DataInterface() {propertyValues {Curve {keys [<at 0.2874 1.0044 RCIM_Cubic>,<at 1.0011 0.002 RCIM_Cubic>]}}})
    ScaleA=true
    ScaleRGB=true
   )
   &SolveForcesAndVelocity(Clamp Velocity=false, Limit Acceleration=false)
   &ScaleSpriteSize(Scale Factor=&Vector2DFromFloat(Value=&FloatFromCurve(FloatCurve=DataInterface() {propertyValues {Curve {keys [
          {interpMode=RCIM_Cubic, leaveTangent=8.4581, tangentMode=RCTM_Break, time=0, value=0}
          {arriveTangent=1.8771, interpMode=RCIM_Cubic, leaveTangent=1.8771, tangentMode=RCTM_User, time=0.3915, value=2.7119}
          {interpMode=RCIM_Cubic, tangentMode=RCTM_User, time=0.9983, value=3.3515}
         ]}}})))
   &DynamicMaterialParameters(Module.Param0WriteEnabled=true, Module.Param1WriteEnabled=true, Module.Param2WriteEnabled=true, Module.Param3WriteEnabled=true)
  ]
  rendererClasses [&NiagaraSpriteRendererProperties]
  renderers [{rendererClass=&NiagaraSpriteRendererProperties}]
 }
 DirectionalBurstSparks001 {
  simTarget=GPUComputeSim
  emitterUpdateScript [
   &EmitterState(Life Cycle Mode=::NewEnumerator1, Loop Behavior=::NewEnumerator1, Loop Delay=$System.delay, Loop Duration=1, UseLoopDelay=true)
   &SpawnBurst_Instantaneous(
    Spawn Count=&MakeIntFromFloat(Float=&MakeCustomFloatFromBool(Bool=&SetBoolByFloatComparison(A=70, B=$System.RandInt), True Float=1, False Float=0))
    Use Spawn Probability=false
    UseLoopCountLimit=false
   )
  ]
  particleSpawnScript [
   &InitializeParticle(
    Color=&SelectLinearColorFromArray(
     Array Sampling Mode=::NewEnumerator0
     Direct Array Index=?
     Color Selection Array=DataInterface() {propertyValues {ColorData [<rgba 350 80 0 1>, <rgba 25 15 0 1>]}}
     Random Seed=0
    )
    Lifetime=&RandomRangeFloat(Minimum=1, Maximum=1.5)
    Mass=&UniformRangedFloat(Minimum=80, Maximum=350)
    Module.Write Color=true
    Module.Write Lifetime=true
    Module.Write Mass=true
    Module.Write Material Random=true
    Module.Write Position=false
    Module.Write Scale=false
    Module.Write SpriteRotation=false
    Module.Write SpriteSize=true
    Position=&SimulationPosition()
    Sprite Size=&Vector2DFromFloat(Value=&UniformRangedFloat(Minimum=5, Maximum=10))
   )
   &CalculateSizeAndRotationalInertiaByMass(Depth=0, Height=2, Mass Modulation=3, Module.Manually Enter Density=false, Width=0.255)
   &AddVelocity(
    Cone Angle=4
    Module.UseConeFalloff=false
    Module.UseVelDistribution=false
    Override Seed=false
    Velocity Mode=::NewEnumerator2
    Velocity Speed=&RandomRangeFloat(Minimum=1250, Maximum=2400)
   )
  ]
  particleUpdateScript [
   &ParticleState()
   &GravityForce(Gravity=&UniformRangedVector(Minimum=<0 0 -980>, Maximum=<1 1 -1800>))
   &Drag(Drag=&UniformRangedFloat(Minimum=0.5, Maximum=1), Use Linear Drag=true, Use Rotational Drag=false) {enabled=false}
   &ScaleColor(
    Scale Alpha=&FloatFromCurve(FloatCurve=DataInterface() {propertyValues {Curve {keys [<at -0.0046 1.0113 RCIM_Cubic>,<at 1.0011 0.002 RCIM_Cubic>]}}}, Scale Curve=0.4)
    ScaleA=true
    ScaleRGB=true
   )
   &Collision(
    Advanced Aging Rate=1
    Age Colliding Particles=false
    Collision Enabled=&MaskBoolBySpawnGroup(
     Masked Bool:NiagaraBool=false
     Passthrough Bool:NiagaraBool=&SetBoolByFloatComparison(A=$Particles.Age, B=0.15)
     Spawn Group Mask:NiagaraInt32=0
    )
    Kill Offscreen Particles=true
    Module.Control Rest State Via Surface Normal=false
    Module.EnableMaxCollisionCount=false
    Module.Randomize Collision Normal Vector=true
   )
   &CurlNoiseForce(Noise Frequency=0.1, Noise Quality / Cost=::NewEnumerator0, Noise Strength=50000, Pan Noise=true, Pan Noise Field=<25 0 0>)
   /SetVariables(Particles.PositionOld:NiagaraPosition=$Particles.Position) {isSetParameters=true}
   &SolveForcesAndVelocity(Clamp Velocity=false, Limit Acceleration=false)
   /SetVariables(Particles.dir:Vector3f=&Subtract_Vector(A=Linked() {linkedVariable {name=Particles.Position, type=NiagaraPosition}}, B=$Particles.PositionOld)) {isSetParameters=true}
   &ScaleSpriteSizeBySpeed(
    Max Scale Factor=<0.7 0.5>
    Min Scale Factor=<0.4 0.1>
    Sample Scale Factor By Curve=true
    Scale Factor Curve=DataInterface() {propertyValues {Curve {keys [
        {arriveTangent=2.865, interpMode=RCIM_Cubic, time=0.1247, value=-0.0006}
        {arriveTangent=2.5675, interpMode=RCIM_Cubic, leaveTangent=2.5675, tangentMode=RCTM_User, time=0.3948, value=0.4893}
        <at 0.9958 0.9974 RCIM_Cubic>
       ]}}}
    Velocity Threshold=1500
   ) {module=SpriteSizeScaleByVelocity}
   &ScaleSpriteSize(Scale Factor=&Vector2DFromFloat(Value=&FloatFromCurve(
      FloatCurve=DataInterface() {propertyValues {Curve {keys [
          {interpMode=RCIM_Cubic, tangentMode=RCTM_Break, time=0.7172, value=1}
          {arriveTangent=-9.028, interpMode=RCIM_Cubic, leaveTangent=-9.028, tangentMode=RCTM_User, time=1, value=0}
         ]}}}
      Scale Curve=1.25
     )))
  ]
  rendererClasses [&NiagaraSpriteRendererProperties]
  renderers [{rendererClass=&NiagaraSpriteRendererProperties}]
 }
 DirectionalBurstSparks002 {
  simTarget=GPUComputeSim
  emitterUpdateScript [
   &EmitterState(
    Life Cycle Mode=::NewEnumerator1
    Loop Behavior=::NewEnumerator1
    Loop Delay=&Add_Float(A=$System.delay, B=0.0025)
    Loop Duration=1
    UseLoopDelay=true
   )
   &SpawnRate(
    SpawnGroup=1
    SpawnRate=&FloatFromCurve(
     FloatCurve=DataInterface() {propertyValues {Curve {keys [{interpMode=RCIM_Cubic, leaveTangent=-1.0025, tangentMode=RCTM_Break, time=0, value=1}, <at 1 0.55>]}}}
     CurveIndex=$Emitter.Age
     Scale Curve=&MakeFloatFromInt(Int=&MakeIntFromFloat(Float=&MakeCustomFloatFromBool(Bool=&SetBoolByFloatComparison(A=70, B=$System.RandInt), True Float=150, False Float=0)))
    )
    Use Spawn Probability=false
   )
  ]
  particleSpawnScript [
   &InitializeParticle(
    Color=&SelectLinearColorFromArray(
     Array Sampling Mode=::NewEnumerator0
     Direct Array Index=?
     Color Selection Array=DataInterface() {propertyValues {ColorData [<rgba 350 80 0 1>, <rgba 25 15 0 1>]}}
     Random Seed=0
    )
    Lifetime=&RandomRangeFloat(Minimum=0.25, Maximum=0.7)
    Mass=&FloatFromCurve(
     FloatCurve=DataInterface() {propertyValues {Curve {keys [<at 0 1>,<at 1 0>]}}}
     Scale Curve=&UniformRangedFloat(Minimum=4, Maximum=250)
    )
    Module.Write Color=true
    Module.Write Lifetime=true
    Module.Write Mass=true
    Module.Write Material Random=true
    Module.Write Position=false
    Module.Write Scale=false
    Module.Write SpriteRotation=false
    Module.Write SpriteSize=true
    Position=&SimulationPosition()
    Sprite Size=&Vector2DFromFloat(Value=&UniformRangedFloat(Minimum=5, Maximum=10))
   ) {enabled=false}
   /Niagara/Modules/Spawn/Initialization/V2/InitializeParticle(
    Color=&SelectLinearColorFromArray(
     Array Sampling Mode=::NewEnumerator0
     Direct Array Index=?
     Color Selection Array=DataInterface() {propertyValues {ColorData [<rgba 350 80 0 1>, <rgba 25 15 0 1>]}}
     Random Seed=0
    )
    Lifetime=&RandomRangeFloat(Minimum=0.25, Maximum=0.7)
    Mass=&FloatFromCurve(
     FloatCurve=DataInterface() {propertyValues {Curve {keys [<at 0 1>,<at 1 0>]}}}
     Scale Curve=&UniformRangedFloat(Minimum=4, Maximum=250)
    )
    Mass Mode=::NewEnumerator1
    Sprite Size=&Vector2DFromFloat(Value=&UniformRangedFloat(Minimum=5, Maximum=10))
    Sprite Size Mode=::NewEnumerator4
    UsePositionOffset=false
    Write Material Random=true
   ) {module=InitializeParticle001}
   /SetVariables(Particles.Position=&self:ScratchDynamicInput(New Particle Attribute Reader:NiagaraDataInterfaceParticleRead=DataInterface() {propertyValues {EmitterBinding {bindingMode=Other, emitterName=DirectionalBurstSparks001}}})) {isSetParameters=true}
   &CalculateSizeAndRotationalInertiaByMass(Depth=0, Height=5, Module.Manually Enter Density=false)
   /SetVariables(Particles.dir:Vector3f=&self:ScratchDynamicInput_01(New Particle Attribute Reader:NiagaraDataInterfaceParticleRead=DataInterface() {propertyValues {EmitterBinding {bindingMode=Other, emitterName=DirectionalBurstSparks001}}})) {isSetParameters=true}
   &ShapeLocation(
    Cylinder Height=1
    Cylinder Radius=15
    MaskSpawnGroup=false
    Non Uniform Scale=<1 1 0.25>
    Override Seed=false
    Rotation Axis=$Particles.dir
    Rotation Mode=::NewEnumerator1
    Shape Primitive=::NewEnumerator1
    Surface Only=true
    Surface Only Band Thickness=0.1
    Use Endcaps In Surface Only Mode=false
    UseOwnerScale=true
   )
   &AddVelocity(
    Module.Use Falloff Distance=false
    Origin Offset Coordinate Space=::World
    Override Seed=false
    Velocity Mode=::NewEnumerator1
    Velocity Origin=$Particles.Position
    Velocity Speed=&FloatFromCurve(
     FloatCurve=DataInterface() {propertyValues {Curve {keys [<at 0 0>, {interpMode=RCIM_Cubic, leaveTangent=-1.6226, tangentMode=RCTM_Break, time=0.05, value=1}, <at 1 0.4>]}}}
     CurveIndex=$Emitter.Age
     Scale Curve=&RandomRangeFloat(Minimum=25, Maximum=1500)
    )
   ) {enabled=false}
   &AddVelocity(
    Cone Angle=8
    Cone Axis=&TransformVector(Vector=$Particles.dir, Source Space=::World, Destination Space=::Local)
    Module.UseConeFalloff=false
    Module.UseVelDistribution=false
    Override Seed=false
    Velocity Mode=::NewEnumerator2
    Velocity Speed=&FloatFromCurve(
     FloatCurve=DataInterface() {propertyValues {Curve {keys [<at 0 0.45>, {interpMode=RCIM_Cubic, leaveTangent=-1.6226, tangentMode=RCTM_Break, time=0.05, value=1}, <at 1 0.4>]}}}
     CurveIndex=$Emitter.Age
     Scale Curve=&RandomRangeFloat(Minimum=25, Maximum=1500)
    )
   ) {module=AddVelocity001}
  ]
  particleUpdateScript [
   &ParticleState()
   &GravityForce(Gravity=&Lerp_Vector(
     A=<0 0 0>
     B=&RandomRangeVector(Minimum=<0 0 -500>, Maximum=<1 1 -3500>)
     Alpha=$Particles.NormalizedAge
    ))
   &ScaleColor(
    Scale Alpha=&FloatFromCurve(FloatCurve=DataInterface() {propertyValues {Curve {keys [<at -0.0046 1.0113 RCIM_Cubic>,<at 1.0011 0.002 RCIM_Cubic>]}}}, Scale Curve=0.4)
    ScaleA=true
    ScaleRGB=true
   )
   &CurlNoiseForce(
    Noise Frequency=0.005
    Noise Quality / Cost=::NewEnumerator0
    Noise Strength=&FloatFromCurve(
     FloatCurve=DataInterface() {propertyValues {Curve {keys [{interpMode=RCIM_Cubic, leaveTangent=2.4969, tangentMode=RCTM_Break, time=0, value=0}, <at 1 1>]}}}
     Scale Curve=150000
    )
    Pan Noise=true
    Pan Noise Field=<25 0 25>
   )
   &Drag(
    Drag=&FloatFromCurve(FloatCurve=DataInterface() {propertyValues {Curve {keys [
         {interpMode=RCIM_Cubic, leaveTangent=-6.9985, tangentMode=RCTM_Break, time=0, value=1}
         {arriveTangent=-0.1691, interpMode=RCIM_Cubic, leaveTangent=-0.1691, tangentMode=RCTM_User, time=0.1, value=0.3}
         <at 1 0.125>
        ]}}})
    Use Linear Drag=true
    Use Rotational Drag=false
   )
   &Collision(
    Advanced Aging Rate=1
    Age Colliding Particles=false
    Collision Enabled=&MaskBoolBySpawnGroup(
     Masked Bool:NiagaraBool=false
     Passthrough Bool:NiagaraBool=&SetBoolByFloatComparison(A=$Particles.Age, B=0.15)
     Spawn Group Mask:NiagaraInt32=0
    )
    Kill Offscreen Particles=true
    Module.Control Rest State Via Surface Normal=false
    Module.EnableMaxCollisionCount=false
    Module.Randomize Collision Normal Vector=true
   )
   &SolveForcesAndVelocity(Clamp Velocity=false, Limit Acceleration=false)
   &ScaleSpriteSizeBySpeed(
    Max Scale Factor=<0.7 0.5>
    Min Scale Factor=<0.4 0.1>
    Sample Scale Factor By Curve=true
    Scale Factor Curve=DataInterface() {propertyValues {Curve {keys [
        {arriveTangent=2.865, interpMode=RCIM_Cubic, time=0.1247, value=-0.0006}
        {arriveTangent=2.5675, interpMode=RCIM_Cubic, leaveTangent=2.5675, tangentMode=RCTM_User, time=0.3948, value=0.4893}
        <at 0.9958 0.9974 RCIM_Cubic>
       ]}}}
    Velocity Threshold=1500
   ) {module=SpriteSizeScaleByVelocity}
   &ScaleSpriteSize(Scale Factor=&Vector2DFromFloat(Value=&FloatFromCurve(FloatCurve=DataInterface() {propertyValues {Curve {keys [
          <at 0 0>
          <at 0.1 1>
          {interpMode=RCIM_Cubic, tangentMode=RCTM_Break, time=0.7172, value=1}
          {arriveTangent=-9.028, interpMode=RCIM_Cubic, leaveTangent=-9.028, tangentMode=RCTM_User, time=1, value=0}
         ]}}})))
  ]
  rendererClasses [&NiagaraSpriteRendererProperties]
  renderers [{rendererClass=&NiagaraSpriteRendererProperties}]
 }
 LW_DirBurstSpark_HIGHMIDLOW {rendererClasses [], renderers []}
 LW_DirBurstSpark_HIGHMIDLOW001 {rendererClasses [], renderers []}
 LW_FN_Core_Smoke_Sprite {rendererClasses [], renderers []}
 Light {
  emitterUpdateScript [
   &EmitterState(Life Cycle Mode=::NewEnumerator1, Loop Behavior=::NewEnumerator1, Loop Delay=$System.delay, Loop Duration=1, UseLoopDelay=true)
   &SpawnBurst_Instantaneous(Spawn Count=1, Use Spawn Probability=false)
  ]
  particleSpawnScript [/Niagara/Modules/Spawn/Initialization/InitializeParticle(
    Color=<rgba 15 10 1 1>
    Lifetime=0.15
    Mesh Scale=&VectorFromFloat(Value=&Multiply_Float(A=$User.ImpactStrength, B=&UniformRangedFloat(Minimum=4, Maximum=6)))
    Module.Write Color=true
    Module.Write Lifetime=true
    Module.Write Mass=false
    Module.Write Material Random=true
    Module.Write Position=false
    Module.Write Scale=true
    Module.Write SpriteRotation=false
    Module.Write SpriteSize=true
    Position=&SimulationPosition()
    Sprite Size=&Vector2DFromFloat(Value=300)
   )]
  particleUpdateScript [
   &ParticleState()
   &ScaleColor(
    Scale Alpha=&FloatFromCurve(FloatCurve=DataInterface() {propertyValues {Curve {keys [<at 0 0.9859 RCIM_Cubic>,<at 1.0011 0 RCIM_Cubic>]}}})
    Scale RGB=/Niagara/DynamicInputs/ScaleByCurve/ScaleVectorByCurve(
     InitialValue:Vector3f=<1 1 1>
     ScaleCurve:NiagaraDataInterfaceVectorCurve=DataInterface() {propertyValues {XCurve {keys [<at 0 1>,<at 1 0>]}, YCurve {keys [<at 0 1>,<at 1 0>]}, ZCurve {keys [<at 0 1>,<at 1 0>]}}}
     CurveIndex:NiagaraFloat=$Particles.NormalizedAge
    )
    ScaleA=true
    ScaleRGB=true
   )
   &DynamicMaterialParameters(
    Index 0 Param 1=&FloatFromCurve(FloatCurve=DataInterface() {propertyValues {Curve {keys [<at 0.55 0>,<at 1 1>]}}})
    Index 0 Param 2=250
    Module.Param0WriteEnabled=true
    Module.Param1WriteEnabled=true
    Module.Param2WriteEnabled=true
    Module.Param3WriteEnabled=true
   )
   &ScaleSpriteSize()
  ]
  rendererClasses [&NiagaraLightRendererProperties]
  renderers [{rendererClass=&NiagaraLightRendererProperties, rendererIndex=0}]
 }
}
summary {
 systemName=NS_Scrape_Impact
 emitters [
  {emitterName=DirBurstSpark_EPIC, simTarget=GPUComputeSim, rendererClasses [&NiagaraSpriteRendererProperties]}
  {emitterName=DirectionalBurstMeshes_01, rendererClasses [&NiagaraMeshRendererProperties]}
  {emitterName=Light, rendererClasses [&NiagaraLightRendererProperties]}
  {bEnabled=false, emitterName=DirectionalBurstPaint, rendererClasses [&NiagaraSpriteRendererProperties]}
  {bEnabled=false, emitterName=World_Smoke_Forward, rendererClasses [&NiagaraSpriteRendererProperties]}
  {emitterName=DirectionalBurstSparks001, simTarget=GPUComputeSim, rendererClasses [&NiagaraSpriteRendererProperties]}
  {emitterName=DirectionalBurstSparks002, simTarget=GPUComputeSim, rendererClasses [&NiagaraSpriteRendererProperties]}
  {bEnabled=false, emitterName=FN_Core_Smoke_Sprite, rendererClasses [&NiagaraSpriteRendererProperties]}
  {
   bEnabled=false
   emitterName=DirBurstSpark_HIGHMIDLOW
   simTarget=GPUComputeSim
   rendererClasses [&NiagaraSpriteRendererProperties]
  }
  {emitterName=LW_DirBurstSpark_HIGHMIDLOW, rendererClasses []}
  {emitterName=LW_DirBurstSpark_HIGHMIDLOW001, rendererClasses []}
  {emitterName=LW_FN_Core_Smoke_Sprite, rendererClasses []}
 ]
 userVariables [
  User.ImpactStrength {type=NiagaraFloat, defaultValue=1, description=""}
  User.LandOnSnowOrIce {type=NiagaraBool}
 ]
}
