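# Niagara one-shot effect for a player scrape impact, throwing mesh debris alongside LW-prefixed paint and smoke emitters whose internals are not shown.
# Two burst spawn groups launch randomly scaled, spinning meshes from a squashed half-ring, away from an offset point.
# Gravity, drag and collision drive them as they fade and shrink, with lifetime and collision varying by spawn group.
# Standard paint and smoke emitters are disabled; User.ImpactStrength has no visible use.
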NiagaraSystem /Valet/FX/Niagara/NS_Scrape_PlayerImpact
dependencies=null
emitters {
 DirectionalBurstMeshes_01 {
  emitterUpdateScript [
   &EmitterState(Life Cycle Mode=::NewEnumerator1, Loop Behavior=::NewEnumerator1, Loop Duration=1, UseLoopDelay=false)
   &SpawnBurst_Instantaneous(Spawn Count=&UniformRangedInt(Minimum=6, Maximum=12), Use Spawn Probability=false)
   &SpawnBurst_Instantaneous(Spawn Count=&UniformRangedInt(Minimum=2, Maximum=5), Spawn Group=1, Use Spawn Probability=false) {module=SpawnBurst_Instantaneous001}
  ]
  particleSpawnScript [
   /SetVariables(Particles.RandomInt:NiagaraInt32=&UniformRangedInt(Minimum=0, Maximum=1)) {isSetParameters=true}
   /Niagara/Modules/Spawn/Initialization/InitializeParticle(
    Lifetime=&MaskFloatBySpawnGroup(Masked Float=&UniformRangedFloat(Minimum=0.5, Maximum=1.2), Passthrough Float=2)
    Mass=&UniformRangedFloat(Minimum=1, Maximum=3)
    Mesh Scale=&UniformRangedVector(Minimum=<0.7 0.7 0.7>, Maximum=<2 2 2>)
    Module.Write Color=true
    Module.Write Lifetime=true
    Module.Write Mass=true
    Module.Write Material Random=true
    Module.Write Position=false
    Module.Write Scale=true
    Module.Write SpriteRotation=false
    Module.Write SpriteSize=false
    Position=&SimulationPosition()
   )
   /Niagara/Modules/Update/Mass/CalculateSizeAndRotationalInertiaByMass(Calculate Mesh Scale=true, Initial Model Dimensions:Vector3f=<12 12 12>, Module.Manually Enter Density=false)
   &AddVelocityInCone(
    Cone Angle=90
    Cone Axis=<1 0 0.8>
    Module.Use Velocity Falloff On Cone Axis=true
    Velocity Falloff Away From Cone Axis=0.333
    Velocity Strength=&UniformRangedFloat(Minimum=175, Maximum=1500)
   ) {enabled=false}
   &CylinderLocation(
    Cylinder Height=0
    Cylinder Radius=60
    Hemisphere X=true
    Lathe Profile=DataInterface() {propertyValues {Curve {keys [<at 0 0.451 RCIM_Cubic>, {arriveTangent=0.049, interpMode=RCIM_Cubic, leaveTangent=0.049, time=0.5, value=1}, <at 1 0.5 RCIM_Cubic>]}}}
    MaskSpawnGroup=false
    Module.Enable Lathe Profile=false
    Module.Override Randomness=false
    Module.Surface Only=true
    Non Uniform Scale:Vector3f=<1 0.35 1>
   )
   &AddVelocityFromPoint(
    Module.Use Falloff Distance=false
    Origin Offset=&UniformRangedVector(Minimum=<0 0 -4>, Maximum=<1 1 -25>)
    Origin Offset Coordinate Space=::Simulation
    Velocity Strength=&UniformRangedFloat(Minimum=250, Maximum=1500)
   )
   /Niagara/Modules/Spawn/Orientation/V2/InitialMeshOrientation(
    Rotation=&UniformRangedVector(Minimum=<0 0 0>, Maximum=<1 1 1>)
    Use Orientation Vector=true
    Use Rotation Vector=true
   )
   &AddRotationalVelocity(Rotation Rate=&UniformRangedVector(Minimum=<-2 -2 -2>, Maximum=<2 2 2>))
   &ApplyInitialForces()
  ]
  particleUpdateScript [
   &ParticleState()
   &GravityForce()
   &Drag(Use Linear Drag=true, Use Rotational Drag=false)
   &ScaleColor(
    Scale Alpha=&FloatFromCurve(FloatCurve=DataInterface() {propertyValues {Curve {keys [<at -0.0046 1.0113 RCIM_Cubic>,<at 1.0011 0.002 RCIM_Cubic>]}}})
    ScaleA=true
    ScaleRGB=true
   )
   &Collision(
    Advanced Aging Rate=1
    Age Colliding Particles=false
    Collision Enabled=/Niagara/DynamicInputs/Helpers/MaskBoolBySpawnGroup(
     Masked Bool:NiagaraBool=false
     Passthrough Bool:NiagaraBool=&SetBoolByFloatComparison(A=$Particles.Age, B=0.05)
     Spawn Group Mask:NiagaraInt32=0
    )
    Control Roll On Collision=true
    Module.Control Rest State Via Surface Normal=false
    Module.EnableMaxCollisionCount=false
    Module.Randomize Collision Normal Vector=true
    ScaleRateOfRotation=false
   )
   &SolveForcesAndVelocity(Clamp Velocity=false, Limit Acceleration=false)
   &SolveRotationalForcesAndVelocity()
   &ScaleMeshSize(Scale Factor=&VectorFromFloat(Value=&FloatFromCurve(FloatCurve=DataInterface() {propertyValues {Curve {keys [<at 0 1>,<at 0.7956 1.0207>,<at 1 0>]}}})))
  ]
  rendererClasses [&NiagaraMeshRendererProperties]
  renderers [
   {rendererClass=&NiagaraMeshRendererProperties, rendererIndex=0}
   {rendererClass=&NiagaraMeshRendererProperties, rendererIndex=1}
  ]
 }
 LW_DirectionalBurstPaint {rendererClasses [], renderers []}
 LW_FN_Core_Smoke_Sprite {rendererClasses [], renderers []}
 World_Smoke_Forward {
  bEnabled=false
  emitterUpdateScript [
   &EmitterState()
   &SpawnBurst_Instantaneous(Spawn Count=3, Use Spawn Probability=false)
  ]
  particleSpawnScript [
   &InitializeParticle(
    Lifetime=&UniformRangedFloat(Minimum=0.8, Maximum=1.4)
    Module.Write Color=true
    Module.Write Lifetime=true
    Module.Write Mass=true
    Module.Write Material Random=true
    Module.Write Position=false
    Module.Write Scale=false
    Module.Write SpriteRotation=true
    Module.Write SpriteSize=true
    Position=&SimulationPosition()
    Sprite Rotation=&NormalizedAngleToDegrees(NormalizedAngle=&UniformRangedFloat(Minimum=0, Maximum=1))
    Sprite Size=&Vector2DFromFloat(Value=&UniformRangedFloat(Minimum=150, Maximum=250))
   )
   &CameraOffset(Camera Offset Amount=25)
   &BoxLocation(Box Size=<180 180 0>, MaskSpawnGroup=false, Module.Override Seed=false, Rotate Box=false, Surface Only=false)
   &AddVelocityFromPoint(Module.Use Falloff Distance=false, Origin Offset=<0 0 5>, Velocity Strength=250)
  ]
  particleUpdateScript [
   &ParticleState()
   &AccelerationForce(Acceleration=<0 0 500>)
   &Drag(Drag=3, Use Linear Drag=true, Use Rotational Drag=false)
   &SolveForcesAndVelocity(Clamp Velocity=false, Limit Acceleration=false)
   /SetVariables(Particles.SubImageIndex=&FloatFromCurve(FloatCurve=DataInterface() {propertyValues {Curve {keys [{arriveTangent=210.305, interpMode=RCIM_Cubic, leaveTangent=113.9454, tangentMode=RCTM_User, time=0, value=10}, <at 1 55>]}}})) {isSetParameters=true}
   &ScaleColor(
    Scale Alpha=&FloatFromCurve(FloatCurve=DataInterface() {propertyValues {Curve {keys [
         {interpMode=RCIM_Cubic, tangentMode=RCTM_User, time=0, value=0}
         {interpMode=RCIM_Cubic, tangentMode=RCTM_User, time=0.0188, value=1}
         {interpMode=RCIM_Cubic, tangentMode=RCTM_User, time=1, value=0}
        ]}}})
    Scale RGB=&VectorFromFloat(Value=&FloatFromCurve(FloatCurve=DataInterface() {propertyValues {Curve {keys [
          {arriveTangent=-9.3735, interpMode=RCIM_Cubic, leaveTangent=-6.2409, tangentMode=RCTM_User, time=0, value=1.25}
          <at 0.3521 0.268>
          <at 1 0.2114 RCIM_Cubic>
         ]}}}))
    ScaleA=true
    ScaleRGB=true
   )
   &DynamicMaterialParameters(
    Index 0 Param 1=&FloatFromCurve(FloatCurve=DataInterface() {propertyValues {Curve {keys [<at 0 1>,<at 1 -1>]}}})
    Index 0 Param 4=1
    Module.Param0WriteEnabled=true
    Module.Param1WriteEnabled=true
    Module.Param2WriteEnabled=true
    Module.Param3WriteEnabled=true
   )
   &ScaleSpriteSize(Scale Factor=&Vector2DFromFloat(Value=&FloatFromCurve(FloatCurve=DataInterface() {propertyValues {Curve {keys [
          {arriveTangent=0.3894, interpMode=RCIM_Cubic, leaveTangent=0.3894, tangentMode=RCTM_User, time=0, value=1}
          {arriveTangent=2.8662, interpMode=RCIM_Cubic, leaveTangent=2.8662, tangentMode=RCTM_User, time=0.2628, value=2.528}
          <at 1 3 RCIM_Cubic>
         ]}}})))
   &SpriteRotationRate(Rotation Rate=&Multiply_Float(
     A=&Lerp_Float(A=-1, B=1, Alpha=$Particles.MaterialRandom)
     B=&FloatFromCurve(FloatCurve=DataInterface() {propertyValues {Curve {keys [<at 0 1>,<at 0.0958 0.0025>,<at 1 0>]}}}, Scale Curve=8000)
    ))
  ]
  rendererClasses [&NiagaraSpriteRendererProperties]
  renderers [{rendererClass=&NiagaraSpriteRendererProperties}]
 }
}
summary {
 systemName=NS_Scrape_PlayerImpact
 emitters [
  {emitterName=DirectionalBurstMeshes_01, rendererClasses [&NiagaraMeshRendererProperties]}
  {bEnabled=false, emitterName=DirectionalBurstPaint, rendererClasses [&NiagaraSpriteRendererProperties]}
  {bEnabled=false, emitterName=World_Smoke_Forward, rendererClasses [&NiagaraSpriteRendererProperties]}
  {bEnabled=false, emitterName=FN_Core_Smoke_Sprite, rendererClasses [&NiagaraSpriteRendererProperties]}
  {emitterName=LW_DirectionalBurstPaint, rendererClasses []}
  {emitterName=LW_FN_Core_Smoke_Sprite, rendererClasses []}
 ]
 userVariables [User.ImpactStrength {type=NiagaraFloat, defaultValue=1, description=""}]
}
