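# Niagara system that lays tire skid mark ribbons behind a vehicle's wheels.
# Two distance-based spawn groups (wheels 0 and 1) use User.WheelsTouchingArray as spawn probability, placing points at skeletal mesh bones/sockets offset by wheel radius.
# Ribbons face along the system Y axis, are camera-offset, and quickly fade in, then out over their lifetime.
# A disabled WheelTest sprite emitter uses inverted touch values; stored WheelWidth does not drive the fixed ribbon width.
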NiagaraSystem /Valet/FX/Niagara/NS_Tire_Skid_Trails
dependencies {usedDataInterfaces [&NiagaraDataInterfaceCurve], usedDynamicInputs [], usedModules [&SystemState]}
emitters {
 SkidMarks {
  emitterUpdateScript [
   &EmitterState(Life Cycle Mode=::NewEnumerator1, Loop Duration=1, UseLoopDelay=false)
   &SpawnPerUnit(
    Module.Use Max Movement Threshold=true
    Module.Use Movement Tolerance=false
    Movement Tolerance=0.1
    Spawn Probability:NiagaraFloat=&self:ScratchGetFloatFromArray(Index:NiagaraInt32=0, WheelsTouchingArray:NiagaraDataInterfaceArrayFloat=$User.WheelsTouchingArray)
    Spawn Spacing=70
    Use Spawn Probability=true
   )
   &SpawnPerUnit(
    Module.Use Max Movement Threshold=true
    Module.Use Movement Tolerance=false
    Movement Tolerance=0.1
    Spawn Group=1
    Spawn Probability:NiagaraFloat=&self:ScratchGetFloatFromArray(Index:NiagaraInt32=1, WheelsTouchingArray:NiagaraDataInterfaceArrayFloat=$User.WheelsTouchingArray)
    Spawn Spacing=70
    Use Spawn Probability=true
   ) {module=SpawnPerUnit001}
  ]
  particleSpawnScript [
   /SetVariables(
    Particles.WheelOffset:Vector3f=&MakeVector(
     X=0
     Y=0
     Z=&self:WheelRadiusOffset(
      Wheel Radius Array:NiagaraDataInterfaceArrayFloat=$User.Wheel_Radius_Array
      Wheel Index:NiagaraInt32=&SpawnGroup()
      Additional Offset:NiagaraFloat=0
     )
    )
    Particles.WheelWidth:NiagaraFloat=&Multiply_Float(
     A=&SelectFloatFromArray(Direct Array Index=&SpawnGroup(), Float Selection Array=$User.Wheel_Width_Array)
     B=1
    )
   ) {isSetParameters=true}
   &InitializeParticle(
    Color=<rgba 1 1 1 0.55>
    Lifetime=3.5
    Mass Mode=::NewEnumerator1
    Ribbon Facing Mode=::NewEnumerator1
    Ribbon Facing Vector:Vector3f=$Engine.Owner.SystemYAxis
    Ribbon Width=50
    Ribbon Width Mode=::NewEnumerator1
    Sprite Size=<25 25>
    Sprite Size Mode=::NewEnumerator4
    UsePositionOffset=false
    Write Material Random=true
   )
   &SkeletalMeshLocation(
    Bone / Socket Index=&SpawnGroup()
    Inherit Velocity=false
    Interpolate Mesh Sampling=false
    MaskSpawnGroup=false
    Mesh Sampling Type=::NewEnumerator5
    Sampled Position Offset=$Particles.WheelOffset
    Sampled Position Offset Coordinate Space:ENiagaraMeshTransforms=::NewEnumerator3
    Skeletal Mesh=$User.New Skeletal Mesh
    Socket Sampling Mode=::NewEnumerator5
   )
  ]
  particleUpdateScript [
   &ParticleState()
   &SolveForcesAndVelocity(Clamp Velocity=false, Limit Acceleration=false)
   &CameraOffset(Camera Offset Amount=32)
   &ScaleColor(
    Scale Alpha=&FloatFromCurve(
     FloatCurve=DataInterface() {propertyValues {Curve {keys [{interpMode=RCIM_Cubic, tangentMode=RCTM_Break, time=0, value=1}, <at 1 0>]}}}
     Scale Curve=&FloatFromCurve(
      FloatCurve=DataInterface() {propertyValues {Curve {keys [{interpMode=RCIM_Cubic, tangentMode=RCTM_Break, time=0, value=0}, {interpMode=RCIM_Cubic, tangentMode=RCTM_Break, time=0.1153, value=1}]}}}
      CurveIndex=$Particles.Age
     )
    )
    ScaleA=true
    ScaleRGB=true
   )
  ]
  rendererClasses [&NiagaraRibbonRendererProperties]
  renderers [{rendererClass=&NiagaraRibbonRendererProperties}]
 }
 WheelTest {
  bEnabled=false
  emitterUpdateScript [
   &EmitterState(Life Cycle Mode=::NewEnumerator1, Loop Duration=1, UseLoopDelay=false)
   &SpawnRate(
    Spawn Probability=&self:ScratchGetFloatFromArray_Invert(Index:NiagaraInt32=0, WheelsTouchingArray:NiagaraDataInterfaceArrayFloat=$User.WheelsTouchingArray)
    SpawnRate=30
    Use Spawn Probability=true
   )
   &SpawnRate(
    Spawn Probability=&self:ScratchGetFloatFromArray_Invert(Index:NiagaraInt32=0, WheelsTouchingArray:NiagaraDataInterfaceArrayFloat=$User.WheelsTouchingArray)
    SpawnGroup=1
    SpawnRate=30
    Use Spawn Probability=true
   ) {module=SpawnRate001}
   &SpawnRate(
    Spawn Probability=&self:ScratchGetFloatFromArray_Invert(Index:NiagaraInt32=0, WheelsTouchingArray:NiagaraDataInterfaceArrayFloat=$User.WheelsTouchingArray)
    SpawnGroup=2
    SpawnRate=30
    Use Spawn Probability=true
   ) {module=SpawnRate002}
   &SpawnRate(
    Spawn Probability=&self:ScratchGetFloatFromArray_Invert(Index:NiagaraInt32=3, WheelsTouchingArray:NiagaraDataInterfaceArrayFloat=$User.WheelsTouchingArray)
    SpawnGroup=3
    SpawnRate=30
    Use Spawn Probability=true
   ) {module=SpawnRate003}
  ]
  particleSpawnScript [
   /Niagara/Modules/Spawn/Initialization/InitializeParticle(
    Lifetime=0.5
    Module.Write Color=true
    Module.Write Lifetime=true
    Module.Write Mass=true
    Module.Write Material Random=true
    Module.Write Position=false
    Module.Write Scale=false
    Module.Write SpriteRotation=false
    Module.Write SpriteSize=true
    Position=&SimulationPosition()
    Sprite Size=<50 50>
   )
   &SkeletalMeshLocation(
    Bone / Socket Index=&SpawnGroup()
    Inherit Velocity=false
    Interpolate Mesh Sampling=false
    MaskSpawnGroup=false
    Mesh Sampling Type=::NewEnumerator5
    Sampled Position Offset Coordinate Space:ENiagaraMeshTransforms=::NewEnumerator3
    Skeletal Mesh=$User.New Skeletal Mesh
    Socket Sampling Mode=::NewEnumerator5
   )
   &AddVelocity(Velocity=<0 0 100>)
  ]
  particleUpdateScript [
   &ParticleState()
   &SolveForcesAndVelocity(Clamp Velocity=false, Limit Acceleration=false)
  ]
  rendererClasses [&NiagaraSpriteRendererProperties]
  renderers [{rendererClass=&NiagaraSpriteRendererProperties}]
 }
}
summary {
 systemName=NS_Tire_Skid_Trails
 emitters [
  {emitterName=SkidMarks, rendererClasses [&NiagaraRibbonRendererProperties]}
  {bEnabled=false, emitterName=WheelTest, rendererClasses [&NiagaraSpriteRendererProperties]}
 ]
 userVariables [
  User.New Skeletal Mesh {
   type=NiagaraDataInterfaceSkeletalMesh
   defaultValue {
    struct=NiagaraExt_VariableValue_DataInterface
    value {dataInterface=&self:NiagaraDataInterfaceSkeletalMesh_0}
   }
  }
  User.Wheel_Radius_Array {
   type=NiagaraDataInterfaceArrayFloat
   defaultValue {
    struct=NiagaraExt_VariableValue_DataInterface
    value {
     dataInterface=&self:NiagaraDataInterfaceArrayFloat_1
     dataInterfaceClass=&NiagaraDataInterfaceArrayFloat
    }
   }
  }
  User.Wheel_Width_Array {
   type=NiagaraDataInterfaceArrayFloat
   defaultValue {
    struct=NiagaraExt_VariableValue_DataInterface
    value {
     dataInterface=&self:NiagaraDataInterfaceArrayFloat_2
     dataInterfaceClass=&NiagaraDataInterfaceArrayFloat
    }
   }
  }
  User.WheelsTouchingArray {
   type=NiagaraDataInterfaceArrayFloat
   defaultValue {
    struct=NiagaraExt_VariableValue_DataInterface
    value {
     dataInterface=&self:NiagaraDataInterfaceArrayFloat_0
     dataInterfaceClass=&NiagaraDataInterfaceArrayFloat
    }
   }
  }
 ]
}
