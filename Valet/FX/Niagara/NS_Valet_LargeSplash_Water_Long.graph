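# Niagara system for a large, long water splash on a tired vehicle, with an initial burst plus ongoing bubbles, foam and wheel spray.
# System ints from a bitwise module, apparently decoding User.WaterSettingBit, multiply spawn counts to gate each effect.
# Mesh splash, bubble and foam emitters snap to User.Water, with foam also spawning by distance moved.
# Spawn groups vary sprite droplets, while GPU tire spray samples a skeletal mesh, aimed by User.TireSpinWaterDirection.
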NiagaraSystem /Valet/FX/Niagara/NS_Valet_LargeSplash_Water_Long
dependencies {
 usedDataInterfaces [
  &NiagaraDataInterfaceCurve
  &NiagaraDataInterfaceVector2DCurve
  &NiagaraDataInterfaceVectorCurve
  &NiagaraDataInterfaceSpriteRendererInfo
  &NiagaraDataInterfaceSkeletalMesh
 ]
 usedDynamicInputs []
 usedModules [
  &self:SystemSpawnScript.SystemScriptSource.SystemScriptGraph.NiagaraNodeAssignment_0.SetVariables_#1
  &self:SetSystemIntsfromBitwise
  &SystemState
 ]
}
emitters {
 Bubbles {
  emitterUpdateScript [
   &EmitterState()
   &SpawnBurst_Instantaneous(
    Spawn Count=&Multiply_Int(A=$System.BigBubbleSpawn, B=12)
    Spawn Group=1
    Use Spawn Probability=false
    UseLoopCountLimit=true
   )
   &SpawnBurst_Instantaneous(
    Spawn Count=&Multiply_Int(A=$System.BigBubbleSpawn, B=6)
    Spawn Group=1
    Spawn Time=0.15
    Use Spawn Probability=false
    UseLoopCountLimit=true
   ) {module=SpawnBurst_Instantaneous001}
   &SpawnRate(
    SpawnRate=&Multiply_Float(A=30, B=&MakeFloatFromInt(Int=$System.LoopFoamSpawn))
    Use Spawn Probability=false
   )
  ]
  particleSpawnScript [
   /SetVariables(Particles.BubbleSize:NiagaraFloat=&FloatFromCurve(
     FloatCurve=DataInterface() {propertyValues {Curve {keys [{arriveTangent=-9.9662, interpMode=RCIM_Cubic, leaveTangent=-9.9662, tangentMode=RCTM_User, time=-0.0046, value=8}, <at 1.0011 4 RCIM_Cubic>]}}}
     CurveIndex=&RandomRangeFloat(Minimum=0, Maximum=1)
     Scale Curve=&MaskFloatBySpawnGroup(Masked Float=1, Passthrough Float=2.2)
    )) {isSetParameters=true}
   &InitializeParticle(
    Lifetime=&MaskFloatBySpawnGroup(
     Masked Float=&RandomRangeFloat(Minimum=1.5, Maximum=3)
     Passthrough Float=&RandomRangeFloat(Minimum=1, Maximum=2.5)
    )
    Mesh Scale Mode=::NewEnumerator3
    UsePositionOffset=false
    Write Material Random=true
   )
   /Game/Effects/Niagara/Modules/Water/WaterBodyLocation(Snap Position to Water Position=true, Use Water Plane Position for Water Position=true, WaterBodyActor:NiagaraDataInterfaceWater=$User.Water)
   &ShapeLocation(
    Cylinder Height=150
    Cylinder Height Midpoint=1
    Cylinder Radius=199
    MaskSpawnGroup=true
    Offset=<0 0 -25>
    Offset Coordinate Space=::Local
    Offset Mode=::NewEnumerator0
    Override Seed=false
    Shape Primitive=::NewEnumerator1
    Surface Only=false
    UseOwnerScale=true
   )
   &ShapeLocation(
    Cylinder Height=120
    Cylinder Height Midpoint=1
    Cylinder Radius=80
    MaskSpawnGroup=true
    Offset=<0 0 -30>
    Offset Coordinate Space=::Local
    Offset Mode=::NewEnumerator0
    Override Seed=false
    Shape Primitive=::NewEnumerator1
    Spawn Group Mask=1
    Surface Only=false
    UseOwnerScale=true
   ) {module=ShapeLocation001}
   &InitialMeshOrientation(Use Rotation Vector=false)
   &InheritVelocity(Inherited Velocity Amount Scale=<0.2 0.2 0>, Speed Limit=false, UseSpeedThreshold=false)
  ]
  particleUpdateScript [
   &SolveForcesAndVelocity(Clamp Velocity=false, Limit Acceleration=false)
   &ParticleState()
   &DynamicMaterialParameters(
    Index 0 Param 1=&MaskFloatBySpawnGroup(
     Masked Float=&FloatFromCurve(FloatCurve=DataInterface() {propertyValues {Curve {keys [
          {arriveTangent=-1.2623, interpMode=RCIM_Cubic, leaveTangent=-1.2623, tangentMode=RCTM_User, time=-0.0046, value=1.451}
          {arriveTangent=-0.5317, interpMode=RCIM_Cubic, leaveTangent=-0.5317, tangentMode=RCTM_User, time=1.0011, value=0.002}
         ]}}})
     Passthrough Float=&FloatFromCurve(FloatCurve=DataInterface() {propertyValues {Curve {keys [
          {arriveTangent=-1.2623, interpMode=RCIM_Cubic, leaveTangent=-1.2623, tangentMode=RCTM_User, time=-0.0046, value=0.5}
          {arriveTangent=-0.5317, interpMode=RCIM_Cubic, leaveTangent=-0.5317, tangentMode=RCTM_User, time=1.0011, value=0.002}
         ]}}})
    )
    Index 0 Param 2=&MaskFloatBySpawnGroup(
     Masked Float=&FloatFromCurve(FloatCurve=DataInterface() {propertyValues {Curve {keys [<at -0.0046 9.311>,<at 1.0011 8.8611>]}}})
     Passthrough Float=&FloatFromCurve(FloatCurve=DataInterface() {propertyValues {Curve {keys [<at -0.0046 9.311>,<at 1.0011 8.8611>]}}}, Scale Curve=2)
    )
    Index 0 Param 3=&FloatFromCurve(FloatCurve=DataInterface() {propertyValues {Curve {keys [<at -0.0046 48.1725>,<at 1.0048 94.7831>]}}})
    Index 0 Param 4=$Particles.BubbleSize
    Module.Param0WriteEnabled=true
    Module.Param1WriteEnabled=true
    Module.Param2WriteEnabled=true
    Module.Param3WriteEnabled=true
   )
   &ScaleColor(
    Scale Alpha=&FloatFromCurve(FloatCurve=DataInterface() {propertyValues {Curve {keys [<at 0 -1>,<at 0.05 0>,<at 0.7904 0.0132>,<at 1.0011 -0.9961>]}}})
    ScaleA=true
    ScaleRGB=true
   )
  ]
  rendererClasses [&NiagaraMeshRendererProperties]
  renderers [{rendererClass=&NiagaraMeshRendererProperties, rendererIndex=0}]
 }
 Initial_Splash {
  emitterUpdateScript [
   &EmitterState(Life Cycle Mode=::NewEnumerator1, Loop Behavior=::NewEnumerator1, Loop Duration=1, UseLoopDelay=false)
   &SpawnBurst_Instantaneous(
    Spawn Count=&Multiply_Int(A=$System.SplashSpawn, B=1)
    Spawn Time=0.12
    Use Spawn Probability=false
    UseLoopCountLimit=false
   )
   &SpawnBurst_Instantaneous(
    Spawn Count=&Multiply_Int(A=$System.SplashSpawn, B=1)
    Spawn Time=0.25
    Use Spawn Probability=false
    UseLoopCountLimit=false
   ) {module=SpawnBurst_Instantaneous001}
   &SpawnBurst_Instantaneous(Spawn Count=$System.SplashSpawn, Spawn Group=1, Spawn Time=0.1, Use Spawn Probability=false, UseLoopCountLimit=false) {module=SpawnBurst_Instantaneous002}
  ]
  particleSpawnScript [
   /SetVariables(Particles.VisibilityTag=&SpawnGroup()) {isSetParameters=true}
   &InitializeParticle(
    Lifetime=&MaskFloatBySpawnGroup(
     Masked Float=&RandomRangeFloat(Minimum=1.5, Maximum=2.2)
     Passthrough Float=&RandomRangeFloat(Minimum=2, Maximum=3)
    )
    Mesh Scale Mode=::NewEnumerator3
    Mesh Uniform Scale=&MaskFloatBySpawnGroup(Masked Float=&RandomRangeFloat(Minimum=0.26, Maximum=0.29), Passthrough Float=100)
    UsePositionOffset=false
    Write Material Random=true
   )
   &WaterBodyLocation(Snap Position to Water Position=true, Use Water Plane Position for Water Position=true, WaterBodyActor:NiagaraDataInterfaceWater=$User.Water)
   &InitialMeshOrientation(
    Mesh Orientation Mode=::NewEnumerator2
    Orientation Vector=&Multiply_Vector(A=$Engine.Owner.SystemXAxis, B=<1 1 0>)
    Rotation=&RandomRangeVector(Minimum=<0 0 -0.05>, Maximum=<0 0 0.05>)
    Use Rotation Vector=true
   )
   &SystemLocation(
    MaskSpawnGroup=false
    Offset=&MakeVector(
     X=&FloatFromCurve(FloatCurve=DataInterface() {propertyValues {Curve {keys [<at 0 3>,<at 1 -5>]}}}, CurveIndex=$Emitter.Age, Scale Curve=250)
     Y=0
     Z=0
    )
   )
  ]
  particleUpdateScript [
   &ParticleState()
   &ScaleMeshSize(Scale Factor=&MaskVectorBySpawnGroup(
     Masked Vector=&VectorFromCurve(VectorCurve=DataInterface() {propertyValues {
        XCurve {keys [<at -0.0011 1.3>,<at 0.9916 2.4565>]}
        YCurve {keys [<at -0.0011 1.3>,<at 0.9883 2.4748>]}
        ZCurve {keys [<at -0.0011 1.1211>,<at 1.0048 2.9133>]}
       }})
     Passthrough Vector=&VectorFromFloat(Value=&FloatFromCurve(FloatCurve=DataInterface() {propertyValues {Curve {keys [
           {arriveTangent=26.4511, interpMode=RCIM_Cubic, leaveTangent=26.451, tangentMode=RCTM_User, time=-0.0046, value=0.6073}
           {arriveTangent=5.8301, interpMode=RCIM_Cubic, leaveTangent=5.8301, time=0.1863, value=3.3239}
           <at 1.0011 6.471 RCIM_Cubic>
          ]}}}))
    ))
   &DynamicMaterialParameters(
    Index 0 Param 1=&MaskFloatBySpawnGroup(
     Masked Float=&FloatFromCurve(FloatCurve=DataInterface() {propertyValues {Curve {keys [
          {arriveTangent=3.8469, interpMode=RCIM_Cubic, leaveTangent=3.8469, tangentMode=RCTM_User, time=-0.0005, value=-1.845}
          {arriveTangent=1.4047, interpMode=RCIM_Cubic, leaveTangent=1.4047, tangentMode=RCTM_User, time=0.2513, value=-1.0104}
          {arriveTangent=1.3476, interpMode=RCIM_Cubic, leaveTangent=1.3476, time=0.8054, value=-0.4692}
          {arriveTangent=3.8161, interpMode=RCIM_Cubic, leaveTangent=3.8161, tangentMode=RCTM_User, time=1.0011, value=0}
         ]}}})
     Passthrough Float=&FloatFromCurve(FloatCurve=DataInterface() {propertyValues {Curve {keys [<at -0.0046 1>,<at 1.0011 3>]}}})
    )
    Index 0 Param 2=&MaskFloatBySpawnGroup(
     Masked Float=&FloatFromCurve(FloatCurve=DataInterface() {propertyValues {Curve {keys [<at -0.0046 4>,<at 1.0011 0>]}}}, Scale Curve=2)
     Passthrough Float=&FloatFromCurve(FloatCurve=DataInterface() {propertyValues {Curve {keys [
          {arriveTangent=1.2087, interpMode=RCIM_Cubic, leaveTangent=1.2087, tangentMode=RCTM_User, time=-0.002, value=0.3756}
          <at 1.0011 1.4657 RCIM_Cubic>
         ]}}})
    )
    Index 0 Param 3=&MaskFloatBySpawnGroup(
     Masked Float=&FloatFromCurve(
      FloatCurve=DataInterface() {propertyValues {Curve {keys [
          {arriveTangent=0.096, interpMode=RCIM_Cubic, leaveTangent=0.096, tangentMode=RCTM_User, time=-0.0046, value=0.0882}
          <at 1.0011 0.1372 RCIM_Cubic>
         ]}}}
      Scale Curve=4
     )
     Passthrough Float=&FloatFromCurve(FloatCurve=DataInterface() {propertyValues {Curve {keys [<at -0.0046 1>,<at 1.0011 0.5>]}}})
    )
    Index 0 Param 4=&MaskFloatBySpawnGroup(
     Masked Float=&FloatFromCurve(
      FloatCurve=DataInterface() {propertyValues {Curve {keys [
          {arriveTangent=1.9125, interpMode=RCIM_Cubic, leaveTangent=0.2728, tangentMode=RCTM_User, time=-0.0046, value=0.9276}
          {arriveTangent=4.6326, interpMode=RCIM_Cubic, leaveTangent=4.6326, tangentMode=RCTM_User, time=1.0011, value=2.7648}
         ]}}}
      Scale Curve=0.6
     )
     Passthrough Float=0
    )
    Index 1 Param 3=200
    Index 2 Param 1=&FloatFromCurve(FloatCurve=DataInterface() {propertyValues {Curve {keys [<at -0.0046 1.4697>,<at 0.369 1.481>,<at 1.0011 2>]}}})
    Index 2 Param 2=&FloatFromCurve(FloatCurve=DataInterface() {propertyValues {Curve {keys [<at -0.0046 0.1>,<at 0.1026 4.2375>,<at 1.0011 0.3>]}}})
    Index 2 Param 3=&FloatFromCurve(FloatCurve=DataInterface() {propertyValues {Curve {keys [
         {arriveTangent=-2.2186, interpMode=RCIM_Cubic, leaveTangent=-2.2186, tangentMode=RCTM_User, time=-0.0046, value=1.0113}
         <at 1.0011 0.002 RCIM_Cubic>
        ]}}})
    Index 2 Param 4=&FloatFromCurve(FloatCurve=DataInterface() {propertyValues {Curve {keys [
         {arriveTangent=-1.9991, interpMode=RCIM_Cubic, time=0, value=1}
         {arriveTangent=-1.3497, interpMode=RCIM_Cubic, leaveTangent=-1.3497, tangentMode=RCTM_User, time=0.1761, value=0.423}
         <at 1 0 RCIM_Cubic>
        ]}}})
    Module.Index0Param0Write=true
    Module.Index0Param1Write=true
    Module.Index0Param2Write=true
    Module.Index0Param3Write=true
    Module.Index1Param0Write=false
    Module.Index1Param1Write=false
    Module.Index1Param2Write=true
    Module.Index1Param3Write=false
    Module.Index2Param0Write=true
    Module.Index2Param1Write=true
    Module.Index2Param2Write=true
    Module.Index2Param3Write=true
    Write Parameter Index 1=true
    Write Parameter Index 2=true
   )
  ]
  rendererClasses [&NiagaraMeshRendererProperties]
  renderers [
   {rendererClass=&NiagaraMeshRendererProperties, rendererIndex=0}
   {rendererClass=&NiagaraMeshRendererProperties, rendererIndex=1}
  ]
 }
 Loop_Foam {
  emitterUpdateScript [
   &EmitterState()
   &SpawnRate(
    SpawnRate=&Multiply_Float(A=6, B=&MakeFloatFromInt(Int=$System.LoopFoamSpawn))
    Use Spawn Probability=false
   )
   &SpawnPerUnit(
    Module.Use Max Movement Threshold=true
    Module.Use Movement Tolerance=true
    Movement Tolerance=5
    Spawn Spacing:NiagaraFloat=&Multiply_Float(A=300, B=&MakeFloatFromInt(Int=$System.LoopFoamSpawn))
    Use Spawn Probability=false
   ) {module=SpawnPerUnit001}
   &SpawnRate(
    SpawnGroup=1
    SpawnRate=&Multiply_Float(A=2, B=&MakeFloatFromInt(Int=$System.LoopFoamSpawn))
    Use Spawn Probability=false
   ) {module=SpawnRate001}
   &SpawnPerUnit(
    Module.Use Max Movement Threshold=true
    Module.Use Movement Tolerance=false
    Movement Tolerance=5
    Spawn Group=1
    Spawn Spacing:NiagaraFloat=&Multiply_Float(A=220, B=&MakeFloatFromInt(Int=$System.LoopFoamSpawn))
    Use Spawn Probability=false
   ) {module=SpawnPerUnit002}
   &SpawnPerUnit(
    Module.Use Max Movement Threshold=true
    Module.Use Movement Tolerance=true
    Movement Tolerance=5
    Spawn Group=2
    Spawn Spacing:NiagaraFloat=&Multiply_Float(A=220, B=&MakeFloatFromInt(Int=$System.LoopFoamSpawn))
    Use Spawn Probability=false
   )
   &SpawnRate(
    SpawnGroup=3
    SpawnRate=&Multiply_Float(A=50, B=&MakeFloatFromInt(Int=$System.LoopFoamSpawn))
    Use Spawn Probability=false
   ) {module=SpawnRate002}
  ]
  particleSpawnScript [
   /SetVariables(Particles.VisibilityTag=&SpawnGroup()) {isSetParameters=true}
   &InitializeParticle(
    Lifetime Max=3
    Lifetime Min=2
    Lifetime Mode=::NewEnumerator1
    Mesh Scale Mode=::NewEnumerator3
    Mesh Uniform Scale=&MaskFloatBySpawnGroup(Masked Float=100, Passthrough Float=500)
    Sprite Rotation Mode=::NewEnumerator2
    Sprite Size Mode=::NewEnumerator3
    Uniform Sprite Size=&FloatFromCurve(
     FloatCurve=DataInterface() {propertyValues {Curve {keys [
         {arriveTangent=-11.7494, interpMode=RCIM_Cubic, leaveTangent=-11.7494, tangentMode=RCTM_User, time=-0.0046, value=7.0507}
         {arriveTangent=-6.0161, interpMode=RCIM_Cubic, leaveTangent=-6.0161, time=0.8437, value=2.2941}
         <at 1.0011 1 RCIM_Cubic>
        ]}}}
     CurveIndex=&RandomRangeFloat(Minimum=0, Maximum=1)
    )
    UsePositionOffset=false
    Write Material Random=true
   )
   &WaterBodyLocation(Snap Position to Water Position=true, Use Water Plane Position for Water Position=true, WaterBodyActor:NiagaraDataInterfaceWater=$User.Water)
   &ShapeLocation(
    Cylinder Height=0
    Cylinder Height Midpoint=0
    Cylinder Radius=400
    MaskSpawnGroup=true
    Override Seed=false
    Shape Primitive=::NewEnumerator1
    Surface Only=false
    UseOwnerScale=true
   )
   &ShapeLocation(
    Cylinder Height=0
    Cylinder Height Midpoint=0
    Cylinder Radius=&RandomRangeFloat(Minimum=300, Maximum=600)
    MaskSpawnGroup=true
    Override Seed=false
    Shape Primitive=::NewEnumerator1
    Spawn Group Mask=3
    Surface Only=false
    UseOwnerScale=true
   ) {module=ShapeLocation001}
   &ShapeLocation(
    Box Size=<900 0 0>
    MaskSpawnGroup=true
    Override Seed=false
    Shape Primitive=::NewEnumerator2
    Spawn Group Mask=1
    Surface Only=false
    UseOwnerScale=true
   ) {module=ShapeLocation002}
   &ShapeLocation(
    Box Size=<900 0 0>
    MaskSpawnGroup=true
    Override Seed=false
    Shape Primitive=::NewEnumerator2
    Spawn Group Mask=2
    Surface Only=false
    UseOwnerScale=true
   ) {module=ShapeLocation003}
   &InheritVelocity(
    Inherited Velocity Amount Scale=&MaskVectorBySpawnGroup(
     Masked Vector=<0.8 0.8 0.8>
     Passthrough Vector=&MaskVectorBySpawnGroup(Masked Vector=<1 1 0>, Passthrough Vector=<0.5 0.5 0>, Spawn Group Mask=3)
    )
    Speed Limit=true
    UseSpeedThreshold=false
   )
   &InitialMeshOrientation(
    Mesh Orientation Mode=::NewEnumerator2
    Orientation Vector=<0 0 0>
    Rotation=&RandomRangeVector(Minimum=<0 0 0>, Maximum=<0 0 1>)
    Use Rotation Vector=true
   )
   &SpriteFacingAndAlignment(Sprite Facing=<0 0 1>, WriteAlignment=false, WriteFacing=true)
  ]
  particleUpdateScript [
   &Drag(
    Drag=&MaskFloatBySpawnGroup(Masked Float=1.5, Passthrough Float=2, Spawn Group Mask=1)
    Use Linear Drag=true
    Use Rotational Drag=false
   )
   &PointForce(
    Force Origin=$Particles.Water Body Data.WaterPlanePosition
    Force Strength=&MaskFloatBySpawnGroup(Masked Float=50, Passthrough Float=0, Spawn Group Mask=3)
    Module.Use Falloff=false
    Module.UseLinearForce=false
   )
   &CurlNoiseForce(
    Noise Frequency=1
    Noise Strength=&MaskFloatBySpawnGroup(
     Masked Float=&FloatFromCurve(FloatCurve=DataInterface() {propertyValues {Curve {keys [
          {arriveTangent=-1159.6794, interpMode=RCIM_Cubic, leaveTangent=-1159.6792, tangentMode=RCTM_User, time=-0.0046, value=200}
          {arriveTangent=-174, interpMode=RCIM_Cubic, leaveTangent=-174, time=0.2223, value=70}
          <at 1.0011 25 RCIM_Cubic>
         ]}}})
     Passthrough Float=0
     Spawn Group Mask=3
    )
    Pan Noise=false
   )
   &SolveForcesAndVelocity(Clamp Velocity=false, Limit Acceleration=false)
   &ParticleState()
   &ScaleSpriteSize(Uniform Curve Sprite Scale=DataInterface() {propertyValues {Curve {keys [<at 0.6958 1.0113 RCIM_Cubic>,<at 1.0011 0.002 RCIM_Cubic>]}}})
   &ScaleMeshSize(Scale Factor=&MaskVectorBySpawnGroup(
     Masked Vector=&VectorFromFloat(Value=&FloatFromCurve(FloatCurve=DataInterface() {propertyValues {Curve {keys [<at -0.0046 0.5>,<at 1.0011 1.5>]}}}))
     Passthrough Vector=<1 1 1>
     Spawn Group Mask=2
    ))
   &DynamicMaterialParameters(
    Index 0 Param 1=&MaskFloatBySpawnGroup(
     Masked Float=&FloatFromCurve(FloatCurve=DataInterface() {propertyValues {Curve {keys [<at -0.0046 1>,<at 1.0011 4>]}}})
     Passthrough Float=&FloatFromCurve(FloatCurve=DataInterface() {propertyValues {Curve {keys [<at -0.0046 1>,<at 1.0011 3>]}}})
    )
    Index 0 Param 2=&MaskFloatBySpawnGroup(
     Masked Float=&FloatFromCurve(FloatCurve=DataInterface() {propertyValues {Curve {keys [<at -0.0046 0.2>,<at 1.0011 1.3>]}}})
     Passthrough Float=&FloatFromCurve(FloatCurve=DataInterface() {propertyValues {Curve {keys [
          {arriveTangent=1.2087, interpMode=RCIM_Cubic, leaveTangent=1.2087, tangentMode=RCTM_User, time=-0.002, value=0.3756}
          <at 1.0011 1.4657 RCIM_Cubic>
         ]}}})
    )
    Index 0 Param 3=&MaskFloatBySpawnGroup(
     Masked Float=&FloatFromCurve(FloatCurve=DataInterface() {propertyValues {Curve {keys [<at -0.0046 1>,<at 1.0011 0.5>]}}})
     Passthrough Float=&FloatFromCurve(FloatCurve=DataInterface() {propertyValues {Curve {keys [<at -0.0046 1>,<at 1.0011 0.5>]}}})
    )
    Index 1 Param 3=200
    Index 2 Param 1=&MaskFloatBySpawnGroup(
     Masked Float=&FloatFromCurve(FloatCurve=DataInterface() {propertyValues {Curve {keys [<at -0.0046 0.4>,<at 1.0011 1.5>]}}})
     Passthrough Float=&FloatFromCurve(FloatCurve=DataInterface() {propertyValues {Curve {keys [<at -0.0046 1.4697>,<at 0.369 1.481>,<at 1.0011 2>]}}})
    )
    Index 2 Param 2=&MaskFloatBySpawnGroup(
     Masked Float=&FloatFromCurve(FloatCurve=DataInterface() {propertyValues {Curve {keys [<at -0.0046 0.1>,<at 0.2854 2.6>,<at 1.0011 0.3>]}}})
     Passthrough Float=&FloatFromCurve(FloatCurve=DataInterface() {propertyValues {Curve {keys [<at -0.0046 0.1>,<at 0.1026 4.2375>,<at 1.0011 0.3>]}}})
    )
    Index 2 Param 3=&MaskFloatBySpawnGroup(
     Masked Float=0.8
     Passthrough Float=&FloatFromCurve(FloatCurve=DataInterface() {propertyValues {Curve {keys [
          {arriveTangent=-2.2186, interpMode=RCIM_Cubic, leaveTangent=-2.2186, tangentMode=RCTM_User, time=-0.0046, value=1.0113}
          <at 1.0011 0.002 RCIM_Cubic>
         ]}}})
    )
    Index 2 Param 4=&MaskFloatBySpawnGroup(
     Masked Float=&FloatFromCurve(FloatCurve=DataInterface() {propertyValues {Curve {keys [
          {arriveTangent=-1.9991, interpMode=RCIM_Cubic, time=0, value=1}
          {arriveTangent=-2.7143, interpMode=RCIM_Cubic, leaveTangent=-2.7143, tangentMode=RCTM_User, time=0.2095, value=0.6901}
          <at 1 0 RCIM_Cubic>
         ]}}})
     Passthrough Float=&FloatFromCurve(FloatCurve=DataInterface() {propertyValues {Curve {keys [
          {arriveTangent=-1.9991, interpMode=RCIM_Cubic, time=0, value=1}
          {arriveTangent=-1.3497, interpMode=RCIM_Cubic, leaveTangent=-1.3497, tangentMode=RCTM_User, time=0.1761, value=0.423}
          <at 1 0 RCIM_Cubic>
         ]}}})
    )
    Module.Index0Param0Write=true
    Module.Index0Param1Write=true
    Module.Index0Param2Write=true
    Module.Index0Param3Write=true
    Module.Index1Param0Write=false
    Module.Index1Param1Write=false
    Module.Index1Param2Write=true
    Module.Index1Param3Write=false
    Module.Index2Param0Write=true
    Module.Index2Param1Write=true
    Module.Index2Param2Write=true
    Module.Index2Param3Write=true
    Write Parameter Index 1=true
    Write Parameter Index 2=true
   )
   /SetVariables(Particles.Sprite Dynamic Material Parameter:Vector4f=<1 0 0 0>) {isSetParameters=true}
   &ScaleColor(
    Scale Alpha=&MaskFloatBySpawnGroup(
     Masked Float=&FloatFromCurve(
      FloatCurve=DataInterface() {propertyValues {Curve {keys [
          <at 0 0 RCIM_Cubic>
          {arriveTangent=0.002, interpMode=RCIM_Cubic, leaveTangent=0.002, time=0.1183, value=1.0113}
          <at 1.0011 0.002 RCIM_Cubic>
         ]}}}
      Scale Curve=0.8
     )
     Passthrough Float=1
     Spawn Group Mask=3
    )
    ScaleA=true
    ScaleRGB=true
   )
  ]
  rendererClasses [&NiagaraMeshRendererProperties, &NiagaraSpriteRendererProperties]
  renderers [
   {rendererClass=&NiagaraMeshRendererProperties, rendererIndex=0}
   {rendererClass=&NiagaraMeshRendererProperties, rendererIndex=1}
   {rendererClass=&NiagaraMeshRendererProperties, rendererIndex=2}
   {rendererClass=&NiagaraSpriteRendererProperties, rendererIndex=3}
  ]
 }
 Splash_Elements {
  emitterUpdateScript [
   &EmitterState(Life Cycle Mode=::NewEnumerator1, Loop Behavior=::NewEnumerator1, Loop Duration=1, UseLoopDelay=false)
   &SpawnBurst_Instantaneous(
    Spawn Count=&Multiply_Int(A=$System.SplashSpawn, B=500)
    Spawn Time=0.1
    Use Spawn Probability=false
    UseLoopCountLimit=false
   )
   &SpawnBurst_Instantaneous(
    Spawn Count=&Multiply_Int(A=$System.SplashSpawn, B=300)
    Spawn Time=0.25
    Use Spawn Probability=false
    UseLoopCountLimit=false
   ) {module=SpawnBurst_Instantaneous001}
   &SpawnBurst_Instantaneous(
    Spawn Count=&Multiply_Int(A=$System.SplashSpawn, B=150)
    Spawn Time=0.35
    Use Spawn Probability=false
    UseLoopCountLimit=false
   ) {module=SpawnBurst_Instantaneous002}
   &SpawnBurst_Instantaneous(
    Spawn Count=&Multiply_Int(A=$System.SplashSpawn, B=5)
    Spawn Group=1
    Spawn Time=0.2
    Use Spawn Probability=false
    UseLoopCountLimit=false
   ) {module=SpawnBurst_Instantaneous003}
   &SpawnBurst_Instantaneous(
    Spawn Count=&Multiply_Int(A=$System.SplashSpawn, B=12)
    Spawn Group=2
    Spawn Time=0.1
    Use Spawn Probability=false
    UseLoopCountLimit=false
   ) {module=SpawnBurst_Instantaneous004}
   &SpawnBurst_Instantaneous(
    Spawn Count=&Multiply_Int(A=$System.SplashSpawn, B=8)
    Spawn Group=2
    Spawn Time=0.2
    Use Spawn Probability=false
    UseLoopCountLimit=false
   ) {module=SpawnBurst_Instantaneous005}
   &SpawnBurst_Instantaneous(
    Spawn Count=&Multiply_Int(A=$System.SplashSpawn, B=6)
    Spawn Group=2
    Spawn Time=0.3
    Use Spawn Probability=false
    UseLoopCountLimit=false
   ) {module=SpawnBurst_Instantaneous006}
  ]
  particleSpawnScript [
   /SetVariables(Particles.VisibilityTag=&SpawnGroup()) {isSetParameters=true}
   &InitializeParticle(
    Color=&MaskLinearColorBySpawnGroup(
     Masked Linear Color=<rgba 1 1 1 1>
     Passthrough Linear Color=&MaskLinearColorBySpawnGroup(Masked Linear Color=<rgba 1.0154 1.3322 1.5 0.3>, Passthrough Linear Color=<rgba 0.1082 0.663 1 1>, Spawn Group Mask=1)
    )
    Lifetime=&MaskFloatBySpawnGroup(
     Masked Float=&RandomRangeFloat(Minimum=1.5, Maximum=3)
     Passthrough Float=&MaskFloatBySpawnGroup(
      Masked Float=&RandomRangeFloat(Minimum=1, Maximum=1.5)
      Passthrough Float=&RandomRangeFloat(Minimum=1, Maximum=1.6)
      Spawn Group Mask=1
     )
    )
    Sprite Rotation Angle=&MaskFloatBySpawnGroup(Masked Float=0, Passthrough Float=&RandomRangeFloat(Minimum=0, Maximum=1))
    Sprite Rotation Mode=::NewEnumerator1
    Sprite Size=&MaskVector2dBySpawnGroup(
     Masked Vector2D=&Vector2DFromCurve(
      Vector2Curve=DataInterface() {propertyValues {
        XCurve {keys [<at -0.003 3.8449 RCIM_Cubic>,<at 0.9988 0.3489 RCIM_Cubic>]}
        YCurve {keys [
          {arriveTangent=-30.8979, interpMode=RCIM_Cubic, leaveTangent=-30.8979, tangentMode=RCTM_User, time=0.001, value=13.7761}
          {arriveTangent=-11.8685, interpMode=RCIM_Cubic, leaveTangent=-11.8685, time=0.2536, value=8.9484}
          <at 0.9988 1.9328 RCIM_Cubic>
         ]}
       }}
      CurveIndex=&RandomRangeFloat(Minimum=0, Maximum=1)
      ScaleCurve=<1 1>
     )
     Passthrough Vector2D=&MaskVector2dBySpawnGroup(
      Masked Vector2D=&Vector2DFromFloat(Value=&RandomRangeFloat(Minimum=180, Maximum=220))
      Passthrough Vector2D=&Vector2DFromFloat(Value=&RandomRangeFloat(Minimum=120, Maximum=170))
      Spawn Group Mask=1
     )
     Spawn Group Mask=0
    )
    Sprite Size Mode=::NewEnumerator4
    UsePositionOffset=false
    Write Material Random=true
   )
   &ShapeLocation(
    Cylinder Height=1
    Cylinder Radius=150
    MaskSpawnGroup=true
    Override Seed=false
    Shape Primitive=::NewEnumerator1
    Spawn Group Mask=1
    Surface Only=true
    Surface Only Band Thickness=2
    UseOwnerScale=true
   )
   &ShapeLocation(
    Cylinder Height=50
    MaskSpawnGroup=true
    Override Seed=false
    Shape Primitive=::NewEnumerator1
    Spawn Group Mask=1
    Surface Only=false
    UseOwnerScale=true
   ) {module=ShapeLocation001}
   &ShapeLocation(
    Cylinder Height=50
    MaskSpawnGroup=true
    Override Seed=false
    Shape Primitive=::NewEnumerator1
    Spawn Group Mask=2
    Surface Only=false
    UseOwnerScale=true
   ) {module=ShapeLocation002}
   &AddVelocity(Velocity=&MaskVectorBySpawnGroup(
     Masked Vector=&RandomRangeVector(Minimum=<0 -300 300>, Maximum=<1200 300 1200>)
     Passthrough Vector=<0 0 0>
    ))
   &AddVelocity(
    Cone Angle=&MaskFloatBySpawnGroup(Masked Float=50, Passthrough Float=30, Spawn Group Mask=2)
    Cone Axis=<0.8 0 1>
    Module.UseConeFalloff=false
    Module.UseVelDistribution=false
    Override Seed=false
    Velocity Mode=::NewEnumerator2
    Velocity Speed=&MaskFloatBySpawnGroup(
     Masked Float=0
     Passthrough Float=&MaskFloatBySpawnGroup(
      Masked Float=&RandomRangeFloat(Minimum=300, Maximum=950)
      Passthrough Float=&RandomRangeFloat(
       Minimum=50
       Maximum=&FloatFromCurve(FloatCurve=DataInterface() {propertyValues {Curve {keys [<at 0 1>,<at 0.1 2.3>]}}}, CurveIndex=$Emitter.NormalizedLoopAge, Scale Curve=1200)
      )
      Spawn Group Mask=1
     )
    )
   ) {module=AddVelocity001}
   &InheritVelocity(
    Inherited Velocity Amount Scale=&MaskVectorBySpawnGroup(
     Masked Vector=<2 2 -1>
     Passthrough Vector=&MaskVectorBySpawnGroup(Masked Vector=<1 1 0>, Passthrough Vector=<0 0 0>, Spawn Group Mask=1)
    )
    Inherited Velocity Speed Limit=500
    Speed Limit=true
    UseSpeedThreshold=false
   )
  ]
  particleUpdateScript [
   &ParticleState()
   &GravityForce(Gravity=&MaskVectorBySpawnGroup(
     Masked Vector=<0 0 -980>
     Passthrough Vector=&MaskVectorBySpawnGroup(
      Masked Vector=<0 0 -490>
      Passthrough Vector=&RandomRangeVector(Minimum=<0 0 -1200>, Maximum=<0 0 -1400>)
      Spawn Group Mask=1
     )
    ))
   &Drag(
    Drag=&MaskFloatBySpawnGroup(
     Masked Float=0.8
     Passthrough Float=&MaskFloatBySpawnGroup(Masked Float=1, Passthrough Float=1.5, Spawn Group Mask=1)
    )
    Use Linear Drag=true
    Use Rotational Drag=false
   )
   &PointForce(
    Force Strength=&MaskFloatBySpawnGroup(Masked Float=200, Passthrough Float=0)
    Module.Use Falloff=false
    Module.UseLinearForce=false
   )
   &CurlNoiseForce(
    Noise Frequency=1
    Noise Strength=&MaskFloatBySpawnGroup(
     Masked Float=&FloatFromCurve(FloatCurve=DataInterface() {propertyValues {Curve {keys [
          {arriveTangent=-1159.6794, interpMode=RCIM_Cubic, leaveTangent=-1159.6792, tangentMode=RCTM_User, time=-0.0046, value=500}
          {arriveTangent=-447.4286, interpMode=RCIM_Cubic, leaveTangent=-447.4286, time=0.2223, value=189.6793}
          <at 1.0011 50 RCIM_Cubic>
         ]}}})
     Passthrough Float=0
    )
    Pan Noise=false
   )
   &SolveForcesAndVelocity(Clamp Velocity=false, Limit Acceleration=false)
   &DynamicMaterialParameters(
    Index 0 Param 1=&MaskFloatBySpawnGroup(
     Masked Float=&RandomRangeFloat(Minimum=4, Maximum=10)
     Passthrough Float=&MaskFloatBySpawnGroup(
      Masked Float=&RandomRangeFloat(Minimum=1, Maximum=2)
      Passthrough Float=&FloatFromCurve(FloatCurve=DataInterface() {propertyValues {Curve {keys [<at 0.25 1>,<at 1 0>]}}})
      Spawn Group Mask=1
     )
    )
    Index 0 Param 2=4
    Index 0 Param 3=2
    Index 0 Param 4=0.01
    Module.Index0Param0Write=true
    Module.Index0Param1Write=true
    Module.Index0Param2Write=true
    Module.Index0Param3Write=true
   )
   &ScaleColor(
    Scale Alpha=&MaskFloatBySpawnGroup(
     Masked Float=&FloatFromCurve(FloatCurve=DataInterface() {propertyValues {Curve {keys [
          <at 0 0 RCIM_Cubic>
          {arriveTangent=0.002, interpMode=RCIM_Cubic, leaveTangent=0.002, time=0.1183, value=1.0113}
          {arriveTangent=-4.2194, interpMode=RCIM_Cubic, leaveTangent=-4.2194, tangentMode=RCTM_User, time=1.0011, value=0.002}
         ]}}})
     Passthrough Float=&MaskFloatBySpawnGroup(
      Masked Float=&FloatFromCurve(FloatCurve=DataInterface() {propertyValues {Curve {keys [<at 0 0 RCIM_Cubic>, {arriveTangent=0.002, interpMode=RCIM_Cubic, leaveTangent=0.002, time=0.1942, value=2}, <at 1.0011 0.002 RCIM_Cubic>]}}})
      Passthrough Float=&FloatFromCurve(FloatCurve=DataInterface() {propertyValues {Curve {keys [<at -0.0046 1.0113 RCIM_Cubic>,<at 1.0011 0.002 RCIM_Cubic>]}}})
      Spawn Group Mask=1
     )
    )
    ScaleA=true
    ScaleRGB=false
   )
   &ScaleSpriteSize(
    Scale Sprite Size Mode=::NewEnumerator0
    Uniform Scale Factor=&MaskFloatBySpawnGroup(
     Masked Float=1
     Passthrough Float=&FloatFromCurve(FloatCurve=DataInterface() {propertyValues {Curve {keys [<at 0 1>,<at 1 2>]}}})
    )
   )
   &SubUVAnimation(
    "Sprite Renderer "=DataInterface() {propertyValues {SpriteRenderer=&self:Splash_Elements_0.NiagaraSpriteRendererProperties_1}}
    UseEndFrame=false
    UseStartFrame=false
   )
  ]
  rendererClasses [&NiagaraSpriteRendererProperties]
  renderers [
   {rendererClass=&NiagaraSpriteRendererProperties}
   {rendererClass=&NiagaraSpriteRendererProperties, rendererIndex=1}
   {rendererClass=&NiagaraSpriteRendererProperties, rendererIndex=2}
  ]
 }
 Tire_Water_Offcast {
  simTarget=GPUComputeSim
  emitterUpdateScript [
   &EmitterState()
   &SpawnRate(
    SpawnRate=&Multiply_Float(
     A=30
     B=&Multiply_Float(
      A=&MakeFloatFromInt(Int=$System.TireSpinWaterSpawn)
      B=&MakeFloatFromInt(Int=$System.LoopFoamSpawn)
     )
    )
    Use Spawn Probability=false
   )
   &SpawnRate(
    SpawnGroup=1
    SpawnRate=&Multiply_Float(
     A=200
     B=&Multiply_Float(
      A=&MakeFloatFromInt(Int=$System.TireSpinWaterSpawn)
      B=&MakeFloatFromInt(Int=$System.LoopFoamSpawn)
     )
    )
    Use Spawn Probability=false
   ) {module=SpawnRate001}
  ]
  particleSpawnScript [
   /SetVariables(Particles.VisibilityTag=&SpawnGroup()) {isSetParameters=true}
   &InitializeParticle(
    Color=<rgba 0.277 0.5637 1 1>
    Lifetime Max=1.1
    Lifetime Min=0.6
    Lifetime Mode=::NewEnumerator1
    Sprite Rotation Angle=&MaskFloatBySpawnGroup(Masked Float=&RandomRangeFloat(Minimum=0, Maximum=360), Passthrough Float=0, Spawn Group Mask=1)
    Sprite Rotation Mode=::NewEnumerator0
    Sprite Size=&MaskVector2dBySpawnGroup(
     Masked Vector2D=&Vector2DFromFloat(Value=&RandomRangeFloat(Minimum=120, Maximum=150))
     Passthrough Vector2D=&Vector2DFromCurve(
      Vector2Curve=DataInterface() {propertyValues {
        XCurve {keys [<at -0.003 3.8449 RCIM_Cubic>,<at 0.9988 0.3489 RCIM_Cubic>]}
        YCurve {keys [
          {arriveTangent=-30.8979, interpMode=RCIM_Cubic, leaveTangent=-30.8979, tangentMode=RCTM_User, time=0.001, value=13.7761}
          {arriveTangent=-11.8685, interpMode=RCIM_Cubic, leaveTangent=-11.8685, time=0.2536, value=8.9484}
          <at 0.9988 1.9328 RCIM_Cubic>
         ]}
       }}
      CurveIndex=&RandomRangeFloat(Minimum=0, Maximum=1)
      ScaleCurve=<0.4 0.3>
     )
     Spawn Group Mask=0
    )
    Sprite Size Mode=::NewEnumerator4
    UsePositionOffset=false
    Write Material Random=true
   )
   &AddVelocityInCone(
    Cone Angle=35
    Cone Axis=&MakeVector(X=$User.TireSpinWaterDirection, Y=0, Z=1)
    Module.Use Velocity Falloff On Cone Axis=false
    Velocity Strength=&UniformRangedFloat(
     Minimum=50
     Maximum=&FloatFromCurve(FloatCurve=DataInterface() {propertyValues {Curve {keys [<at 0 1>,<at 0.1 2.3>]}}}, CurveIndex=$Emitter.NormalizedLoopAge, Scale Curve=300)
    )
   )
   &CameraOffset(Camera Offset Amount=30)
   &SkeletalMeshLocation(
    MaskSpawnGroup=false
    Mesh Sampling Type=::NewEnumerator5
    OverrideRandomSeed=false
    Sampled Position Offset=&MakeVector(X=&Multiply_Float(A=$User.TireSpinWaterDirection, B=25), Y=0, Z=-35)
    Sampled Position Offset Coordinate Space:ENiagaraSkelMeshTransforms=::NewEnumerator2
    Skeletal Mesh=$System.Skeletal Mesh
   )
   &ShapeLocation(MaskSpawnGroup=true, Override Seed=false, Spawn Group Mask=1, Sphere Radius=20, UseOwnerScale=true)
  ]
  particleUpdateScript [
   &ParticleState()
   &GravityForce(Gravity=&MakeVector(X=0, Y=0, Z=&RandomRangeFloat(Minimum=-1500, Maximum=-1900)))
   &Drag(Drag=1.8, Use Linear Drag=true, Use Rotational Drag=false)
   &SolveForcesAndVelocity(Clamp Velocity=false, Limit Acceleration=false)
   &ScaleColor(
    Scale Alpha=&FloatFromCurve(FloatCurve=DataInterface() {propertyValues {Curve {keys [
         <at 0.2106 0.9976 RCIM_Cubic>
         {arriveTangent=-1.2594, interpMode=RCIM_Cubic, leaveTangent=-1.2594, time=0.5039, value=0.1439}
         <at 1.0011 0.002 RCIM_Cubic>
        ]}}})
    ScaleA=true
    ScaleRGB=false
   )
   &ScaleSpriteSize(
    Scale Sprite Size Mode=::NewEnumerator0
    Uniform Scale Factor=&MaskFloatBySpawnGroup(
     Masked Float=&FloatFromCurve(FloatCurve=DataInterface() {propertyValues {Curve {keys [{arriveTangent=1.7416, interpMode=RCIM_Cubic, leaveTangent=1.7416, tangentMode=RCTM_User, time=0, value=0.5}, <at 1 2>]}}})
     Passthrough Float=&FloatFromCurve(FloatCurve=DataInterface() {propertyValues {Curve {keys [
          {arriveTangent=1.7416, interpMode=RCIM_Cubic, leaveTangent=1.7416, tangentMode=RCTM_User, time=0, value=0.5}
          {arriveTangent=-0.5, interpMode=RCIM_Cubic, leaveTangent=-0.5, time=0.3598, value=2.0029}
          <at 1 0>
         ]}}})
    )
   )
  ]
  rendererClasses [&NiagaraSpriteRendererProperties]
  renderers [
   {rendererClass=&NiagaraSpriteRendererProperties}
   {rendererClass=&NiagaraSpriteRendererProperties, rendererIndex=1}
  ]
 }
}
summary {
 systemName=NS_Valet_LargeSplash_Water_Long
 emitters [
  {emitterName=Bubbles, rendererClasses [&NiagaraMeshRendererProperties]}
  {emitterName=Loop_Foam, rendererClasses [&NiagaraMeshRendererProperties, &NiagaraSpriteRendererProperties]}
  {emitterName=Tire_Water_Offcast, simTarget=GPUComputeSim, rendererClasses [&NiagaraSpriteRendererProperties]}
  {emitterName=Initial_Splash, rendererClasses [&NiagaraMeshRendererProperties]}
  {emitterName=Splash_Elements, rendererClasses [&NiagaraSpriteRendererProperties]}
 ]
 userVariables [
  User.TireSpinWaterDirection {type=NiagaraFloat, defaultValue=-1, description=""}
  User.Water {
   type=NiagaraDataInterfaceWater
   description=""
   defaultValue {
    struct=NiagaraExt_VariableValue_DataInterface
    value {
     dataInterface=&self:NiagaraDataInterfaceWater_1
     dataInterfaceClass=/Script/Water.NiagaraDataInterfaceWater
    }
   }
  }
  User.WaterSettingBit {type=NiagaraInt32}
 ]
}
