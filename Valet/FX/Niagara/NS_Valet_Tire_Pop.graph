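# NS_Valet_Tire_Pop is a Niagara system for a tire pop effect in Valet, named for linger smoke, sparks and mesh layers.
# The original five emitters are disabled: two sprite smoke, two FN_Mesh mesh emitters and GPU sprite sparks.
# Five enabled LW_-prefixed counterparts, likely lightweight emitter versions, replace them, though their renderers are not listed here.
# The system exposes no user parameters.
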NiagaraSystem /Valet/FX/Niagara/NS_Valet_Tire_Pop
dependencies=null
emitters {
 LW_FN_Mesh {rendererClasses [], renderers []}
 LW_FN_Mesh002 {rendererClasses [], renderers []}
 LW_LingerSmoke {rendererClasses [], renderers []}
 LW_LingerSmoke001 {rendererClasses [], renderers []}
 LW_Sparks {rendererClasses [], renderers []}
}
summary {
 systemName=NS_Valet_Tire_Pop
 emitters [
  {bEnabled=false, emitterName=LingerSmoke, rendererClasses [&NiagaraSpriteRendererProperties]}
  {bEnabled=false, emitterName=LingerSmoke001, rendererClasses [&NiagaraSpriteRendererProperties]}
  {bEnabled=false, emitterName=FN_Mesh, rendererClasses [&NiagaraMeshRendererProperties]}
  {bEnabled=false, emitterName=Sparks, simTarget=GPUComputeSim, rendererClasses [&NiagaraSpriteRendererProperties]}
  {bEnabled=false, emitterName=FN_Mesh002, rendererClasses [&NiagaraMeshRendererProperties]}
  {emitterName=LW_LingerSmoke, rendererClasses []}
  {emitterName=LW_LingerSmoke001, rendererClasses []}
  {emitterName=LW_FN_Mesh, rendererClasses []}
  {emitterName=LW_Sparks, rendererClasses []}
  {emitterName=LW_FN_Mesh002, rendererClasses []}
 ]
 userVariables []
}
