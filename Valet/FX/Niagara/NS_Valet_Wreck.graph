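# Niagara system NS_Valet_Wreck, by its naming a fire, smoke and ember effect for a wrecked vehicle in Valet.
# Three LW_-prefixed emitters are enabled: fire sprites (SubUV), local embers and world smoke, with no renderers listed for them.
# Their original sprite-renderer counterparts remain in the system but are disabled, suggesting a swap to lightweight versions.
# It exposes Duration and Vehicle_Speed as float user parameters; how they drive the emitters is not shown.
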NiagaraSystem /Valet/FX/Niagara/NS_Valet_Wreck
dependencies=null
emitters {
 LW_Fire_Sprites_SubUV {rendererClasses [], renderers []}
 LW_Local_Embers_Forward {rendererClasses [], renderers []}
 LW_World_Smoke_Forward {rendererClasses [], renderers []}
}
summary {
 systemName=NS_Valet_Wreck
 emitters [
  {bEnabled=false, emitterName=Local_Embers_Forward, rendererClasses [&NiagaraSpriteRendererProperties]}
  {bEnabled=false, emitterName=World_Smoke_Forward, rendererClasses [&NiagaraSpriteRendererProperties]}
  {bEnabled=false, emitterName=Fire_Sprites_SubUV, rendererClasses [&NiagaraSpriteRendererProperties]}
  {emitterName=LW_Local_Embers_Forward, rendererClasses []}
  {emitterName=LW_World_Smoke_Forward, rendererClasses []}
  {emitterName=LW_Fire_Sprites_SubUV, rendererClasses []}
 ]
 userVariables [
  User.Duration {type=NiagaraFloat, defaultValue=5, description=""}
  User.Vehicle_Speed {type=NiagaraFloat, defaultValue=0, description=""}
 ]
}
