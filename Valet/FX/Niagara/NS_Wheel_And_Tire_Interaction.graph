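# Niagara system for vehicle wheel effects, such as tire dust, sand and snow spray, plus exhaust smoke with gear-shift flashes.
# System scripts derive booleans from bitwise input and peel-out surface IDs, feeding emitters alongside user speed, slide and wheel-contact values.
# GPU tire dust partitions particles across active wheel bones, offset by wheel width and radius.
# Spawn groups multiplex sprite, mesh and light layers within emitters through masked inputs and visibility tags.
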NiagaraSystem /Valet/FX/Niagara/NS_Wheel_And_Tire_Interaction
dependencies {
 usedDataInterfaces [
  &NiagaraDataInterfaceArrayColor
  &NiagaraDataInterfaceMeshRendererInfo
  &NiagaraDataInterfaceCurve
  &NiagaraDataInterfaceArrayFloat3
  &NiagaraDataInterfaceArrayInt32
  &NiagaraDataInterfaceColorCurve
  &NiagaraDataInterfaceSkeletalMesh
 ]
 usedDynamicInputs [
  &MakeFloatFromVector2d
  &SimulationPosition
  &self:TogglesActive
 ]
 usedModules [
  &self:SystemSpawnScript.SystemScriptSource.SystemScriptGraph.NiagaraNodeAssignment_0.SetVariables_#1
  &SystemState
  &self:SetSystemBoolsfromBitwise
  &self:SystemSpawnScript.SystemScriptSource.SystemScriptGraph.NiagaraNodeAssignment_1.SetVariables_#2
  &self:SystemSpawnScript.SystemScriptSource.SystemScriptGraph.NiagaraNodeAssignment_2.SetVariables_#3
  &self:SetPeeloutIDBools
  &self:SystemSpawnScript.SystemScriptSource.SystemScriptGraph.NiagaraNodeAssignment_3.SetVariables_#4
  &Timeline
 ]
}
emitters {
 Dust_Tires_World {
  simTarget=GPUComputeSim
  emitterUpdateScript [
   &EmitterState()
   &SpawnRate(
    SpawnRate=&Multiply_Float(
     A=&self:AmountToSpawn(
      " Min Amount":NiagaraFloat=7
      Max Amount:NiagaraFloat=28
      Total Amount:NiagaraInt32=$System.TotalDust
      Vehicle Speed:NiagaraFloat=&Abs_Float(Float:NiagaraFloat=$User.Vehicle_Speed)
      Vehicle Speed Exponent:NiagaraFloat=&MakeCustomFloatFromBool(Bool=$System.SpawnSand, True Float=0.9, False Float=1.4)
     )
     B=&MakeCustomFloatFromBool(
      Bool=&SetBoolByIntComparison(A=$User.PeelOutColorIndex, B=1, Comparison Type=::NewEnumerator2)
      True Float=&FloatFromCurve(
       FloatCurve=$System.Dust_SpawnCurve
       CurveIndex=&Lerp_Float(A=0, B=&Abs_Float(Float:NiagaraFloat=$User.Vehicle_Speed), Alpha=$System.Timeline.InverseNormalizedTime)
      )
      False Float=0
     )
    )
    Use Spawn Probability=false
   )
   &SpawnRate(
    Spawn Probability=0.25
    SpawnGroup=1
    SpawnRate=&Multiply_Float(
     A=&Multiply_Float(
      A=&self:AmountToSpawn(
       " Min Amount":NiagaraFloat=0
       Max Amount:NiagaraFloat=&Lerp_Float(A=0, B=10, Alpha=&Abs_Float(Float:NiagaraFloat=$User.Vehicle_Speed))
       Total Amount:NiagaraInt32=$System.TotalDust
       Vehicle Speed:NiagaraFloat=&Abs_Float(Float:NiagaraFloat=$User.Vehicle_Speed)
       Vehicle Speed Exponent:NiagaraFloat=&MakeCustomFloatFromBool(Bool=$System.SpawnSand, True Float=0.9, False Float=1.4)
      )
      B=&Multiply_Float(
       A=&SelectFloatFromArray(Direct Array Index=$User.PeelOutColorIndex, Float Selection Array=DataInterface() {propertyValues {FloatData [0,1,1,1,0,0,0,0]}})
       B=&MakeCustomFloatFromBool(Bool=$System.VehicleHasDriver, True Float=1, False Float=0)
      )
     )
     B=&MakeCustomFloatFromBool(
      Bool=&Boolean_OrOperation(
       Bool1=&Boolean_OrOperation(
        Bool1=$System.Dust_Wheel_RL
        Bool2=&Boolean_OrOperation(
         Bool1=$System.TireIntactRL
         Bool2=&SetBoolByFloatComparison(
          A=&SelectFloatFromArray(Direct Array Index=3, Float Selection Array=$User.WheelsTouchingArray)
          B=1
          Comparison Type=::NewEnumerator2
         )
        )
       )
       Bool2=&Boolean_OrOperation(
        Bool1=$System.Dust_Wheel_RR
        Bool2=&Boolean_OrOperation(
         Bool1=$System.TireIntactRR
         Bool2=&SetBoolByFloatComparison(
          A=&SelectFloatFromArray(Direct Array Index=2, Float Selection Array=$User.WheelsTouchingArray)
          B=1
          Comparison Type=::NewEnumerator2
         )
        )
       )
      )
      True Float=1
      False Float=0
     )
    )
    Use Spawn Probability=true
   ) {module=SpawnRate001}
   &SpawnRate(
    SpawnGroup=2
    SpawnRate=&Multiply_Float(
     A=&self:AmountToSpawn(
      " Min Amount":NiagaraFloat=16
      Max Amount:NiagaraFloat=40
      Total Amount:NiagaraInt32=$System.TotalDust
      Vehicle Speed:NiagaraFloat=&Abs_Float(Float:NiagaraFloat=$User.Vehicle_Speed)
      Vehicle Speed Exponent:NiagaraFloat=1.4
     )
     B=&Multiply_Float(
      A=&SelectFloatFromArray(Direct Array Index=$User.PeelOutColorIndex, Float Selection Array=DataInterface() {propertyValues {FloatData [0,0,0,0,0,1,1,0]}})
      B=&MakeCustomFloatFromBool(Bool=$System.VehicleHasDriver, True Float=1, False Float=0)
     )
    )
    Use Spawn Probability=false
   ) {module=SpawnRate002}
   &SpawnBurst_Instantaneous(Spawn Count=$System.TotalDust, Spawn Group=3, Use Spawn Probability=false, UseLoopCountLimit=false)
  ]
  particleSpawnScript [
   /SetVariables(Particles.VisibilityTag=&SpawnGroup()) {isSetParameters=true, module=SetVariables_#5}
   &PartitionParticles(Number Of Partitions=$System.TotalDust)
   /SetVariables(Particles.BoneSelection:NiagaraInt32=&self:TogglesBoneSet(
     Partitions:NiagaraInt32=$Particles.PartitionParticles.Partition
     Front Left Tire:NiagaraBool=&Boolean_OrOperation(
      Bool1=$System.Dust_Wheel_FL
      Bool2=&Boolean_OrOperation(
       Bool1=$System.TireIntactFL
       Bool2=&SetBoolByFloatComparison(
        A=&SelectFloatFromArray(Direct Array Index=1, Float Selection Array=$User.WheelsTouchingArray)
        B=1
        Comparison Type=::NewEnumerator2
       )
      )
     )
     Front Right Tire:NiagaraBool=&Boolean_OrOperation(
      Bool1=$System.Dust_Wheel_FR
      Bool2=&Boolean_OrOperation(
       Bool1=$System.TireIntactFR
       Bool2=&SetBoolByFloatComparison(
        A=&SelectFloatFromArray(Direct Array Index=0, Float Selection Array=$User.WheelsTouchingArray)
        B=1
        Comparison Type=::NewEnumerator2
       )
      )
     )
     Rear Left Tire:NiagaraBool=&Boolean_OrOperation(
      Bool1=$System.Dust_Wheel_RL
      Bool2=&Boolean_OrOperation(
       Bool1=$System.TireIntactRL
       Bool2=&SetBoolByFloatComparison(
        A=&SelectFloatFromArray(Direct Array Index=3, Float Selection Array=$User.WheelsTouchingArray)
        B=1
        Comparison Type=::NewEnumerator2
       )
      )
     )
     Rear Right Tire:NiagaraBool=&Boolean_OrOperation(
      Bool1=$System.Dust_Wheel_RR
      Bool2=&Boolean_OrOperation(
       Bool1=$System.TireIntactRR
       Bool2=&SetBoolByFloatComparison(
        A=&SelectFloatFromArray(Direct Array Index=2, Float Selection Array=$User.WheelsTouchingArray)
        B=1
        Comparison Type=::NewEnumerator2
       )
      )
     )
     Middle Left Tire:NiagaraBool=&Boolean_OrOperation(
      Bool1=$System.Dust_Wheel_ML
      Bool2=&Boolean_OrOperation(
       Bool1=$System.TireIntactML
       Bool2=&SetBoolByFloatComparison(
        A=&SelectFloatFromArray(Direct Array Index=5, Float Selection Array=$User.WheelsTouchingArray)
        B=1
        Comparison Type=::NewEnumerator2
       )
      )
     )
     Middle Right Tire:NiagaraBool=&Boolean_OrOperation(
      Bool1=$System.Dust_Wheel_MR
      Bool2=&Boolean_OrOperation(
       Bool1=$System.TireIntactMR
       Bool2=&SetBoolByFloatComparison(
        A=&SelectFloatFromArray(Direct Array Index=4, Float Selection Array=$User.WheelsTouchingArray)
        B=1
        Comparison Type=::NewEnumerator2
       )
      )
     )
    )) {isSetParameters=true, module=SetVariables_#6}
   &InitializeParticle(
    Color=&MaskLinearColorBySpawnGroup(
     Masked Linear Color=&RandomRangeLinearColor(Minimum=<rgba 0.2448 0.2448 0.2448 1>, Maximum=<rgba 1 1 1 1>)
     Passthrough Linear Color=&MaskLinearColorBySpawnGroup(
      Masked Linear Color=&SelectLinearColorFromArray(
       Array Sampling Mode=::NewEnumerator1
       Direct Array Index=$User.PeelOutColorIndex
       Color Selection Array=DataInterface() {propertyValues {ColorData [<rgba 0 0 0 1>, <rgba 0 0 0 1>, <rgba 0 0 0 1>, <rgba 0 0 0 1>, <rgba 0 0 0 1>, <rgba 1 1 1 1>, <rgba 0.3939 0.7539 1 1>]}}
       Random Seed=?
      )
      Passthrough Linear Color=&SelectLinearColorFromArray(
       Array Sampling Mode=::NewEnumerator1
       Direct Array Index=$User.PeelOutColorIndex
       Color Selection Array=$System.PeelOutColorArray
       Random Seed=?
      )
      Spawn Group Mask=2
     )
     Spawn Group Mask=1
    )
    Lifetime=&MaskFloatBySpawnGroup(
     Masked Float=&MakeCustomFloatFromBool(
      Bool=&Boolean_OrOperation(Bool1=$System.SpawnSnow, Bool2=$System.SpawnSand)
      True Float=&RandomRangeFloat(Minimum=1, Maximum=1.8)
      False Float=&Lerp_Float(
       A=&UniformRangedFloat(Minimum=1, Maximum=1.4)
       B=&UniformRangedFloat(Minimum=0.5, Maximum=0.3)
       Alpha=&Abs_Float(Float:NiagaraFloat=$User.Vehicle_Speed)
      )
     )
     Passthrough Float=&MaskFloatBySpawnGroup(
      Masked Float=&Lerp_Float(A=0.6, B=1.4, Alpha=&Abs_Float(Float:NiagaraFloat=$User.Vehicle_Speed))
      Passthrough Float=&MaskFloatBySpawnGroup(
       Masked Float=&RandomRangeFloat(Minimum=1, Maximum=2)
       Passthrough Float=&RandomRangeFloat(Minimum=0.1, Maximum=0.2)
       Spawn Group Mask=2
      )
      Spawn Group Mask=1
     )
    )
    Mesh Renderer Array Visibility Mode=::NewEnumerator3
    Mesh Renderer Info=DataInterface() {propertyValues {MeshRenderer=&self:Dust_Tires_World_11.NiagaraMeshRendererProperties_0}}
    Mesh Scale Mode=::NewEnumerator1
    Mesh Uniform Scale Max=0.25
    Mesh Uniform Scale Min=0.05
    Sprite Rotation Mode=::NewEnumerator2
    Sprite Size=&MaskVector2dBySpawnGroup(
     Masked Vector2D=&Vector2DFromFloat(Value=&Lerp_Float(
       A=&Lerp_Float(
        A=&UniformRangedFloat(Minimum=90, Maximum=150)
        B=&UniformRangedFloat(Minimum=300, Maximum=400)
        Alpha=&Abs_Float(Float:NiagaraFloat=$User.Vehicle_Speed)
       )
       B=550
       Alpha=&Abs_Float(Float:NiagaraFloat=$User.Slide_Direction)
      ))
     Passthrough Vector2D=&Vector2DFromFloat(Value=&RandomRangeFloat(Minimum=5, Maximum=9))
     Spawn Group Mask=0
    )
    Sprite Size Mode=::NewEnumerator4
    UsePositionOffset=false
    Write Material Random=true
   )
   &SubUVAnimation(End Frame=3, SubUV Animation Mode:ENiagara_SubUVLookupMode=::NewEnumerator2)
   &SkeletalMeshLocation(
    Bone / Socket Index=&MaskIntBySpawnGroup(
     Masked Int=&MakeCustomIntFromBool(
      Bool=&Boolean_AndOperation(
       Bool1=&Boolean_OrOperation(
        Bool1=$System.Dust_Wheel_RR
        Bool2=&Boolean_OrOperation(
         Bool1=$System.TireIntactRR
         Bool2=&SetBoolByFloatComparison(
          A=&SelectFloatFromArray(Direct Array Index=2, Float Selection Array=$User.WheelsTouchingArray)
          B=1
          Comparison Type=::NewEnumerator2
         )
        )
       )
       Bool2=&Boolean_OrOperation(
        Bool1=$System.Dust_Wheel_RL
        Bool2=&Boolean_OrOperation(
         Bool1=$System.TireIntactRL
         Bool2=&SetBoolByFloatComparison(
          A=&SelectFloatFromArray(Direct Array Index=3, Float Selection Array=$User.WheelsTouchingArray)
          B=1
          Comparison Type=::NewEnumerator2
         )
        )
       )
      )
      True INT=&MakeCustomIntFromBool(
       Bool=&SetBoolByIntComparison(A=$Particles.PartitionParticles.Partition, B=2, Comparison Type=::NewEnumerator0)
       True INT=3
       False INT=2
      )
      False INT=&MakeCustomIntFromBool(
       Bool=&Boolean_OrOperation(
        Bool1=$System.Dust_Wheel_RR
        Bool2=&Boolean_OrOperation(
         Bool1=$System.TireIntactRR
         Bool2=&SetBoolByFloatComparison(
          A=&SelectFloatFromArray(Direct Array Index=2, Float Selection Array=$User.WheelsTouchingArray)
          B=1
          Comparison Type=::NewEnumerator2
         )
        )
       )
       True INT=2
       False INT=&MakeCustomIntFromBool(
        Bool=&Boolean_OrOperation(
         Bool1=$System.Dust_Wheel_RL
         Bool2=&Boolean_OrOperation(
          Bool1=$System.TireIntactRL
          Bool2=&SetBoolByFloatComparison(
           A=&SelectFloatFromArray(Direct Array Index=3, Float Selection Array=$User.WheelsTouchingArray)
           B=1
           Comparison Type=::NewEnumerator2
          )
         )
        )
        True INT=3
        False INT=-1
       )
      )
     )
     Passthrough Int=$Particles.BoneSelection
     Spawn Group Mask=1
    )
    MaskSpawnGroup=false
    Mesh Sampling Type=::NewEnumerator5
    Orientation Sampling=::NewEnumerator1
    Sampled Position Offset=&MakeVector(
     X=0
     Y=&MaskFloatBySpawnGroup(
      Masked Float=&self:WheelWidthUniformRange(
       WheelWidthArray:NiagaraDataInterfaceArrayFloat=$User.Wheel_Width_Array
       WheelIndex:NiagaraInt32=&MakeCustomIntFromBool(
        Bool=&SetBoolByIntComparison(A=$Particles.PartitionParticles.Partition, B=1, Comparison Type=::NewEnumerator0)
        True INT=3
        False INT=2
       )
      )
      Passthrough Float=&self:WheelWidthUniformRange(WheelWidthArray:NiagaraDataInterfaceArrayFloat=$User.Wheel_Width_Array, WheelIndex:NiagaraInt32=$Particles.BoneSelection)
      Spawn Group Mask=1
     )
     Z=&MaskFloatBySpawnGroup(
      Masked Float=&self:WheelRadiusOffset(
       Wheel Radius Array:NiagaraDataInterfaceArrayFloat=$User.Wheel_Radius_Array
       Wheel Index:NiagaraInt32=&MakeCustomIntFromBool(
        Bool=&SetBoolByIntComparison(A=$Particles.PartitionParticles.Partition, B=1, Comparison Type=::NewEnumerator0)
        True INT=3
        False INT=2
       )
       Additional Offset:NiagaraFloat=0
       Wheel Radius Addl Scale:NiagaraFloat=1
      )
      Passthrough Float=&self:WheelRadiusOffset(
       Wheel Radius Array:NiagaraDataInterfaceArrayFloat=$User.Wheel_Radius_Array
       Wheel Index:NiagaraInt32=$Particles.BoneSelection
       Additional Offset:NiagaraFloat=0
       Wheel Radius Addl Scale:NiagaraFloat=1
      )
      Spawn Group Mask=1
     )
    )
    Skeletal Mesh=$System.Vehicle_Selection_02
    Socket Sampling Mode=::NewEnumerator5
   )
   &ShapeLocation(MaskSpawnGroup=true, Override Seed=false, Spawn Group Mask=2, Sphere Radius=18, UseOwnerScale=true)
   &AddVelocity(
    Cone Angle=12
    Cone Axis=&MaskVectorBySpawnGroup(
     Masked Vector=&MakeVector(X=1, Y=$User.Slide_Direction, Z=0)
     Passthrough Vector=&MakeVector(X=1, Y=$User.Slide_Direction, Z=0.25)
    )
    Distribution Along Cone Axis=0.6
    Module.UseConeFalloff=false
    Module.UseVelDistribution=true
    Override Seed=false
    Velocity Mode=::NewEnumerator2
    Velocity Speed=&MaskFloatBySpawnGroup(
     Masked Float=&Lerp_Float(A=0, B=1, Alpha=&Abs_Float(Float:NiagaraFloat=$User.Vehicle_Speed))
     Passthrough Float=&MaskFloatBySpawnGroup(
      Masked Float=&Lerp_Float(
       A=&RandomRangeFloat(Minimum=0, Maximum=1)
       B=&RandomRangeFloat(Minimum=300, Maximum=800)
       Alpha=&Abs_Float(Float:NiagaraFloat=$User.Vehicle_Speed)
      )
      Passthrough Float=0
      Spawn Group Mask=1
     )
    )
   ) {module=AddVelocity001}
   &AddVelocity(Velocity=&MaskVectorBySpawnGroup(
     Masked Vector=&MakeVector(
      X=0
      Y=&Multiply_Float(A=$User.Slide_Direction, B=300)
      Z=&Multiply_Float(
       A=&MakeCustomFloatFromBool(
        Bool=&Boolean_OrOperation(Bool1=$System.SpawnSnow, Bool2=$System.SpawnSand)
        True Float=&SelectFloatFromArray(
         Direct Array Index=&SpawnGroup()
         Float Selection Array=DataInterface() {propertyValues {FloatData [0,0,400,400,0,0]}}
        )
        False Float=0
       )
       B=&Lerp_Float(A=0, B=1, Alpha=&Abs_Float(Float:NiagaraFloat=$User.Vehicle_Speed))
      )
     )
     Passthrough Vector=&MaskVectorBySpawnGroup(
      Masked Vector=&MakeVector(X=0, Y=&Multiply_Float(A=$User.Slide_Direction, B=1800), Z=0)
      Passthrough Vector=&MaskVectorBySpawnGroup(
       Masked Vector=&RandomRangeVector(Minimum=<20 -100 300>, Maximum=&MakeVector(X=300, Y=100, Z=600))
       Passthrough Vector=<0 0 0>
       Spawn Group Mask=2
      )
      Spawn Group Mask=1
     )
    ))
   /SetVariables(Particles.Lifetime=&MakeCustomFloatFromBool(
     Bool=&SetBoolByIntComparison(A=&SpawnGroup(), B=3, Comparison Type=::NewEnumerator4)
     True Float=&Multiply_Float(A=$Particles.Lifetime, B=2)
     False Float=$Particles.Lifetime
    )) {isSetParameters=true}
  ]
  particleUpdateScript [
   &ParticleState()
   &GravityForce(Gravity=&MaskVectorBySpawnGroup(
     Masked Vector=&MakeVector(
      X=0
      Y=0
      Z=&MakeCustomFloatFromBool(Bool=&Boolean_OrOperation(Bool1=$System.SpawnSand, Bool2=$System.SpawnSnow), True Float=-800, False Float=50)
     )
     Passthrough Vector=&MaskVectorBySpawnGroup(Masked Vector=<0 0 0>, Passthrough Vector=<0 0 -980>, Spawn Group Mask=3)
    ))
   &Drag(
    Drag=&MaskFloatBySpawnGroup(
     Masked Float=&RandomRangeFloat(Minimum=0.5, Maximum=1.25)
     Passthrough Float=&MaskFloatBySpawnGroup(
      Masked Float=1
      Passthrough Float=&MaskFloatBySpawnGroup(Masked Float=0, Passthrough Float=0.5, Spawn Group Mask=3)
      Spawn Group Mask=2
     )
     Spawn Group Mask=1
    )
    Use Linear Drag=true
    Use Rotational Drag=false
   )
   &ScaleColor(
    Scale Alpha=&MaskFloatBySpawnGroup(
     Masked Float=&FloatFromCurve(
      FloatCurve=DataInterface() {propertyValues {Curve {keys [
          {arriveTangent=4.1785, interpMode=RCIM_Cubic, leaveTangent=4.1785, tangentMode=RCTM_User, time=0, value=0}
          {arriveTangent=0.9995, interpMode=RCIM_Cubic, leaveTangent=0.9995, tangentMode=RCTM_User, time=0.1127, value=1.0138}
          {arriveTangent=-1.1426, interpMode=RCIM_Cubic, leaveTangent=-1.1426, time=0.5187, value=0.2402}
          {arriveTangent=-0.3824, interpMode=RCIM_Cubic, leaveTangent=-0.0405, tangentMode=RCTM_Break, time=1, value=0}
         ]}}}
      Scale Curve=0.7
     )
     Passthrough Float=&MaskFloatBySpawnGroup(
      Masked Float=&FloatFromCurve(
       FloatCurve=DataInterface() {propertyValues {Curve {keys [
           {arriveTangent=4.1785, interpMode=RCIM_Cubic, leaveTangent=4.1785, tangentMode=RCTM_User, time=0, value=0}
           {arriveTangent=0.9995, interpMode=RCIM_Cubic, leaveTangent=0.9995, tangentMode=RCTM_User, time=0.1438, value=1.0199}
           {arriveTangent=-1.1913, interpMode=RCIM_Cubic, leaveTangent=-1.1913, time=0.5124, value=0.3091}
           {arriveTangent=-0.3824, interpMode=RCIM_Cubic, leaveTangent=-0.0405, tangentMode=RCTM_Break, time=1, value=0}
          ]}}}
       Scale Curve=5
      )
      Passthrough Float=&FloatFromCurve(
       FloatCurve=DataInterface() {propertyValues {Curve {keys [
           {interpMode=RCIM_Cubic, tangentMode=RCTM_Break, time=-0.0054, value=0.0052}
           {arriveTangent=-0.6325, interpMode=RCIM_Cubic, leaveTangent=0.2892, tangentMode=RCTM_Break, time=0.0811, value=0.9977}
           {arriveTangent=-0.7171, interpMode=RCIM_Cubic, leaveTangent=-0.7171, tangentMode=RCTM_User, time=0.3372, value=0.3367}
           <at 1.0011 0.002 RCIM_Cubic>
          ]}}}
       Scale Curve=&FloatFromCurve(
        FloatCurve=DataInterface() {propertyValues {Curve {
           defaultValue=1
           keys [
            {interpMode=RCIM_Cubic, tangentMode=RCTM_User, time=0, value=1}
            {arriveTangent=-1.6398, interpMode=RCIM_Cubic, leaveTangent=-1.6398, tangentMode=RCTM_User, time=1, value=0}
           ]
          }}}
        CurveIndex=$System.Timeline.NormalizedTime
        Scale Curve=0.5
       )
      )
      Spawn Group Mask=3
     )
     Spawn Group Mask=2
    )
    ScaleA=true
    ScaleRGB=true
   )
   &InheritSourceMovement(
    Applied Position Scale=&MaskVectorBySpawnGroup(
     Masked Vector=<0 0 0>
     Passthrough Vector=&MaskVectorBySpawnGroup(Masked Vector=<0 0 0>, Passthrough Vector=<0.5 0.5 0.5>, Spawn Group Mask=3)
     Spawn Group Mask=1
    )
    ApplyForce=false
    ApplyPosition=true
    ApplyPositionFalloff=false
   )
   &Collision(
    Age Colliding Particles=false
    Collision Enabled=/Niagara/DynamicInputs/Helpers/MaskBoolBySpawnGroup(
     Masked Bool:NiagaraBool=false
     Passthrough Bool:NiagaraBool=&MaskBoolBySpawnGroup(
      Masked Bool:NiagaraBool=false
      Passthrough Bool:NiagaraBool=&Boolean_AndOperation(
       Bool1=&SetBoolByFloatComparison(A=&RandomRangeFloat(Minimum=0, Maximum=1), B=0.25)
       Bool2=&SetBoolByFloatComparison(A=$Particles.Age, B=0.05)
      )
      Spawn Group Mask:NiagaraInt32=3
     )
     Spawn Group Mask:NiagaraInt32=0
    )
    Control Roll On Collision=true
    Friction=0.4
    Friction During a Bounce=0.2
    GPU Collision Type=::NewEnumerator2
    KillOnCollision=false
    Mesh Dimensions:Vector3f=<30 30 30>
    Module.Control Rest State Via Surface Normal=false
    Module.EnableMaxCollisionCount=false
    Module.Randomize Collision Normal Vector=true
    Radius Calculation Type=::NewEnumerator1
    Restitution=0.7
    ScaleRateOfRotation=false
   )
   &UpdateMeshOrientation(
    Rotation Rate=&MaskFloatBySpawnGroup(
     Masked Float=0
     Passthrough Float=&MakeCustomFloatFromBool(Bool=$Particles.HasCollided, True Float=0, False Float=1)
     Spawn Group Mask=3
    )
    Rotation Vector=&RandomRangeVector(Minimum=<0 0 0>, Maximum=<1 1 1>)
   )
   &SolveForcesAndVelocity(Clamp Velocity=false, Limit Acceleration=false)
   &ScaleSpriteSize(Scale Factor=&Vector2DFromFloat(Value=&MaskFloatBySpawnGroup(
      Masked Float=&FloatFromCurve(FloatCurve=DataInterface() {propertyValues {Curve {keys [<at 0 1>,<at 1 0>]}}})
      Passthrough Float=&FloatFromCurve(
       FloatCurve=DataInterface() {propertyValues {Curve {keys [{arriveTangent=2.0487, interpMode=RCIM_Cubic, leaveTangent=2.0487, tangentMode=RCTM_User, time=0, value=0.15}, <at 1 1>]}}}
       Scale Curve=&Lerp_Float(A=1, B=0.4, Alpha=$System.Timeline.InverseNormalizedTime)
      )
      Spawn Group Mask=2
     )))
   &SpriteRotationRate(Rotation Rate=&Multiply_Float(A=$Particles.Initial.MaterialRandom, B=1))
   &DynamicMaterialParameters(
    Index 0 Param 1=&Add_Float(
     A=&RandomRangeFloat(Minimum=0, Maximum=1)
     B=&FloatFromCurve(FloatCurve=DataInterface() {propertyValues {Curve {keys [
          {arriveTangent=2.7533, interpMode=RCIM_Cubic, leaveTangent=2.847, tangentMode=RCTM_Break, time=0, value=0}
          {arriveTangent=1.3581, interpMode=RCIM_Cubic, leaveTangent=1.3581, tangentMode=RCTM_User, time=0.2558, value=0.5125}
          {arriveTangent=0.0529, interpMode=RCIM_Cubic, leaveTangent=0.2021, tangentMode=RCTM_User, time=1, value=1}
         ]}}})
    )
    Index 0 Param 3=&FloatFromCurve(FloatCurve=DataInterface() {propertyValues {Curve {keys [{arriveTangent=-53.9199, time=0, value=0}, {leaveTangent=0.2723, time=0.25, value=0}]}}})
    Index 0 Param 4=&FloatFromCurve(FloatCurve=DataInterface() {propertyValues {Curve {keys [<at 0 1>,<at 1 0>]}}})
    Index 1 Param 1=&FloatFromCurve(FloatCurve=DataInterface() {propertyValues {Curve {keys [<at 0 0>, {arriveTangent=4.2257, interpMode=RCIM_Cubic, leaveTangent=4.2257, tangentMode=RCTM_User, time=0.25, value=0}]}}})
    Index 1 Param 2=&FloatFromCurve(FloatCurve=DataInterface() {propertyValues {Curve {keys [<at 0 0>]}}})
    Index 1 Param 3=&FloatFromCurve(FloatCurve=DataInterface() {propertyValues {Curve {keys [{interpMode=RCIM_Cubic, leaveTangent=6.3917, tangentMode=RCTM_Break, time=0, value=0}]}}})
    Index 1 Param 4=&FloatFromCurve(FloatCurve=DataInterface() {propertyValues {Curve {keys [{arriveTangent=-0.136, leaveTangent=-0.136, time=0, value=1}]}}})
    Index 2 Param 1=&FloatFromCurve(FloatCurve=DataInterface() {propertyValues {Curve {keys [
         {arriveTangent=2.646, interpMode=RCIM_Cubic, leaveTangent=2.646, tangentMode=RCTM_User, time=0, value=0}
         {leaveTangent=-0.0719, time=1, value=1}
        ]}}})
    Index 2 Param 2=&FloatFromCurve(
     FloatCurve=DataInterface() {propertyValues {Curve {keys [
         {arriveTangent=1.9056, time=0, value=0.2}
         {arriveTangent=0.0259, interpMode=RCIM_Cubic, leaveTangent=0.0259, tangentMode=RCTM_User, time=1, value=1.4}
        ]}}}
     Scale Curve=&RandomRangeFloat(Minimum=0.5, Maximum=1)
    )
    Index 2 Param 3=&FloatFromCurve(FloatCurve=DataInterface() {propertyValues {Curve {keys [
         {interpMode=RCIM_Cubic, leaveTangent=2.3064, tangentMode=RCTM_Break, time=0, value=1}
         {interpMode=RCIM_Cubic, leaveTangent=11.1233, tangentMode=RCTM_Break, time=1, value=2}
        ]}}})
    Index 2 Param 4=&FloatFromCurve(
     FloatCurve=DataInterface() {propertyValues {Curve {keys [
         {arriveTangent=-0.6867, interpMode=RCIM_Cubic, leaveTangent=-0.6867, tangentMode=RCTM_User, time=0, value=1}
         {interpMode=RCIM_Cubic, leaveTangent=-1.2956, tangentMode=RCTM_Break, time=1, value=0.75}
        ]}}}
     Scale Curve=1.6
    )
    Module.Param0WriteEnabled=true
    Module.Param0WriteEnabled1=true
    Module.Param0WriteEnabled2=true
    Module.Param1WriteEnabled=false
    Module.Param1WriteEnabled1=true
    Module.Param1WriteEnabled2=true
    Module.Param2WriteEnabled=true
    Module.Param2WriteEnabled1=true
    Module.Param2WriteEnabled2=true
    Module.Param3WriteEnabled=true
    Module.Param3WriteEnabled1=true
    Module.Param3WriteEnabled2=true
    Write Parameter Index 1=true
    Write Parameter Index 2=true
   )
   &ScaleMeshSizeBySpeed(
    Max Scale Factor=<1 1 1>
    Min Scale Factor:Vector3f=&MaskVectorBySpawnGroup(Masked Vector=<1 1 1>, Passthrough Vector=<0.7 0.7 0.7>, Spawn Group Mask=3)
   )
   &ScaleMeshSize(Scale Factor=&VectorFromFloat(Value=&MaskFloatBySpawnGroup(
      Masked Float=1
      Passthrough Float=&FloatFromCurve(FloatCurve=DataInterface() {propertyValues {Curve {keys [
           <at 0 0.075>
           <at 0.0756 1>
           {arriveTangent=-0.2, interpMode=RCIM_Cubic, leaveTangent=-0.2, time=0.7956, value=1.0207}
           {arriveTangent=-5.4083, interpMode=RCIM_Cubic, leaveTangent=-5.4083, tangentMode=RCTM_User, time=1, value=0}
          ]}}})
      Spawn Group Mask=3
     )))
   &SkeletalMeshLocation(
    Bone / Socket Index=$Particles.BoneSelection
    MaskSpawnGroup=true
    Mesh Sampling Type=::NewEnumerator5
    Sampled Position Offset=&SelectVectorFromArray(
     Direct Array Index=$Particles.BoneSelection
     Vector Selection Array=DataInterface() {propertyValues {FloatData [<0 0 0>, <0 0 0>, <0 0 0>, <0 0 0>, <0 0 0>, <0 0 0>]}}
    )
    Skeletal Mesh=$System.Vehicle_Selection_02
    Socket Sampling Mode=::NewEnumerator5
    Spawn Group Mask=3
    Use Sampled Mesh Attribute as Ribbon ID=false
   ) {enabled=false, module=SkeletalMeshLocation003}
   &SkeletalMeshLocation(
    Bone / Socket Index=$Particles.BoneSelection
    MaskSpawnGroup=false
    Mesh Sampling Type=::NewEnumerator5
    Orientation Sampling=::NewEnumerator1
    Sampled Position Offset=&SelectVectorFromArray(
     Direct Array Index=$Particles.BoneSelection
     Vector Selection Array=DataInterface() {propertyValues {FloatData [<0 0 0>, <0 0 0>, <0 0 0>, <0 0 0>, <0 0 0>, <0 0 0>]}}
    )
    Skeletal Mesh=$System.Vehicle_Selection_02
    Socket Sampling Mode=::NewEnumerator5
    Spawn Group Mask=3
    Use Sampled Mesh Attribute as Ribbon ID=false
    Write To Particle Position=false
   ) {module=SkeletalMeshLocation001}
   /SetVariables(Particles.DustWheel Dynamic Parameter:Vector4f=&MakeVector4(X=5, Y=1, Z=0, W=0)) {isSetParameters=true, module=SetVariables_#7}
   /SetVariables(Particles.WheelScale:Vector3f=&self:ScaleMeshfromWheelDimensions(
     Mesh Unscaled Size:Vector3f=<49 49 49>
     Wheel Radius Array:NiagaraDataInterfaceArrayFloat=$User.Wheel_Radius_Array
     Wheel Width Array:NiagaraDataInterfaceArrayFloat=$User.Wheel_Width_Array
     Wheel Index:NiagaraInt32=$Particles.BoneSelection
     Additional Scaling:Vector3f=<1 1 1>
    )) {isSetParameters=true, module=SetVariables_#8}
   &ScaleColor(
    Scale Alpha=&Lerp_Float(A=0, B=1, Alpha=&Abs_Float(Float:NiagaraFloat=$User.Vehicle_Speed))
    ScaleA=true
    ScaleRGB=false
   ) {module=ScaleColor001}
   /SetVariables(
    Particles.MeshOrientation=&self:ScratchDynamicInput(
     Quat:Quat4f=$Particles.SkeletalMeshLocation001.SampledRotation
     Mask Index Attribute:NiagaraInt32=&SpawnGroup()
     Mask Index:NiagaraInt32=3
    )
    Particles.Position=&MaskVectorBySpawnGroup(
     Masked Vector=$StackContext.SkeletalMeshLocation001.SampledPosition
     Passthrough Vector=Linked() {linkedVariable {name=Particles.Position, type=NiagaraPosition}}
     Spawn Group Mask=3
    )
   ) {isSetParameters=true}
  ]
  rendererClasses [&NiagaraSpriteRendererProperties, &NiagaraMeshRendererProperties]
  renderers [
   {rendererClass=&NiagaraSpriteRendererProperties}
   {rendererClass=&NiagaraMeshRendererProperties, rendererIndex=1}
   {rendererClass=&NiagaraSpriteRendererProperties, rendererIndex=2}
   {rendererClass=&NiagaraMeshRendererProperties, rendererIndex=3}
   {rendererClass=&NiagaraMeshRendererProperties, rendererIndex=4}
  ]
 }
 Exhaust_Local {
  emitterUpdateScript [
   &EmitterState()
   &SpawnRate(
    Spawn Probability=&MakeCustomFloatFromBool(
     Bool=&Boolean_AndOperation(Bool1=$System.VehicleHasDriver, Bool2=$System.VehicleHasFuel)
     True Float=1
     False Float=0
    )
    SpawnRate=&Multiply_Float(
     A=24
     B=&Multiply_Float(
      A=&MakeCustomFloatFromBool(Bool=&Boolean_OrOperation(Bool1=$User.IsElectric, Bool2=$System.IsWaterDeep), True Float=0, False Float=1)
      B=&MakeCustomFloatFromBool(Bool=$System.isExhaustEnable, True Float=1, False Float=0)
     )
    )
    Use Spawn Probability=true
   )
   &SpawnRate(
    Spawn Probability=&MakeCustomFloatFromBool(
     Bool=&Boolean_AndOperation(Bool1=$System.VehicleHasDriver, Bool2=$System.VehicleHasFuel)
     True Float=1
     False Float=0
    )
    SpawnGroup=1
    SpawnRate=&Multiply_Float(
     A=24
     B=&Multiply_Float(
      A=&MakeCustomFloatFromBool(Bool=&Boolean_OrOperation(Bool1=$User.IsElectric, Bool2=$System.IsWaterDeep), True Float=0, False Float=1)
      B=&MakeCustomFloatFromBool(Bool=$System.isExhaustEnable, True Float=1, False Float=0)
     )
    )
    Use Spawn Probability=true
   ) {module=SpawnRate001}
   &SpawnBurst_Instantaneous(
    Spawn Count=&Multiply_Int(
     A=&RandomRangeInt(Minimum=1, Maximum=2)
     B=&MakeCustomIntFromBool(Bool=$System.IsWaterDeep, True INT=0, False INT=1)
    )
    Spawn Group=2
    Spawn Probability=&MakeCustomFloatFromBool(
     Bool=&Boolean_AndOperation(
      Bool1=$System.VehicleHasDriver
      Bool2=&Boolean_AndOperation(Bool1=$System.VehicleHasFuel, Bool2=$System.isShifting)
     )
     True Float=1
     False Float=0
    )
    Use Spawn Probability=true
   )
   &SpawnBurst_Instantaneous(
    Spawn Count=&Multiply_Int(
     A=&RandomRangeInt(Minimum=1, Maximum=2)
     B=&MakeCustomIntFromBool(Bool=$System.IsWaterDeep, True INT=0, False INT=1)
    )
    Spawn Group=3
    Spawn Probability=&MakeCustomFloatFromBool(
     Bool=&Boolean_AndOperation(
      Bool1=$System.VehicleHasDriver
      Bool2=&Boolean_AndOperation(Bool1=$System.VehicleHasFuel, Bool2=$System.isShifting)
     )
     True Float=1
     False Float=0
    )
    Use Spawn Probability=true
   ) {module=SpawnBurst_Instantaneous001}
   &SpawnBurst_Instantaneous(
    Spawn Count=&Multiply_Int(A=1, B=&MakeCustomIntFromBool(Bool=$System.IsWaterDeep, True INT=0, False INT=1))
    Spawn Group=4
    Spawn Probability=&MakeCustomFloatFromBool(
     Bool=&Boolean_AndOperation(
      Bool1=$System.VehicleHasDriver
      Bool2=&Boolean_AndOperation(Bool1=$System.VehicleHasFuel, Bool2=$System.isShifting)
     )
     True Float=1
     False Float=0
    )
    Use Spawn Probability=true
   ) {module=SpawnBurst_Instantaneous002}
   &SpawnBurst_Instantaneous(
    Spawn Count=&Multiply_Int(A=1, B=&MakeCustomIntFromBool(Bool=$System.IsWaterDeep, True INT=0, False INT=1))
    Spawn Group=5
    Spawn Probability=&MakeCustomFloatFromBool(
     Bool=&Boolean_AndOperation(
      Bool1=$System.VehicleHasDriver
      Bool2=&Boolean_AndOperation(Bool1=$System.VehicleHasFuel, Bool2=$System.isShifting)
     )
     True Float=1
     False Float=0
    )
    Use Spawn Probability=true
   ) {module=SpawnBurst_Instantaneous003}
  ]
  particleSpawnScript [
   &InitializeParticle(
    Color=&MakeCustomLinearColorFromBool(
     Bool=&SetBoolByIntComparison(A=&SpawnGroup(), B=2, Comparison Type=::NewEnumerator2)
     True Color=<rgba 1 1 1 1>
     False Color=$User.ExhaustColor
    )
    Lifetime=&MakeCustomFloatFromBool(
     Bool=&SetBoolByIntComparison(A=&SpawnGroup(), B=2, Comparison Type=::NewEnumerator2)
     True Float=&MakeCustomFloatFromBool(
      Bool=$System.Downshifting
      True Float=&RandomRangeFloat(Minimum=0.1, Maximum=0.14)
      False Float=&RandomRangeFloat(Minimum=0.225, Maximum=0.375)
     )
     False Float=&Lerp_Float(A=&RandomRangeFloat(Minimum=0.225, Maximum=0.8), B=1, Alpha=$User.Vehicle_Speed)
    )
    Position Mode=::NewEnumerator0
    Sprite Rotation Mode=::NewEnumerator2
    Sprite Size Mode=::NewEnumerator3
    Uniform Sprite Size=&Lerp_Float(
     A=&RandomRangeFloat(Minimum=40, Maximum=60)
     B=&RandomRangeFloat(Minimum=60, Maximum=80)
     Alpha=$User.Vehicle_Speed
    )
    Write Material Random=true
   ) {module=InitializeParticle001}
   /SetVariables(Particles.VisibilityTag=&SelectIntFromArray(Direct Array Index=&SpawnGroup(), Int Selection Array=DataInterface() {propertyValues {IntData [0,0,1,1,1,1]}})) {isSetParameters=true}
  ]
  particleUpdateScript [
   &ParticleState()
   &SkeletalMeshLocation(
    Bone / Socket Index=&SelectIntFromArray(Direct Array Index=&SpawnGroup(), Int Selection Array=DataInterface() {propertyValues {IntData [6,7,6,7,6,7]}})
    "Inherit Velocity (Scale)":NiagaraFloat=&MakeCustomFloatFromBool(
     Bool=&SetBoolByIntComparison(A=&SpawnGroup(), B=2, Comparison Type=::NewEnumerator2)
     True Float=0
     False Float=-0.05
    )
    MaskSpawnGroup=false
    Mesh Sampling Type=::NewEnumerator5
    Sampled Position Offset=&MaskVectorBySpawnGroup(
     Masked Vector=&Multiply_VectorByFloat(
      Vector=<1 0 0>
      Float=&Multiply_Float(A=$Particles.NormalizedAge, B=&Lerp_Float(A=90, B=60, Alpha=$User.Vehicle_Speed))
     )
     Passthrough Vector=&MaskVectorBySpawnGroup(
      Masked Vector=&Multiply_VectorByFloat(
       Vector=<1 0 0>
       Float=&Multiply_Float(A=$Particles.NormalizedAge, B=&Lerp_Float(A=90, B=60, Alpha=$User.Vehicle_Speed))
      )
      Passthrough Vector=<0 0 0>
      Spawn Group Mask=3
     )
     Spawn Group Mask=2
    )
    Skeletal Mesh=$System.Vehicle_Selection_02
    Socket Sampling Mode=::NewEnumerator5
    Velocity Sampling=::NewEnumerator0
    VelocitySpeedLimit=false
   )
   /SetVariables(Particles.Position=&MakeCustomVectorFromBool(
     Bool=&SetBoolByIntComparison(A=&SpawnGroup(), B=2, Comparison Type=::NewEnumerator2)
     True Vector=Linked() {linkedVariable {name=Particles.Position, type=NiagaraPosition}}
     False Vector=&Add_Vector(
      A=$Particles.SkeletalMeshLocation.SampledPosition
      B=&Multiply_VectorByFloat(
       Vector=&RotateVectorByQuaternion(
        Vector To Rotate=&MakeVector(X=1, Y=&Lerp_Float(A=0, B=$User.Slide_Direction, Alpha=$Particles.NormalizedAge), Z=0)
        Quaternion=$Particles.SkeletalMeshLocation.SampledRotation
       )
       Float=&Multiply_Float(A=$Particles.NormalizedAge, B=&Lerp_Float(A=125, B=100, Alpha=$User.Vehicle_Speed))
      )
     )
    )) {isSetParameters=true}
   &ScaleColor(
    Scale Alpha=&FloatFromCurve(
     FloatCurve=DataInterface() {propertyValues {Curve {keys [
         <at 0 0>
         {arriveTangent=0.002, interpMode=RCIM_Cubic, leaveTangent=-2.5776, tangentMode=RCTM_Break, time=0.1559, value=1.0113}
         <at 1.0011 0.002 RCIM_Cubic>
        ]}}}
     Scale Curve=0.5
    )
    ScaleA=true
    ScaleRGB=true
   )
   /Niagara/Modules/Update/SubUV/SubUVAnimation(Number Of Frames=&FloatFromCurve(
     FloatCurve=DataInterface() {propertyValues {Curve {keys [
         {interpMode=RCIM_Cubic, leaveTangent=9.8981, tangentMode=RCTM_Break, time=0, value=0}
         {arriveTangent=1, interpMode=RCIM_Cubic, leaveTangent=1, tangentMode=RCTM_Break, time=0.075, value=0.5807}
         <at 1 1>
        ]}}}
     Scale Curve=&Multiply_Float(A=$Particles.RandomFloat, B=63)
    ))
   &ScaleSpriteSize(Scale Factor=&Vector2DFromFloat(Value=&MakeCustomFloatFromBool(
      Bool=&SetBoolByIntComparison(A=&SpawnGroup(), B=2, Comparison Type=::NewEnumerator2)
      True Float=&FloatFromCurve(
       FloatCurve=DataInterface() {propertyValues {Curve {keys [{interpMode=RCIM_Cubic, leaveTangent=2.377, tangentMode=RCTM_Break, time=0, value=0.179}, <at 0.8329 1>]}}}
       Scale Curve=0.9
      )
      False Float=&FloatFromCurve(
       FloatCurve=DataInterface() {propertyValues {Curve {keys [{interpMode=RCIM_Cubic, tangentMode=RCTM_Break, time=0, value=0.179}, <at 0.8329 1>]}}}
       Scale Curve=$User.ExhaustScale
      )
     )))
   &SpriteRotationRate(Rotation Rate=&Multiply_Float(A=$Particles.Initial.MaterialRandom, B=1))
   /SetVariables(
    Particles.ShiftBlip Dyanmic Parameter:Vector4f=&MakeVector4(
     X=&FloatFromCurve(FloatCurve=DataInterface() {propertyValues {Curve {keys [{arriveTangent=2.3188, interpMode=RCIM_Cubic, time=0, value=0}, <at 0.5 1 RCIM_Cubic>]}}})
     Y=&RandomRangeFloat(Minimum=0.333, Maximum=1)
     Z=&FloatFromCurve(FloatCurve=DataInterface() {propertyValues {Curve {keys [{arriveTangent=1.4224, interpMode=RCIM_Cubic, time=0.5, value=1}, <at 1 2 RCIM_Cubic>]}}})
     W=&RandomRangeFloat(Minimum=1.333, Maximum=2)
    )
    Particles.ShiftBlip SubUVIndex:NiagaraFloat=&Lerp_Float(A=&RandomRangeFloat(Minimum=0, Maximum=32), B=63, Alpha=$Particles.NormalizedAge)
    Particles.ShiftBlipColor:LinearColor=&ColorFromCurve(ColorCurve=DataInterface() {propertyValues {
       AlphaCurve {keys [<at 0.4847 1 RCIM_Cubic>,<at 1 0 RCIM_Cubic>]}
       BlueCurve {keys [<at 0.0194 50 RCIM_Cubic>,<at 0.1503 150>,<at 0.1866 50>,<at 0.3975 0 RCIM_Cubic>]}
       GreenCurve {keys [<at 0.0194 30.8898 RCIM_Cubic>,<at 0.1503 92.6695>,<at 0.1866 0>,<at 0.3975 5.5901 RCIM_Cubic>]}
       RedCurve {keys [<at 0.0194 0.1 RCIM_Cubic>,<at 0.1503 0>,<at 0.1866 10.781>,<at 0.3975 25 RCIM_Cubic>]}
      }})
   ) {isSetParameters=true}
   &FN_Light_Attributes(
    Color=&ColorFromCurve(ColorCurve=DataInterface() {propertyValues {
       AlphaCurve {keys [<at 0.0118 0>,<at 0.3542 1 RCIM_Cubic>,<at 1 0 RCIM_Cubic>]}
       BlueCurve {keys [<at 0.054 0>,<at 0.4426 0 RCIM_Cubic>]}
       GreenCurve {keys [<at 0.054 4.7963>,<at 0.4426 2.6832 RCIM_Cubic>]}
       RedCurve {keys [<at 0.054 24>,<at 0.4426 12 RCIM_Cubic>]}
      }})
    Create Color Attribute=true
    Create Exponent Attribute=true
    Create Radius Attribute=true
    Exponent=8
    Radius=120
    Scale Color=&MakeCustomFloatFromBool(Bool=$System.Downshifting, True Float=0.15, False Float=1)
   )
  ]
  rendererClasses [&NiagaraSpriteRendererProperties, &NiagaraLightRendererProperties]
  renderers [
   {rendererClass=&NiagaraSpriteRendererProperties}
   {rendererClass=&NiagaraSpriteRendererProperties, rendererIndex=1}
   {rendererClass=&NiagaraLightRendererProperties, rendererIndex=2}
  ]
 }
 High_Dust_Tires_World {
  emitterUpdateScript [
   &EmitterState()
   &SpawnRate(
    SpawnRate=&Multiply_Float(
     A=&self:AmountToSpawn(
      " Min Amount":NiagaraFloat=7
      Max Amount:NiagaraFloat=28
      Total Amount:NiagaraInt32=$System.TotalDust
      Vehicle Speed:NiagaraFloat=&Abs_Float(Float:NiagaraFloat=$User.Vehicle_Speed)
      Vehicle Speed Exponent:NiagaraFloat=&MakeCustomFloatFromBool(Bool=$System.SpawnSand, True Float=0.9, False Float=1.4)
     )
     B=&MakeCustomFloatFromBool(
      Bool=&SetBoolByIntComparison(A=$User.PeelOutColorIndex, B=1, Comparison Type=::NewEnumerator2)
      True Float=&FloatFromCurve(FloatCurve=$System.Dust_SpawnCurve, CurveIndex=&Abs_Float(Float:NiagaraFloat=$User.Vehicle_Speed))
      False Float=0
     )
    )
    Use Spawn Probability=false
   )
   &SpawnRate(
    Spawn Probability=0.25
    SpawnGroup=1
    SpawnRate=&Multiply_Float(
     A=&Multiply_Float(
      A=&self:AmountToSpawn(
       " Min Amount":NiagaraFloat=0
       Max Amount:NiagaraFloat=&Lerp_Float(A=0, B=10, Alpha=&Abs_Float(Float:NiagaraFloat=$User.Vehicle_Speed))
       Total Amount:NiagaraInt32=$System.TotalDust
       Vehicle Speed:NiagaraFloat=&Abs_Float(Float:NiagaraFloat=$User.Vehicle_Speed)
       Vehicle Speed Exponent:NiagaraFloat=&MakeCustomFloatFromBool(Bool=$System.SpawnSand, True Float=0.9, False Float=1.4)
      )
      B=&Multiply_Float(
       A=&SelectFloatFromArray(Direct Array Index=$User.PeelOutColorIndex, Float Selection Array=DataInterface() {propertyValues {FloatData [0,1,1,1,0,0,0,0]}})
       B=&MakeCustomFloatFromBool(Bool=$System.VehicleHasDriver, True Float=1, False Float=0)
      )
     )
     B=&MakeCustomFloatFromBool(
      Bool=&Boolean_OrOperation(
       Bool1=&Boolean_OrOperation(
        Bool1=$System.Dust_Wheel_RL
        Bool2=&Boolean_OrOperation(
         Bool1=$System.TireIntactRL
         Bool2=&SetBoolByFloatComparison(
          A=&SelectFloatFromArray(Direct Array Index=3, Float Selection Array=$User.WheelsTouchingArray)
          B=1
          Comparison Type=::NewEnumerator2
         )
        )
       )
       Bool2=&Boolean_OrOperation(
        Bool1=$System.Dust_Wheel_RR
        Bool2=&Boolean_OrOperation(
         Bool1=$System.TireIntactRR
         Bool2=&SetBoolByFloatComparison(
          A=&SelectFloatFromArray(Direct Array Index=2, Float Selection Array=$User.WheelsTouchingArray)
          B=1
          Comparison Type=::NewEnumerator2
         )
        )
       )
      )
      True Float=1
      False Float=0
     )
    )
    Use Spawn Probability=true
   ) {module=SpawnRate001}
   &SpawnRate(
    SpawnGroup=2
    SpawnRate=&Multiply_Float(
     A=&self:AmountToSpawn(
      " Min Amount":NiagaraFloat=16
      Max Amount:NiagaraFloat=40
      Total Amount:NiagaraInt32=$System.TotalDust
      Vehicle Speed:NiagaraFloat=&Abs_Float(Float:NiagaraFloat=$User.Vehicle_Speed)
      Vehicle Speed Exponent:NiagaraFloat=1.4
     )
     B=&Multiply_Float(
      A=&SelectFloatFromArray(Direct Array Index=$User.PeelOutColorIndex, Float Selection Array=DataInterface() {propertyValues {FloatData [0,0,0,0,0,1,1,0]}})
      B=&MakeCustomFloatFromBool(Bool=$System.VehicleHasDriver, True Float=1, False Float=0)
     )
    )
    Use Spawn Probability=false
   ) {module=SpawnRate002}
   &SpawnBurst_Instantaneous(Spawn Count=$System.TotalDust, Spawn Group=3, Use Spawn Probability=false, UseLoopCountLimit=false)
  ]
  particleSpawnScript [
   /SetVariables(Particles.VisibilityTag=&SpawnGroup()) {isSetParameters=true, module=SetVariables_#5}
   &PartitionParticles(Number Of Partitions=$System.TotalDust)
   /SetVariables(Particles.BoneSelection:NiagaraInt32=&self:TogglesBoneSet(
     Partitions:NiagaraInt32=$Particles.PartitionParticles.Partition
     Front Left Tire:NiagaraBool=&Boolean_OrOperation(
      Bool1=$System.Dust_Wheel_FL
      Bool2=&Boolean_OrOperation(
       Bool1=$System.TireIntactFL
       Bool2=&SetBoolByFloatComparison(
        A=&SelectFloatFromArray(Direct Array Index=1, Float Selection Array=$User.WheelsTouchingArray)
        B=1
        Comparison Type=::NewEnumerator2
       )
      )
     )
     Front Right Tire:NiagaraBool=&Boolean_OrOperation(
      Bool1=$System.Dust_Wheel_FR
      Bool2=&Boolean_OrOperation(
       Bool1=$System.TireIntactFR
       Bool2=&SetBoolByFloatComparison(
        A=&SelectFloatFromArray(Direct Array Index=0, Float Selection Array=$User.WheelsTouchingArray)
        B=1
        Comparison Type=::NewEnumerator2
       )
      )
     )
     Rear Left Tire:NiagaraBool=&Boolean_OrOperation(
      Bool1=$System.Dust_Wheel_RL
      Bool2=&Boolean_OrOperation(
       Bool1=$System.TireIntactRL
       Bool2=&SetBoolByFloatComparison(
        A=&SelectFloatFromArray(Direct Array Index=3, Float Selection Array=$User.WheelsTouchingArray)
        B=1
        Comparison Type=::NewEnumerator2
       )
      )
     )
     Rear Right Tire:NiagaraBool=&Boolean_OrOperation(
      Bool1=$System.Dust_Wheel_RR
      Bool2=&Boolean_OrOperation(
       Bool1=$System.TireIntactRR
       Bool2=&SetBoolByFloatComparison(
        A=&SelectFloatFromArray(Direct Array Index=2, Float Selection Array=$User.WheelsTouchingArray)
        B=1
        Comparison Type=::NewEnumerator2
       )
      )
     )
     Middle Left Tire:NiagaraBool=&Boolean_OrOperation(
      Bool1=$System.Dust_Wheel_ML
      Bool2=&Boolean_OrOperation(
       Bool1=$System.TireIntactML
       Bool2=&SetBoolByFloatComparison(
        A=&SelectFloatFromArray(Direct Array Index=5, Float Selection Array=$User.WheelsTouchingArray)
        B=1
        Comparison Type=::NewEnumerator2
       )
      )
     )
     Middle Right Tire:NiagaraBool=&Boolean_OrOperation(
      Bool1=$System.Dust_Wheel_MR
      Bool2=&Boolean_OrOperation(
       Bool1=$System.TireIntactMR
       Bool2=&SetBoolByFloatComparison(
        A=&SelectFloatFromArray(Direct Array Index=4, Float Selection Array=$User.WheelsTouchingArray)
        B=1
        Comparison Type=::NewEnumerator2
       )
      )
     )
    )) {isSetParameters=true, module=SetVariables_#6}
   &InitializeParticle(
    Color=&MaskLinearColorBySpawnGroup(
     Masked Linear Color=&RandomRangeLinearColor(Minimum=<rgba 0.2448 0.2448 0.2448 1>, Maximum=<rgba 1 1 1 1>)
     Passthrough Linear Color=&MaskLinearColorBySpawnGroup(
      Masked Linear Color=&SelectLinearColorFromArray(
       Array Sampling Mode=::NewEnumerator1
       Direct Array Index=$User.PeelOutColorIndex
       Color Selection Array=DataInterface() {propertyValues {ColorData [<rgba 0 0 0 1>, <rgba 0 0 0 1>, <rgba 0 0 0 1>, <rgba 0 0 0 1>, <rgba 0 0 0 1>, <rgba 1 1 1 1>, <rgba 0.3939 0.7539 1 1>]}}
       Random Seed=?
      )
      Passthrough Linear Color=&SelectLinearColorFromArray(
       Array Sampling Mode=::NewEnumerator1
       Direct Array Index=$User.PeelOutColorIndex
       Color Selection Array=$System.PeelOutColorArray
       Random Seed=?
      )
      Spawn Group Mask=2
     )
     Spawn Group Mask=1
    )
    Lifetime=&MaskFloatBySpawnGroup(
     Masked Float=&MakeCustomFloatFromBool(
      Bool=&Boolean_OrOperation(Bool1=$System.SpawnSnow, Bool2=$System.SpawnSand)
      True Float=&RandomRangeFloat(Minimum=1, Maximum=1.8)
      False Float=&Lerp_Float(
       A=&UniformRangedFloat(Minimum=1, Maximum=1.4)
       B=&UniformRangedFloat(Minimum=0.5, Maximum=0.3)
       Alpha=&Abs_Float(Float:NiagaraFloat=$User.Vehicle_Speed)
      )
     )
     Passthrough Float=&MaskFloatBySpawnGroup(
      Masked Float=&Lerp_Float(A=0.6, B=1.4, Alpha=&Abs_Float(Float:NiagaraFloat=$User.Vehicle_Speed))
      Passthrough Float=&MaskFloatBySpawnGroup(
       Masked Float=&RandomRangeFloat(Minimum=1, Maximum=2)
       Passthrough Float=&RandomRangeFloat(Minimum=0.1, Maximum=0.2)
       Spawn Group Mask=2
      )
      Spawn Group Mask=1
     )
    )
    Mesh Scale Mode=::NewEnumerator1
    Mesh Uniform Scale Max=0.25
    Mesh Uniform Scale Min=0.05
    Sprite Rotation Mode=::NewEnumerator2
    Sprite Size=&MaskVector2dBySpawnGroup(
     Masked Vector2D=&Vector2DFromFloat(Value=&Lerp_Float(
       A=&Lerp_Float(
        A=&UniformRangedFloat(Minimum=90, Maximum=150)
        B=&UniformRangedFloat(Minimum=300, Maximum=400)
        Alpha=&Abs_Float(Float:NiagaraFloat=$User.Vehicle_Speed)
       )
       B=550
       Alpha=&Abs_Float(Float:NiagaraFloat=$User.Slide_Direction)
      ))
     Passthrough Vector2D=&Vector2DFromFloat(Value=&RandomRangeFloat(Minimum=5, Maximum=9))
     Spawn Group Mask=0
    )
    Sprite Size Mode=::NewEnumerator4
    UsePositionOffset=false
    Write Material Random=true
   )
   /Niagara/Modules/Update/SubUV/V2/SubUVAnimation(End Frame=3, SubUV Animation Mode:ENiagara_SubUVLookupMode=::NewEnumerator2)
   &SkeletalMeshLocation(
    Bone / Socket Index=&MaskIntBySpawnGroup(
     Masked Int=&MakeCustomIntFromBool(
      Bool=&Boolean_AndOperation(
       Bool1=&Boolean_OrOperation(
        Bool1=$System.Dust_Wheel_RR
        Bool2=&Boolean_OrOperation(
         Bool1=$System.TireIntactRR
         Bool2=&SetBoolByFloatComparison(
          A=&SelectFloatFromArray(Direct Array Index=2, Float Selection Array=$User.WheelsTouchingArray)
          B=1
          Comparison Type=::NewEnumerator2
         )
        )
       )
       Bool2=&Boolean_OrOperation(
        Bool1=$System.Dust_Wheel_RL
        Bool2=&Boolean_OrOperation(
         Bool1=$System.TireIntactRL
         Bool2=&SetBoolByFloatComparison(
          A=&SelectFloatFromArray(Direct Array Index=3, Float Selection Array=$User.WheelsTouchingArray)
          B=1
          Comparison Type=::NewEnumerator2
         )
        )
       )
      )
      True INT=&MakeCustomIntFromBool(
       Bool=&SetBoolByIntComparison(A=$Particles.PartitionParticles.Partition, B=2, Comparison Type=::NewEnumerator0)
       True INT=3
       False INT=2
      )
      False INT=&MakeCustomIntFromBool(
       Bool=&Boolean_OrOperation(
        Bool1=$System.Dust_Wheel_RR
        Bool2=&Boolean_OrOperation(
         Bool1=$System.TireIntactRR
         Bool2=&SetBoolByFloatComparison(
          A=&SelectFloatFromArray(Direct Array Index=2, Float Selection Array=$User.WheelsTouchingArray)
          B=1
          Comparison Type=::NewEnumerator2
         )
        )
       )
       True INT=2
       False INT=&MakeCustomIntFromBool(
        Bool=&Boolean_OrOperation(
         Bool1=$System.Dust_Wheel_RL
         Bool2=&Boolean_OrOperation(
          Bool1=$System.TireIntactRL
          Bool2=&SetBoolByFloatComparison(
           A=&SelectFloatFromArray(Direct Array Index=3, Float Selection Array=$User.WheelsTouchingArray)
           B=1
           Comparison Type=::NewEnumerator2
          )
         )
        )
        True INT=3
        False INT=-1
       )
      )
     )
     Passthrough Int=$Particles.BoneSelection
     Spawn Group Mask=1
    )
    MaskSpawnGroup=false
    Mesh Sampling Type=::NewEnumerator5
    Orientation Sampling=::NewEnumerator1
    Sampled Position Offset=&MakeVector(
     X=0
     Y=&MaskFloatBySpawnGroup(
      Masked Float=&self:WheelWidthUniformRange(
       WheelWidthArray:NiagaraDataInterfaceArrayFloat=$User.Wheel_Width_Array
       WheelIndex:NiagaraInt32=&MakeCustomIntFromBool(
        Bool=&SetBoolByIntComparison(A=$Particles.PartitionParticles.Partition, B=1, Comparison Type=::NewEnumerator0)
        True INT=3
        False INT=2
       )
      )
      Passthrough Float=&self:WheelWidthUniformRange(WheelWidthArray:NiagaraDataInterfaceArrayFloat=$User.Wheel_Width_Array, WheelIndex:NiagaraInt32=$Particles.BoneSelection)
      Spawn Group Mask=1
     )
     Z=&MaskFloatBySpawnGroup(
      Masked Float=&self:WheelRadiusOffset(
       Wheel Radius Array:NiagaraDataInterfaceArrayFloat=$User.Wheel_Radius_Array
       Wheel Index:NiagaraInt32=&MakeCustomIntFromBool(
        Bool=&SetBoolByIntComparison(A=$Particles.PartitionParticles.Partition, B=1, Comparison Type=::NewEnumerator0)
        True INT=3
        False INT=2
       )
       Additional Offset:NiagaraFloat=0
       Wheel Radius Addl Scale:NiagaraFloat=1
      )
      Passthrough Float=&self:WheelRadiusOffset(
       Wheel Radius Array:NiagaraDataInterfaceArrayFloat=$User.Wheel_Radius_Array
       Wheel Index:NiagaraInt32=$Particles.BoneSelection
       Additional Offset:NiagaraFloat=0
       Wheel Radius Addl Scale:NiagaraFloat=1
      )
      Spawn Group Mask=1
     )
    )
    Skeletal Mesh=$System.Vehicle_Selection_02
    Socket Sampling Mode=::NewEnumerator5
   )
   &ShapeLocation(MaskSpawnGroup=true, Override Seed=false, Spawn Group Mask=2, Sphere Radius=18, UseOwnerScale=true)
   &AddVelocity(
    Cone Angle=12
    Cone Axis=&MaskVectorBySpawnGroup(
     Masked Vector=&MakeVector(X=1, Y=$User.Slide_Direction, Z=0)
     Passthrough Vector=&MakeVector(X=1, Y=$User.Slide_Direction, Z=0.25)
    )
    Distribution Along Cone Axis=0.6
    Module.UseConeFalloff=false
    Module.UseVelDistribution=true
    Override Seed=false
    Velocity Mode=::NewEnumerator2
    Velocity Speed=&MaskFloatBySpawnGroup(
     Masked Float=&Lerp_Float(A=0, B=1, Alpha=&Abs_Float(Float:NiagaraFloat=$User.Vehicle_Speed))
     Passthrough Float=&MaskFloatBySpawnGroup(
      Masked Float=&Lerp_Float(
       A=&RandomRangeFloat(Minimum=0, Maximum=1)
       B=&RandomRangeFloat(Minimum=300, Maximum=800)
       Alpha=&Abs_Float(Float:NiagaraFloat=$User.Vehicle_Speed)
      )
      Passthrough Float=0
      Spawn Group Mask=1
     )
    )
   ) {module=AddVelocity001}
   &AddVelocity(Velocity=&MaskVectorBySpawnGroup(
     Masked Vector=&MakeVector(
      X=0
      Y=&Multiply_Float(A=$User.Slide_Direction, B=300)
      Z=&Multiply_Float(
       A=&MakeCustomFloatFromBool(
        Bool=&Boolean_OrOperation(Bool1=$System.SpawnSnow, Bool2=$System.SpawnSand)
        True Float=&SelectFloatFromArray(
         Direct Array Index=&SpawnGroup()
         Float Selection Array=DataInterface() {propertyValues {FloatData [0,0,400,400,0,0]}}
        )
        False Float=0
       )
       B=&Lerp_Float(A=0, B=1, Alpha=&Abs_Float(Float:NiagaraFloat=$User.Vehicle_Speed))
      )
     )
     Passthrough Vector=&MaskVectorBySpawnGroup(
      Masked Vector=&MakeVector(X=0, Y=&Multiply_Float(A=$User.Slide_Direction, B=1800), Z=0)
      Passthrough Vector=&MaskVectorBySpawnGroup(
       Masked Vector=&RandomRangeVector(Minimum=<20 -100 300>, Maximum=&MakeVector(X=300, Y=100, Z=600))
       Passthrough Vector=<0 0 0>
       Spawn Group Mask=2
      )
      Spawn Group Mask=1
     )
    ))
   /SetVariables(Particles.Lifetime=&MakeCustomFloatFromBool(
     Bool=&SetBoolByIntComparison(A=&SpawnGroup(), B=3, Comparison Type=::NewEnumerator4)
     True Float=&Multiply_Float(A=$Particles.Lifetime, B=2)
     False Float=$Particles.Lifetime
    )) {isSetParameters=true}
  ]
  particleUpdateScript [
   &ParticleState()
   &GravityForce(Gravity=&MaskVectorBySpawnGroup(
     Masked Vector=&MakeVector(
      X=0
      Y=0
      Z=&MakeCustomFloatFromBool(Bool=&Boolean_OrOperation(Bool1=$System.SpawnSand, Bool2=$System.SpawnSnow), True Float=-800, False Float=50)
     )
     Passthrough Vector=&MaskVectorBySpawnGroup(Masked Vector=<0 0 0>, Passthrough Vector=<0 0 -980>, Spawn Group Mask=3)
    ))
   &Drag(
    Drag=&MaskFloatBySpawnGroup(
     Masked Float=&RandomRangeFloat(Minimum=0.5, Maximum=1.25)
     Passthrough Float=&MaskFloatBySpawnGroup(
      Masked Float=1
      Passthrough Float=&MaskFloatBySpawnGroup(Masked Float=0, Passthrough Float=0.5, Spawn Group Mask=3)
      Spawn Group Mask=2
     )
     Spawn Group Mask=1
    )
    Use Linear Drag=true
    Use Rotational Drag=false
   )
   &ScaleColor(
    Scale Alpha=&MaskFloatBySpawnGroup(
     Masked Float=&FloatFromCurve(
      FloatCurve=DataInterface() {propertyValues {Curve {keys [
          {arriveTangent=4.1785, interpMode=RCIM_Cubic, leaveTangent=4.1785, tangentMode=RCTM_User, time=0, value=0}
          {arriveTangent=0.9995, interpMode=RCIM_Cubic, leaveTangent=0.9995, tangentMode=RCTM_User, time=0.0178, value=1}
          {arriveTangent=-1.0182, interpMode=RCIM_Cubic, leaveTangent=-1.0182, time=0.5124, value=0.3091}
          {arriveTangent=-0.3824, interpMode=RCIM_Cubic, leaveTangent=-0.0405, tangentMode=RCTM_Break, time=1, value=0}
         ]}}}
      Scale Curve=0.7
     )
     Passthrough Float=&MaskFloatBySpawnGroup(
      Masked Float=&FloatFromCurve(
       FloatCurve=DataInterface() {propertyValues {Curve {keys [
           {arriveTangent=0.002, interpMode=RCIM_Cubic, leaveTangent=0.002, tangentMode=RCTM_Break, time=0.004, value=1.0113}
           <at 1.0011 0.002 RCIM_Cubic>
          ]}}}
       Scale Curve=5
      )
      Passthrough Float=&FloatFromCurve(
       FloatCurve=DataInterface() {propertyValues {Curve {keys [
           {interpMode=RCIM_Cubic, tangentMode=RCTM_Break, time=-0.0054, value=0.0052}
           {arriveTangent=10.4414, interpMode=RCIM_Cubic, leaveTangent=-2.1615, tangentMode=RCTM_Break, time=0.0811, value=1.0113}
           {arriveTangent=-0.7171, interpMode=RCIM_Cubic, leaveTangent=-0.7171, tangentMode=RCTM_User, time=0.3372, value=0.3367}
           <at 1.0011 0.002 RCIM_Cubic>
          ]}}}
       Scale Curve=&Multiply_Float(
        A=&MakeCustomFloatFromBool(
         Bool=&SetBoolByIntComparison(A=$User.PeelOutColorIndex, B=1, Comparison Type=::NewEnumerator4)
         True Float=0.4
         False Float=1
        )
        B=&Lerp_Float(A=0, B=1, Alpha=&ClampFloat(Float=&Abs_Float(Float:NiagaraFloat=$User.Vehicle_Speed)))
       )
      )
      Spawn Group Mask=3
     )
     Spawn Group Mask=2
    )
    ScaleA=true
    ScaleRGB=true
   )
   &InheritSourceMovement(
    Applied Position Scale=&MaskVectorBySpawnGroup(
     Masked Vector=<0 0 0>
     Passthrough Vector=&MaskVectorBySpawnGroup(Masked Vector=<0 0 0>, Passthrough Vector=<0.5 0.5 0.5>, Spawn Group Mask=3)
     Spawn Group Mask=1
    )
    ApplyForce=false
    ApplyPosition=true
    ApplyPositionFalloff=false
   )
   &UpdateMeshOrientation(
    Rotation Rate=&MaskFloatBySpawnGroup(Masked Float=0, Passthrough Float=1, Spawn Group Mask=3)
    Rotation Vector=&RandomRangeVector(Minimum=<0 0 0>, Maximum=<1 1 1>)
   )
   &SolveForcesAndVelocity(Clamp Velocity=false, Limit Acceleration=false)
   &ScaleSpriteSize(Scale Factor=&Vector2DFromFloat(Value=&MaskFloatBySpawnGroup(
      Masked Float=&FloatFromCurve(FloatCurve=DataInterface() {propertyValues {Curve {keys [<at 0 1>,<at 1 0>]}}})
      Passthrough Float=&FloatFromCurve(
       FloatCurve=DataInterface() {propertyValues {Curve {keys [{arriveTangent=2.0487, interpMode=RCIM_Cubic, leaveTangent=2.0487, tangentMode=RCTM_User, time=0, value=0.15}, <at 1 1>]}}}
       Scale Curve=&MakeCustomFloatFromBool(
        Bool=&SetBoolByIntComparison(A=$User.PeelOutColorIndex, B=1, Comparison Type=::NewEnumerator4)
        True Float=0.4
        False Float=1
       )
      )
      Spawn Group Mask=2
     )))
   &SpriteRotationRate(Rotation Rate=&Multiply_Float(A=$Particles.Initial.MaterialRandom, B=1))
   &DynamicMaterialParameters(
    Index 0 Param 1=&Add_Float(
     A=&RandomRangeFloat(Minimum=0, Maximum=1)
     B=&FloatFromCurve(FloatCurve=DataInterface() {propertyValues {Curve {keys [
          {arriveTangent=2.7533, interpMode=RCIM_Cubic, leaveTangent=2.847, tangentMode=RCTM_Break, time=0, value=0}
          {arriveTangent=1.3581, interpMode=RCIM_Cubic, leaveTangent=1.3581, tangentMode=RCTM_User, time=0.2558, value=0.5125}
          {arriveTangent=0.0529, interpMode=RCIM_Cubic, leaveTangent=0.2021, tangentMode=RCTM_User, time=1, value=1}
         ]}}})
    )
    Index 0 Param 3=&FloatFromCurve(FloatCurve=DataInterface() {propertyValues {Curve {keys [{arriveTangent=-53.9199, time=0, value=0}, {leaveTangent=0.2723, time=0.25, value=0}]}}})
    Index 0 Param 4=&FloatFromCurve(FloatCurve=DataInterface() {propertyValues {Curve {keys [<at 0 1>,<at 1 0>]}}})
    Index 1 Param 1=&FloatFromCurve(FloatCurve=DataInterface() {propertyValues {Curve {keys [<at 0 0>, {arriveTangent=4.2257, interpMode=RCIM_Cubic, leaveTangent=4.2257, tangentMode=RCTM_User, time=0.25, value=0}]}}})
    Index 1 Param 2=&FloatFromCurve(FloatCurve=DataInterface() {propertyValues {Curve {keys [<at 0 0>]}}})
    Index 1 Param 3=&FloatFromCurve(FloatCurve=DataInterface() {propertyValues {Curve {keys [{interpMode=RCIM_Cubic, leaveTangent=6.3917, tangentMode=RCTM_Break, time=0, value=0}]}}})
    Index 1 Param 4=&FloatFromCurve(FloatCurve=DataInterface() {propertyValues {Curve {keys [{arriveTangent=-0.136, leaveTangent=-0.136, time=0, value=1}]}}})
    Index 2 Param 1=&FloatFromCurve(FloatCurve=DataInterface() {propertyValues {Curve {keys [
         {arriveTangent=2.646, interpMode=RCIM_Cubic, leaveTangent=2.646, tangentMode=RCTM_User, time=0, value=0}
         {leaveTangent=-0.0719, time=1, value=1}
        ]}}})
    Index 2 Param 2=&FloatFromCurve(
     FloatCurve=DataInterface() {propertyValues {Curve {keys [
         {arriveTangent=1.9056, time=0, value=0.2}
         {arriveTangent=0.0259, interpMode=RCIM_Cubic, leaveTangent=0.0259, tangentMode=RCTM_User, time=1, value=1.4}
        ]}}}
     Scale Curve=&RandomRangeFloat(Minimum=0.5, Maximum=1)
    )
    Index 2 Param 3=&FloatFromCurve(FloatCurve=DataInterface() {propertyValues {Curve {keys [
         {interpMode=RCIM_Cubic, leaveTangent=2.3064, tangentMode=RCTM_Break, time=0, value=1}
         {interpMode=RCIM_Cubic, leaveTangent=11.1233, tangentMode=RCTM_Break, time=1, value=2}
        ]}}})
    Index 2 Param 4=&FloatFromCurve(
     FloatCurve=DataInterface() {propertyValues {Curve {keys [
         {arriveTangent=-0.6867, interpMode=RCIM_Cubic, leaveTangent=-0.6867, tangentMode=RCTM_User, time=0, value=1}
         {interpMode=RCIM_Cubic, leaveTangent=-1.2956, tangentMode=RCTM_Break, time=1, value=0.75}
        ]}}}
     Scale Curve=1.6
    )
    Module.Param0WriteEnabled=true
    Module.Param0WriteEnabled1=true
    Module.Param0WriteEnabled2=true
    Module.Param1WriteEnabled=false
    Module.Param1WriteEnabled1=true
    Module.Param1WriteEnabled2=true
    Module.Param2WriteEnabled=true
    Module.Param2WriteEnabled1=true
    Module.Param2WriteEnabled2=true
    Module.Param3WriteEnabled=true
    Module.Param3WriteEnabled1=true
    Module.Param3WriteEnabled2=true
    Write Parameter Index 1=true
    Write Parameter Index 2=true
   )
   &ScaleMeshSizeBySpeed(
    Max Scale Factor=<1 1 1>
    Min Scale Factor:Vector3f=&MaskVectorBySpawnGroup(Masked Vector=<1 1 1>, Passthrough Vector=<0.7 0.7 0.7>, Spawn Group Mask=3)
   )
   &ScaleMeshSize(Scale Factor=&VectorFromFloat(Value=&MaskFloatBySpawnGroup(
      Masked Float=1
      Passthrough Float=&FloatFromCurve(FloatCurve=DataInterface() {propertyValues {Curve {keys [
           <at 0 0.075>
           <at 0.0756 1>
           {arriveTangent=-0.2, interpMode=RCIM_Cubic, leaveTangent=-0.2, time=0.7956, value=1.0207}
           {arriveTangent=-5.4083, interpMode=RCIM_Cubic, leaveTangent=-5.4083, tangentMode=RCTM_User, time=1, value=0}
          ]}}})
      Spawn Group Mask=3
     )))
   &SkeletalMeshLocation(
    Bone / Socket Index=$Particles.BoneSelection
    MaskSpawnGroup=true
    Mesh Sampling Type=::NewEnumerator5
    Sampled Position Offset=&SelectVectorFromArray(
     Direct Array Index=$Particles.BoneSelection
     Vector Selection Array=DataInterface() {propertyValues {FloatData [<0 0 0>, <0 0 0>, <-5 0 -2>, <-5 0 -2>, <0 0 0>, <0 0 0>]}}
    )
    Skeletal Mesh=$System.Vehicle_Selection_02
    Socket Sampling Mode=::NewEnumerator5
    Spawn Group Mask=3
    Use Sampled Mesh Attribute as Ribbon ID=false
    Velocity Sampling=::NewEnumerator0
    VelocitySpeedLimit=false
   ) {enabled=false, module=SkeletalMeshLocation003}
   &SkeletalMeshLocation(
    Bone / Socket Index=$Particles.BoneSelection
    MaskSpawnGroup=false
    Mesh Sampling Type=::NewEnumerator5
    Orientation Sampling=::NewEnumerator1
    Sampled Position Offset=&SelectVectorFromArray(
     Direct Array Index=$Particles.BoneSelection
     Vector Selection Array=DataInterface() {propertyValues {FloatData [<0 0 0>, <0 0 0>, <0 0 0>, <0 0 0>, <0 0 0>, <0 0 0>]}}
    )
    Skeletal Mesh=$System.Vehicle_Selection_02
    Socket Sampling Mode=::NewEnumerator5
    Spawn Group Mask=3
    Use Sampled Mesh Attribute as Ribbon ID=false
    Write To Particle Position=false
   ) {module=SkeletalMeshLocation001}
   /SetVariables(Particles.DustWheel Dynamic Parameter:Vector4f=&MakeVector4(X=5, Y=1, Z=0, W=0)) {isSetParameters=true, module=SetVariables_#7}
   /SetVariables(Particles.WheelScale:Vector3f=&self:ScaleMeshfromWheelDimensions(
     Mesh Unscaled Size:Vector3f=<49 49 49>
     Wheel Radius Array:NiagaraDataInterfaceArrayFloat=$User.Wheel_Radius_Array
     Wheel Width Array:NiagaraDataInterfaceArrayFloat=$User.Wheel_Width_Array
     Wheel Index:NiagaraInt32=$Particles.BoneSelection
     Additional Scaling:Vector3f=<1 1 1>
    )) {isSetParameters=true, module=SetVariables_#8}
   &ScaleColor(
    Scale Alpha=&Lerp_Float(A=0, B=1, Alpha=&Abs_Float(Float:NiagaraFloat=$User.Vehicle_Speed))
    ScaleA=true
    ScaleRGB=false
   ) {module=ScaleColor001}
   /SetVariables(
    Particles.MeshOrientation=&self:ScratchDynamicInput(
     Quat:Quat4f=$Particles.SkeletalMeshLocation001.SampledRotation
     Mask Index Attribute:NiagaraInt32=&SpawnGroup()
     Mask Index:NiagaraInt32=3
    )
    Particles.Position=&MaskVectorBySpawnGroup(
     Masked Vector=$StackContext.SkeletalMeshLocation001.SampledPosition
     Passthrough Vector=Linked() {linkedVariable {name=Particles.Position, type=NiagaraPosition}}
     Spawn Group Mask=3
    )
   ) {isSetParameters=true}
  ]
  rendererClasses [&NiagaraSpriteRendererProperties, &NiagaraMeshRendererProperties]
  renderers [
   {rendererClass=&NiagaraSpriteRendererProperties}
   {rendererClass=&NiagaraMeshRendererProperties, rendererIndex=1}
   {rendererClass=&NiagaraSpriteRendererProperties, rendererIndex=2}
   {rendererClass=&NiagaraMeshRendererProperties, rendererIndex=3}
  ]
 }
 LOW_Dust_Tires_World {
  emitterUpdateScript [
   &EmitterState()
   &SpawnRate(
    SpawnRate=&Multiply_Float(
     A=&self:AmountToSpawn(
      " Min Amount":NiagaraFloat=7
      Max Amount:NiagaraFloat=28
      Total Amount:NiagaraInt32=$System.TotalDust
      Vehicle Speed:NiagaraFloat=&Abs_Float(Float:NiagaraFloat=$User.Vehicle_Speed)
      Vehicle Speed Exponent:NiagaraFloat=&MakeCustomFloatFromBool(Bool=$System.SpawnSand, True Float=0.9, False Float=1.4)
     )
     B=&MakeCustomFloatFromBool(
      Bool=&SetBoolByIntComparison(A=$User.PeelOutColorIndex, B=1, Comparison Type=::NewEnumerator2)
      True Float=&FloatFromCurve(FloatCurve=$System.Dust_SpawnCurve, CurveIndex=&Abs_Float(Float:NiagaraFloat=$User.Vehicle_Speed))
      False Float=0
     )
    )
    Use Spawn Probability=false
   )
   &SpawnRate(
    SpawnGroup=2
    SpawnRate=&Multiply_Float(
     A=&self:AmountToSpawn(
      " Min Amount":NiagaraFloat=16
      Max Amount:NiagaraFloat=40
      Total Amount:NiagaraInt32=$System.TotalDust
      Vehicle Speed:NiagaraFloat=&Abs_Float(Float:NiagaraFloat=$User.Vehicle_Speed)
      Vehicle Speed Exponent:NiagaraFloat=1.4
     )
     B=&Multiply_Float(
      A=&SelectFloatFromArray(Direct Array Index=$User.PeelOutColorIndex, Float Selection Array=DataInterface() {propertyValues {FloatData [0,0,0,0,0,1,1,0]}})
      B=&MakeCustomFloatFromBool(Bool=$System.VehicleHasDriver, True Float=1, False Float=0)
     )
    )
    Use Spawn Probability=false
   ) {module=SpawnRate002}
   &SpawnBurst_Instantaneous(Spawn Count=$System.TotalDust, Spawn Group=3, Use Spawn Probability=false, UseLoopCountLimit=false)
  ]
  particleSpawnScript [
   /SetVariables(Particles.VisibilityTag=&SpawnGroup()) {isSetParameters=true, module=SetVariables_#5}
   &PartitionParticles(Number Of Partitions=$System.TotalDust)
   /SetVariables(Particles.BoneSelection:NiagaraInt32=&self:TogglesBoneSet(
     Partitions:NiagaraInt32=$Particles.PartitionParticles.Partition
     Front Left Tire:NiagaraBool=&Boolean_OrOperation(
      Bool1=$System.Dust_Wheel_FL
      Bool2=&Boolean_OrOperation(
       Bool1=$System.TireIntactFL
       Bool2=&SetBoolByFloatComparison(
        A=&SelectFloatFromArray(Direct Array Index=1, Float Selection Array=$User.WheelsTouchingArray)
        B=1
        Comparison Type=::NewEnumerator2
       )
      )
     )
     Front Right Tire:NiagaraBool=&Boolean_OrOperation(
      Bool1=$System.Dust_Wheel_FR
      Bool2=&Boolean_OrOperation(
       Bool1=$System.TireIntactFR
       Bool2=&SetBoolByFloatComparison(
        A=&SelectFloatFromArray(Direct Array Index=0, Float Selection Array=$User.WheelsTouchingArray)
        B=1
        Comparison Type=::NewEnumerator2
       )
      )
     )
     Rear Left Tire:NiagaraBool=&Boolean_OrOperation(
      Bool1=$System.Dust_Wheel_RL
      Bool2=&Boolean_OrOperation(
       Bool1=$System.TireIntactRL
       Bool2=&SetBoolByFloatComparison(
        A=&SelectFloatFromArray(Direct Array Index=3, Float Selection Array=$User.WheelsTouchingArray)
        B=1
        Comparison Type=::NewEnumerator2
       )
      )
     )
     Rear Right Tire:NiagaraBool=&Boolean_OrOperation(
      Bool1=$System.Dust_Wheel_RR
      Bool2=&Boolean_OrOperation(
       Bool1=$System.TireIntactRR
       Bool2=&SetBoolByFloatComparison(
        A=&SelectFloatFromArray(Direct Array Index=2, Float Selection Array=$User.WheelsTouchingArray)
        B=1
        Comparison Type=::NewEnumerator2
       )
      )
     )
     Middle Left Tire:NiagaraBool=&Boolean_OrOperation(
      Bool1=$System.Dust_Wheel_ML
      Bool2=&Boolean_OrOperation(
       Bool1=$System.TireIntactML
       Bool2=&SetBoolByFloatComparison(
        A=&SelectFloatFromArray(Direct Array Index=5, Float Selection Array=$User.WheelsTouchingArray)
        B=1
        Comparison Type=::NewEnumerator2
       )
      )
     )
     Middle Right Tire:NiagaraBool=&Boolean_OrOperation(
      Bool1=$System.Dust_Wheel_MR
      Bool2=&Boolean_OrOperation(
       Bool1=$System.TireIntactMR
       Bool2=&SetBoolByFloatComparison(
        A=&SelectFloatFromArray(Direct Array Index=4, Float Selection Array=$User.WheelsTouchingArray)
        B=1
        Comparison Type=::NewEnumerator2
       )
      )
     )
    )) {isSetParameters=true, module=SetVariables_#6}
   &InitializeParticle(
    Color=&MaskLinearColorBySpawnGroup(
     Masked Linear Color=&RandomRangeLinearColor(Minimum=<rgba 0.2448 0.2448 0.2448 1>, Maximum=<rgba 1 1 1 1>)
     Passthrough Linear Color=&MaskLinearColorBySpawnGroup(
      Masked Linear Color=&SelectLinearColorFromArray(
       Array Sampling Mode=::NewEnumerator1
       Direct Array Index=$User.PeelOutColorIndex
       Color Selection Array=DataInterface() {propertyValues {ColorData [<rgba 0 0 0 1>, <rgba 0 0 0 1>, <rgba 0 0 0 1>, <rgba 0 0 0 1>, <rgba 0 0 0 1>, <rgba 1 1 1 1>, <rgba 0.3939 0.7539 1 1>]}}
       Random Seed=?
      )
      Passthrough Linear Color=&SelectLinearColorFromArray(
       Array Sampling Mode=::NewEnumerator1
       Direct Array Index=$User.PeelOutColorIndex
       Color Selection Array=$System.PeelOutColorArray
       Random Seed=?
      )
      Spawn Group Mask=2
     )
     Spawn Group Mask=1
    )
    Lifetime=&MaskFloatBySpawnGroup(
     Masked Float=&MakeCustomFloatFromBool(
      Bool=&Boolean_OrOperation(Bool1=$System.SpawnSnow, Bool2=$System.SpawnSand)
      True Float=&RandomRangeFloat(Minimum=1.5, Maximum=2)
      False Float=&Lerp_Float(
       A=&UniformRangedFloat(Minimum=1, Maximum=1.4)
       B=&UniformRangedFloat(Minimum=0.5, Maximum=0.3)
       Alpha=&Abs_Float(Float:NiagaraFloat=$User.Vehicle_Speed)
      )
     )
     Passthrough Float=&MaskFloatBySpawnGroup(
      Masked Float=&Lerp_Float(A=0.6, B=1.4, Alpha=&Abs_Float(Float:NiagaraFloat=$User.Vehicle_Speed))
      Passthrough Float=&MaskFloatBySpawnGroup(
       Masked Float=&RandomRangeFloat(Minimum=1, Maximum=2)
       Passthrough Float=&RandomRangeFloat(Minimum=0.1, Maximum=0.2)
       Spawn Group Mask=2
      )
      Spawn Group Mask=1
     )
    )
    Mesh Scale Mode=::NewEnumerator1
    Mesh Uniform Scale Max=0.25
    Mesh Uniform Scale Min=0.05
    Sprite Rotation Mode=::NewEnumerator2
    Sprite Size=&MaskVector2dBySpawnGroup(
     Masked Vector2D=&Vector2DFromFloat(Value=&Lerp_Float(
       A=&Lerp_Float(
        A=&UniformRangedFloat(Minimum=90, Maximum=150)
        B=&UniformRangedFloat(Minimum=300, Maximum=400)
        Alpha=&Abs_Float(Float:NiagaraFloat=$User.Vehicle_Speed)
       )
       B=550
       Alpha=&Abs_Float(Float:NiagaraFloat=$User.Slide_Direction)
      ))
     Passthrough Vector2D=&Vector2DFromFloat(Value=&RandomRangeFloat(Minimum=5, Maximum=9))
     Spawn Group Mask=0
    )
    Sprite Size Mode=::NewEnumerator4
    UsePositionOffset=false
    Write Material Random=true
   )
   &SubUVAnimation(End Frame=3, SubUV Animation Mode:ENiagara_SubUVLookupMode=::NewEnumerator2)
   &SkeletalMeshLocation(
    Bone / Socket Index=&MaskIntBySpawnGroup(
     Masked Int=&MakeCustomIntFromBool(
      Bool=&Boolean_AndOperation(
       Bool1=&Boolean_OrOperation(
        Bool1=$System.Dust_Wheel_RR
        Bool2=&Boolean_OrOperation(
         Bool1=$System.TireIntactRR
         Bool2=&SetBoolByFloatComparison(
          A=&SelectFloatFromArray(Direct Array Index=2, Float Selection Array=$User.WheelsTouchingArray)
          B=1
          Comparison Type=::NewEnumerator2
         )
        )
       )
       Bool2=&Boolean_OrOperation(
        Bool1=$System.Dust_Wheel_RL
        Bool2=&Boolean_OrOperation(
         Bool1=$System.TireIntactRL
         Bool2=&SetBoolByFloatComparison(
          A=&SelectFloatFromArray(Direct Array Index=3, Float Selection Array=$User.WheelsTouchingArray)
          B=1
          Comparison Type=::NewEnumerator2
         )
        )
       )
      )
      True INT=&MakeCustomIntFromBool(
       Bool=&SetBoolByIntComparison(A=$Particles.PartitionParticles.Partition, B=2, Comparison Type=::NewEnumerator0)
       True INT=3
       False INT=2
      )
      False INT=&MakeCustomIntFromBool(
       Bool=&Boolean_OrOperation(
        Bool1=$System.Dust_Wheel_RR
        Bool2=&Boolean_OrOperation(
         Bool1=$System.TireIntactRR
         Bool2=&SetBoolByFloatComparison(
          A=&SelectFloatFromArray(Direct Array Index=2, Float Selection Array=$User.WheelsTouchingArray)
          B=1
          Comparison Type=::NewEnumerator2
         )
        )
       )
       True INT=2
       False INT=&MakeCustomIntFromBool(
        Bool=&Boolean_OrOperation(
         Bool1=$System.Dust_Wheel_RL
         Bool2=&Boolean_OrOperation(
          Bool1=$System.TireIntactRL
          Bool2=&SetBoolByFloatComparison(
           A=&SelectFloatFromArray(Direct Array Index=3, Float Selection Array=$User.WheelsTouchingArray)
           B=1
           Comparison Type=::NewEnumerator2
          )
         )
        )
        True INT=3
        False INT=-1
       )
      )
     )
     Passthrough Int=$Particles.BoneSelection
     Spawn Group Mask=1
    )
    MaskSpawnGroup=true
    Mesh Sampling Type=::NewEnumerator5
    Orientation Sampling=::NewEnumerator1
    Sampled Position Offset=&MakeVector(
     X=0
     Y=&MaskFloatBySpawnGroup(
      Masked Float=&self:WheelWidthUniformRange(
       WheelWidthArray:NiagaraDataInterfaceArrayFloat=$User.Wheel_Width_Array
       WheelIndex:NiagaraInt32=&MakeCustomIntFromBool(
        Bool=&SetBoolByIntComparison(A=$Particles.PartitionParticles.Partition, B=1, Comparison Type=::NewEnumerator0)
        True INT=3
        False INT=2
       )
      )
      Passthrough Float=&self:WheelWidthUniformRange(WheelWidthArray:NiagaraDataInterfaceArrayFloat=$User.Wheel_Width_Array, WheelIndex:NiagaraInt32=$Particles.BoneSelection)
      Spawn Group Mask=1
     )
     Z=&MaskFloatBySpawnGroup(
      Masked Float=&self:WheelRadiusOffset(
       Wheel Radius Array:NiagaraDataInterfaceArrayFloat=$User.Wheel_Radius_Array
       Wheel Index:NiagaraInt32=&MakeCustomIntFromBool(
        Bool=&SetBoolByIntComparison(A=$Particles.PartitionParticles.Partition, B=1, Comparison Type=::NewEnumerator0)
        True INT=3
        False INT=2
       )
       Additional Offset:NiagaraFloat=0
       Wheel Radius Addl Scale:NiagaraFloat=1
      )
      Passthrough Float=&self:WheelRadiusOffset(
       Wheel Radius Array:NiagaraDataInterfaceArrayFloat=$User.Wheel_Radius_Array
       Wheel Index:NiagaraInt32=$Particles.BoneSelection
       Additional Offset:NiagaraFloat=0
       Wheel Radius Addl Scale:NiagaraFloat=1
      )
      Spawn Group Mask=1
     )
    )
    Skeletal Mesh=$System.Vehicle_Selection_02
    Socket Sampling Mode=::NewEnumerator5
   )
   &ShapeLocation(MaskSpawnGroup=true, Override Seed=false, Spawn Group Mask=2, Sphere Radius=18, UseOwnerScale=true)
   &AddVelocity(
    Cone Angle=12
    Cone Axis=&MaskVectorBySpawnGroup(
     Masked Vector=&MakeVector(X=1, Y=$User.Slide_Direction, Z=0)
     Passthrough Vector=&MakeVector(X=1, Y=$User.Slide_Direction, Z=0.25)
    )
    Distribution Along Cone Axis=0.6
    Module.UseConeFalloff=false
    Module.UseVelDistribution=true
    Override Seed=false
    Velocity Mode=::NewEnumerator2
    Velocity Speed=&MaskFloatBySpawnGroup(
     Masked Float=&Lerp_Float(A=0, B=1, Alpha=&Abs_Float(Float:NiagaraFloat=$User.Vehicle_Speed))
     Passthrough Float=&MaskFloatBySpawnGroup(
      Masked Float=&Lerp_Float(
       A=&RandomRangeFloat(Minimum=0, Maximum=1)
       B=&RandomRangeFloat(Minimum=300, Maximum=800)
       Alpha=&Abs_Float(Float:NiagaraFloat=$User.Vehicle_Speed)
      )
      Passthrough Float=0
      Spawn Group Mask=1
     )
    )
   ) {module=AddVelocity001}
   &AddVelocity(Velocity=&MaskVectorBySpawnGroup(
     Masked Vector=&MakeVector(
      X=0
      Y=&Multiply_Float(A=$User.Slide_Direction, B=300)
      Z=&Multiply_Float(
       A=&MakeCustomFloatFromBool(
        Bool=&Boolean_OrOperation(Bool1=$System.SpawnSnow, Bool2=$System.SpawnSand)
        True Float=&SelectFloatFromArray(
         Direct Array Index=&SpawnGroup()
         Float Selection Array=DataInterface() {propertyValues {FloatData [0,0,400,400,0,0]}}
        )
        False Float=0
       )
       B=&Lerp_Float(A=0, B=1, Alpha=&Abs_Float(Float:NiagaraFloat=$User.Vehicle_Speed))
      )
     )
     Passthrough Vector=&MaskVectorBySpawnGroup(
      Masked Vector=&MakeVector(X=0, Y=&Multiply_Float(A=$User.Slide_Direction, B=1800), Z=0)
      Passthrough Vector=&MaskVectorBySpawnGroup(
       Masked Vector=&RandomRangeVector(Minimum=<20 -100 300>, Maximum=&MakeVector(X=300, Y=100, Z=600))
       Passthrough Vector=<0 0 0>
       Spawn Group Mask=2
      )
      Spawn Group Mask=1
     )
    ))
   /SetVariables(Particles.Lifetime=&MakeCustomFloatFromBool(
     Bool=&SetBoolByIntComparison(A=&SpawnGroup(), B=3, Comparison Type=::NewEnumerator4)
     True Float=&Multiply_Float(A=$Particles.Lifetime, B=2)
     False Float=$Particles.Lifetime
    )) {isSetParameters=true}
  ]
  particleUpdateScript [
   &ParticleState()
   &GravityForce(Gravity=&MaskVectorBySpawnGroup(
     Masked Vector=&MakeVector(
      X=0
      Y=0
      Z=&MakeCustomFloatFromBool(Bool=&Boolean_OrOperation(Bool1=$System.SpawnSand, Bool2=$System.SpawnSnow), True Float=-800, False Float=50)
     )
     Passthrough Vector=&MaskVectorBySpawnGroup(Masked Vector=<0 0 0>, Passthrough Vector=<0 0 -980>, Spawn Group Mask=3)
    ))
   &Drag(
    Drag=&MaskFloatBySpawnGroup(
     Masked Float=&RandomRangeFloat(Minimum=0.5, Maximum=1.25)
     Passthrough Float=&MaskFloatBySpawnGroup(
      Masked Float=1
      Passthrough Float=&MaskFloatBySpawnGroup(Masked Float=0, Passthrough Float=0.5, Spawn Group Mask=3)
      Spawn Group Mask=2
     )
     Spawn Group Mask=1
    )
    Use Linear Drag=true
    Use Rotational Drag=false
   )
   &ScaleColor(
    Scale Alpha=&MaskFloatBySpawnGroup(
     Masked Float=&FloatFromCurve(
      FloatCurve=DataInterface() {propertyValues {Curve {keys [
          {arriveTangent=4.1785, interpMode=RCIM_Cubic, leaveTangent=4.1785, tangentMode=RCTM_User, time=0, value=0}
          {arriveTangent=0.9995, interpMode=RCIM_Cubic, leaveTangent=0.9995, tangentMode=RCTM_User, time=0.0178, value=1}
          {arriveTangent=-1.0182, interpMode=RCIM_Cubic, leaveTangent=-1.0182, time=0.5124, value=0.3091}
          {arriveTangent=-0.3824, interpMode=RCIM_Cubic, leaveTangent=-0.0405, tangentMode=RCTM_Break, time=1, value=0}
         ]}}}
      Scale Curve=0.7
     )
     Passthrough Float=&MaskFloatBySpawnGroup(
      Masked Float=&FloatFromCurve(
       FloatCurve=DataInterface() {propertyValues {Curve {keys [
           {arriveTangent=0.002, interpMode=RCIM_Cubic, leaveTangent=0.002, tangentMode=RCTM_Break, time=0.004, value=1.0113}
           <at 1.0011 0.002 RCIM_Cubic>
          ]}}}
       Scale Curve=5
      )
      Passthrough Float=&FloatFromCurve(
       FloatCurve=DataInterface() {propertyValues {Curve {keys [
           {interpMode=RCIM_Cubic, tangentMode=RCTM_Break, time=-0.0054, value=0.0052}
           {arriveTangent=10.4414, interpMode=RCIM_Cubic, leaveTangent=-2.1615, tangentMode=RCTM_Break, time=0.0811, value=1.0113}
           {arriveTangent=-0.7171, interpMode=RCIM_Cubic, leaveTangent=-0.7171, tangentMode=RCTM_User, time=0.3372, value=0.3367}
           <at 1.0011 0.002 RCIM_Cubic>
          ]}}}
       Scale Curve=&Multiply_Float(
        A=&MakeCustomFloatFromBool(
         Bool=&SetBoolByIntComparison(A=$User.PeelOutColorIndex, B=1, Comparison Type=::NewEnumerator4)
         True Float=0.4
         False Float=1
        )
        B=&Lerp_Float(A=0, B=1, Alpha=&ClampFloat(Float=&Abs_Float(Float:NiagaraFloat=$User.Vehicle_Speed)))
       )
      )
      Spawn Group Mask=3
     )
     Spawn Group Mask=2
    )
    ScaleA=true
    ScaleRGB=true
   )
   &InheritSourceMovement(
    Applied Position Scale=&MaskVectorBySpawnGroup(
     Masked Vector=<0 0 0>
     Passthrough Vector=&MaskVectorBySpawnGroup(Masked Vector=<0 0 0>, Passthrough Vector=<0.5 0.5 0.5>, Spawn Group Mask=3)
     Spawn Group Mask=1
    )
    ApplyForce=false
    ApplyPosition=true
    ApplyPositionFalloff=false
   )
   &UpdateMeshOrientation(
    Rotation Rate=&MaskFloatBySpawnGroup(Masked Float=0, Passthrough Float=1, Spawn Group Mask=3)
    Rotation Vector=&RandomRangeVector(Minimum=<0 0 0>, Maximum=<1 1 1>)
   )
   &SolveForcesAndVelocity(Clamp Velocity=false, Limit Acceleration=false)
   &ScaleSpriteSize(Scale Factor=&Vector2DFromFloat(Value=&MaskFloatBySpawnGroup(
      Masked Float=&FloatFromCurve(FloatCurve=DataInterface() {propertyValues {Curve {keys [<at 0 1>,<at 1 0>]}}})
      Passthrough Float=&FloatFromCurve(
       FloatCurve=DataInterface() {propertyValues {Curve {keys [{arriveTangent=2.0487, interpMode=RCIM_Cubic, leaveTangent=2.0487, tangentMode=RCTM_User, time=0, value=0.15}, <at 1 1>]}}}
       Scale Curve=&MakeCustomFloatFromBool(
        Bool=&SetBoolByIntComparison(A=$User.PeelOutColorIndex, B=1, Comparison Type=::NewEnumerator4)
        True Float=0.4
        False Float=1
       )
      )
      Spawn Group Mask=2
     )))
   &SpriteRotationRate(Rotation Rate=&Multiply_Float(A=$Particles.Initial.MaterialRandom, B=1))
   &DynamicMaterialParameters(
    Index 0 Param 1=&Add_Float(
     A=&RandomRangeFloat(Minimum=0, Maximum=1)
     B=&FloatFromCurve(FloatCurve=DataInterface() {propertyValues {Curve {keys [
          {arriveTangent=2.7533, interpMode=RCIM_Cubic, leaveTangent=2.847, tangentMode=RCTM_Break, time=0, value=0}
          {arriveTangent=1.3581, interpMode=RCIM_Cubic, leaveTangent=1.3581, tangentMode=RCTM_User, time=0.2558, value=0.5125}
          {arriveTangent=0.0529, interpMode=RCIM_Cubic, leaveTangent=0.2021, tangentMode=RCTM_User, time=1, value=1}
         ]}}})
    )
    Index 0 Param 3=&FloatFromCurve(FloatCurve=DataInterface() {propertyValues {Curve {keys [{arriveTangent=-53.9199, time=0, value=0}, {leaveTangent=0.2723, time=0.25, value=0}]}}})
    Index 0 Param 4=&FloatFromCurve(FloatCurve=DataInterface() {propertyValues {Curve {keys [<at 0 1>,<at 1 0>]}}})
    Index 1 Param 1=&FloatFromCurve(FloatCurve=DataInterface() {propertyValues {Curve {keys [<at 0 0>, {arriveTangent=4.2257, interpMode=RCIM_Cubic, leaveTangent=4.2257, tangentMode=RCTM_User, time=0.25, value=0}]}}})
    Index 1 Param 2=&FloatFromCurve(FloatCurve=DataInterface() {propertyValues {Curve {keys [<at 0 0>]}}})
    Index 1 Param 3=&FloatFromCurve(FloatCurve=DataInterface() {propertyValues {Curve {keys [{interpMode=RCIM_Cubic, leaveTangent=6.3917, tangentMode=RCTM_Break, time=0, value=0}]}}})
    Index 1 Param 4=&FloatFromCurve(FloatCurve=DataInterface() {propertyValues {Curve {keys [{arriveTangent=-0.136, leaveTangent=-0.136, time=0, value=1}]}}})
    Index 2 Param 1=&FloatFromCurve(FloatCurve=DataInterface() {propertyValues {Curve {keys [
         {arriveTangent=2.646, interpMode=RCIM_Cubic, leaveTangent=2.646, tangentMode=RCTM_User, time=0, value=0}
         {leaveTangent=-0.0719, time=1, value=1}
        ]}}})
    Index 2 Param 2=&FloatFromCurve(
     FloatCurve=DataInterface() {propertyValues {Curve {keys [
         {arriveTangent=1.9056, time=0, value=0.2}
         {arriveTangent=0.0259, interpMode=RCIM_Cubic, leaveTangent=0.0259, tangentMode=RCTM_User, time=1, value=1.4}
        ]}}}
     Scale Curve=&RandomRangeFloat(Minimum=0.5, Maximum=1)
    )
    Index 2 Param 3=&FloatFromCurve(FloatCurve=DataInterface() {propertyValues {Curve {keys [
         {interpMode=RCIM_Cubic, leaveTangent=2.3064, tangentMode=RCTM_Break, time=0, value=1}
         {interpMode=RCIM_Cubic, leaveTangent=11.1233, tangentMode=RCTM_Break, time=1, value=2}
        ]}}})
    Index 2 Param 4=&FloatFromCurve(
     FloatCurve=DataInterface() {propertyValues {Curve {keys [
         {arriveTangent=-0.6867, interpMode=RCIM_Cubic, leaveTangent=-0.6867, tangentMode=RCTM_User, time=0, value=1}
         {interpMode=RCIM_Cubic, leaveTangent=-1.2956, tangentMode=RCTM_Break, time=1, value=0.75}
        ]}}}
     Scale Curve=1.6
    )
    Module.Param0WriteEnabled=true
    Module.Param0WriteEnabled1=true
    Module.Param0WriteEnabled2=true
    Module.Param1WriteEnabled=false
    Module.Param1WriteEnabled1=true
    Module.Param1WriteEnabled2=true
    Module.Param2WriteEnabled=true
    Module.Param2WriteEnabled1=true
    Module.Param2WriteEnabled2=true
    Module.Param3WriteEnabled=true
    Module.Param3WriteEnabled1=true
    Module.Param3WriteEnabled2=true
    Write Parameter Index 1=true
    Write Parameter Index 2=true
   )
   &ScaleMeshSizeBySpeed(
    Max Scale Factor=<1 1 1>
    Min Scale Factor:Vector3f=&MaskVectorBySpawnGroup(Masked Vector=<1 1 1>, Passthrough Vector=<0.7 0.7 0.7>, Spawn Group Mask=3)
   )
   &ScaleMeshSize(Scale Factor=&VectorFromFloat(Value=&MaskFloatBySpawnGroup(
      Masked Float=1
      Passthrough Float=&FloatFromCurve(FloatCurve=DataInterface() {propertyValues {Curve {keys [
           <at 0 0.075>
           <at 0.0756 1>
           {arriveTangent=-0.2, interpMode=RCIM_Cubic, leaveTangent=-0.2, time=0.7956, value=1.0207}
           {arriveTangent=-5.4083, interpMode=RCIM_Cubic, leaveTangent=-5.4083, tangentMode=RCTM_User, time=1, value=0}
          ]}}})
      Spawn Group Mask=3
     ))) {enabled=false}
   &SkeletalMeshLocation(
    Bone / Socket Index=$Particles.BoneSelection
    MaskSpawnGroup=true
    Mesh Sampling Type=::NewEnumerator5
    Sampled Position Offset=&SelectVectorFromArray(
     Direct Array Index=$Particles.BoneSelection
     Vector Selection Array=DataInterface() {propertyValues {FloatData [<0 0 0>, <0 0 0>, <-5 0 -2>, <-5 0 -2>, <0 0 0>, <0 0 0>]}}
    )
    Skeletal Mesh=$System.Vehicle_Selection_02
    Socket Sampling Mode=::NewEnumerator5
    Spawn Group Mask=3
    Use Sampled Mesh Attribute as Ribbon ID=false
    Velocity Sampling=::NewEnumerator0
    VelocitySpeedLimit=false
   ) {enabled=false, module=SkeletalMeshLocation003}
   &SkeletalMeshLocation(
    Bone / Socket Index=$Particles.BoneSelection
    MaskSpawnGroup=false
    Mesh Sampling Type=::NewEnumerator5
    Orientation Sampling=::NewEnumerator1
    Sampled Position Offset=&SelectVectorFromArray(
     Direct Array Index=$Particles.BoneSelection
     Vector Selection Array=DataInterface() {propertyValues {FloatData [<0 0 0>, <0 0 0>, <0 0 0>, <0 0 0>, <0 0 0>, <0 0 0>]}}
    )
    Skeletal Mesh=$System.Vehicle_Selection_02
    Socket Sampling Mode=::NewEnumerator5
    Spawn Group Mask=3
    Use Sampled Mesh Attribute as Ribbon ID=false
    Write To Particle Position=false
   ) {module=SkeletalMeshLocation001}
   /SetVariables(Particles.DustWheel Dynamic Parameter:Vector4f=&MakeVector4(X=5, Y=1, Z=0, W=0)) {isSetParameters=true, module=SetVariables_#7}
   /SetVariables(Particles.WheelScale:Vector3f=&self:ScaleMeshfromWheelDimensions(
     Mesh Unscaled Size:Vector3f=<49 49 49>
     Wheel Radius Array:NiagaraDataInterfaceArrayFloat=$User.Wheel_Radius_Array
     Wheel Width Array:NiagaraDataInterfaceArrayFloat=$User.Wheel_Width_Array
     Wheel Index:NiagaraInt32=$Particles.BoneSelection
     Additional Scaling:Vector3f=<1 1 1>
    )) {isSetParameters=true, module=SetVariables_#8}
   &ScaleColor(
    Scale Alpha=&Lerp_Float(A=0, B=1, Alpha=&Abs_Float(Float:NiagaraFloat=$User.Vehicle_Speed))
    ScaleA=true
    ScaleRGB=false
   ) {module=ScaleColor001}
   /SetVariables(
    Particles.MeshOrientation=&self:ScratchDynamicInput(
     Quat:Quat4f=$Particles.SkeletalMeshLocation001.SampledRotation
     Mask Index Attribute:NiagaraInt32=&SpawnGroup()
     Mask Index:NiagaraInt32=3
    )
    Particles.Position=&MaskVectorBySpawnGroup(
     Masked Vector=$Particles.SkeletalMeshLocation001.SampledPosition
     Passthrough Vector=Linked() {linkedVariable {name=Particles.Position, type=NiagaraPosition}}
     Spawn Group Mask=3
    )
   ) {isSetParameters=true}
  ]
  rendererClasses [&NiagaraSpriteRendererProperties, &NiagaraMeshRendererProperties]
  renderers [
   {rendererClass=&NiagaraSpriteRendererProperties}
   {rendererClass=&NiagaraSpriteRendererProperties, rendererIndex=1}
   {rendererClass=&NiagaraMeshRendererProperties, rendererIndex=2}
  ]
 }
 Peel_Tires_Spikes {
  emitterUpdateScript [
   &EmitterState()
   &SpawnRate(
    Spawn Probability=&Multiply_Float(
     A=&MakeCustomFloatFromBool(Bool=$System.IsWaterDeep, True Float=0, False Float=1)
     B=&MakeCustomFloatFromBool(
      Bool=&SetBoolByFloatComparison(
       A=&SelectFloatFromArray(Direct Array Index=3, Float Selection Array=$User.WheelsTouchingArray)
       B=1
       Comparison Type=::NewEnumerator2
      )
      True Float=1
      False Float=0
     )
    )
    SpawnGroup=2
    SpawnRate=&MakeFloatFromInt(Int=&MakeCustomIntFromBool(Bool=$System.PeelOut_Wheel_RR, True INT=126, False INT=0))
    Use Spawn Probability=true
   ) {module=SpawnRate001}
   &SpawnRate(
    Spawn Probability=&Multiply_Float(
     A=&MakeCustomFloatFromBool(Bool=$System.IsWaterDeep, True Float=0, False Float=1)
     B=&MakeCustomFloatFromBool(
      Bool=&SetBoolByFloatComparison(
       A=&SelectFloatFromArray(Direct Array Index=2, Float Selection Array=$User.WheelsTouchingArray)
       B=1
       Comparison Type=::NewEnumerator2
      )
      True Float=1
      False Float=0
     )
    )
    SpawnGroup=3
    SpawnRate=&MakeFloatFromInt(Int=&MakeCustomIntFromBool(Bool=$System.PeelOut_Wheel_RL, True INT=126, False INT=0))
    Use Spawn Probability=true
   )
  ]
  particleSpawnScript [
   &InitializeParticle(
    Color=<rgba 1 0.9167 0.7917 0.35>
    Lifetime=&UniformRangedFloat(Minimum=0.11, Maximum=0.14)
    Mass Mode=::NewEnumerator1
    Sprite Rotation Mode=::NewEnumerator0
    Sprite Size=&MakeVector2D(
     X=&UniformRangedFloat(Minimum=15, Maximum=25)
     Y=&Lerp_Float(
      A=&UniformRangedFloat(Minimum=35, Maximum=70)
      B=&UniformRangedFloat(Minimum=100, Maximum=180)
      Alpha=$User.Vehicle_Speed
     )
    )
    Sprite Size Mode=::NewEnumerator4
    UsePositionOffset=false
    Write Material Random=true
   ) {module=InitializeParticle001}
   &AddVelocityInCone(
    Cone Angle=27
    Cone Axis=&MakeVector(X=1, Y=$User.Slide_Direction, Z=0.5)
    Module.Use Velocity Falloff On Cone Axis=true
    Velocity Distribution Along Cone Axis=0.6
    Velocity Falloff Away From Cone Axis=0.333
    Velocity Strength=1900
   )
   /Niagara/Modules/Update/SubUV/SubUVAnimation(Number Of Frames=3, SubUV Animation Mode=::NewEnumerator1)
  ]
  particleUpdateScript [
   &ParticleState()
   &SkeletalMeshLocation(
    Bone / Socket Index=&SpawnGroup()
    Inherit Velocity=false
    MaskSpawnGroup=false
    Mesh Sampling Type=::NewEnumerator5
    Sampled Position Offset=&MakeVector(
     X=0
     Y=&self:WheelWidthUniformRange(WheelWidthArray:NiagaraDataInterfaceArrayFloat=$User.Wheel_Width_Array, WheelIndex:NiagaraInt32=&SpawnGroup())
     Z=&self:WheelRadiusOffset(
      Wheel Radius Array:NiagaraDataInterfaceArrayFloat=$User.Wheel_Radius_Array
      Wheel Index:NiagaraInt32=&SpawnGroup()
      Additional Offset:NiagaraFloat=0
      Wheel Radius Addl Scale:NiagaraFloat=1
     )
    )
    Skeletal Mesh=$System.Vehicle_Selection_02
    Socket Sampling Mode=::NewEnumerator5
    Write Sampled Mesh Attrributes to Particle Attributes=true
   )
   &SolveForcesAndVelocity(Clamp Velocity=false, Limit Acceleration=false)
   &DynamicMaterialParameters(
    Index 0 Param 1=&FloatFromCurve(FloatCurve=DataInterface() {propertyValues {Curve {keys [<at -0.0046 1.0113>,<at 1.0011 -1>]}}})
    Index 0 Param 4=0.5
    Module.Param0WriteEnabled=true
    Module.Param1WriteEnabled=true
    Module.Param2WriteEnabled=true
    Module.Param3WriteEnabled=true
   )
   &ScaleColor(
    Scale Alpha=&FloatFromCurve(FloatCurve=DataInterface() {propertyValues {Curve {keys [<at 0 0>, {arriveTangent=-3.1724, interpMode=RCIM_Cubic, leaveTangent=-3.1724, tangentMode=RCTM_User, time=0.0788, value=1.0113}, <at 1 0>]}}})
    ScaleA=true
    ScaleRGB=true
   )
   &ScaleSpriteSize(Scale Factor=&Vector2DFromFloat(Value=&FloatFromCurve(FloatCurve=DataInterface() {propertyValues {Curve {keys [{arriveTangent=2.3953, interpMode=RCIM_Cubic, leaveTangent=2.3953, tangentMode=RCTM_User, time=0, value=0.2}, <at 1 1>]}}})))
  ]
  rendererClasses [&NiagaraSpriteRendererProperties]
  renderers [{rendererClass=&NiagaraSpriteRendererProperties}]
 }
 Sparks_Tires_Blown {
  emitterUpdateScript [
   &EmitterState()
   &SpawnRate(
    SpawnRate=&Multiply_Float(
     A=&self:AmountToSpawn(
      " Min Amount":NiagaraFloat=24
      Max Amount:NiagaraFloat=48
      Total Amount:NiagaraInt32=$System.TotalSparks
      Vehicle Speed:NiagaraFloat=&Abs_Float(Float:NiagaraFloat=$User.Vehicle_Speed)
      Vehicle Speed Exponent:NiagaraFloat=1.25
     )
     B=&FloatFromCurve(FloatCurve=$System.Sparks_SpawnCurve, CurveIndex=&Abs_Float(Float:NiagaraFloat=$User.Vehicle_Speed))
    )
    Use Spawn Probability=false
   )
   &SpawnRate(
    SpawnGroup=1
    SpawnRate=&Multiply_Float(
     A=&self:AmountToSpawn(
      " Min Amount":NiagaraFloat=48
      Max Amount:NiagaraFloat=96
      Total Amount:NiagaraInt32=$System.TotalSparks
      Vehicle Speed:NiagaraFloat=&Abs_Float(Float:NiagaraFloat=$User.Vehicle_Speed)
      Vehicle Speed Exponent:NiagaraFloat=1.25
     )
     B=&FloatFromCurve(FloatCurve=$System.Sparks_SpawnCurve, CurveIndex=&Abs_Float(Float:NiagaraFloat=$User.Vehicle_Speed))
    )
    Use Spawn Probability=false
   ) {module=SpawnRate001}
   &SpawnBurst_Instantaneous(Spawn Count=$System.TotalSparks, Spawn Group=2, Use Spawn Probability=false, UseLoopCountLimit=false)
  ]
  particleSpawnScript [
   /SetVariables(Particles.VisibilityTag=&SelectIntFromArray(Direct Array Index=&SpawnGroup(), Int Selection Array=DataInterface() {propertyValues {IntData [0,0,1]}})) {isSetParameters=true}
   &PartitionParticles(Number Of Partitions=$System.TotalSparks)
   &InitializeParticle(
    Color Maximum=&SelectLinearColorFromArray(
     Array Sampling Mode=::NewEnumerator1
     Direct Array Index=&SpawnGroup()
     Color Selection Array=DataInterface() {propertyValues {ColorData [<rgba 40 0.5 0 1>, <rgba 40 11.0654 0 1>, <rgba 150 41.4953 0 1>]}}
     Random Seed=?
    )
    Color Minimum=&SelectLinearColorFromArray(
     Array Sampling Mode=::NewEnumerator1
     Direct Array Index=&SpawnGroup()
     Color Selection Array=DataInterface() {propertyValues {ColorData [<rgba 25 15.7812 0 1>, <rgba 80 15.9878 0 1>, <rgba 50 9.9924 0 1>]}}
     Random Seed=?
    )
    Color Mode=::NewEnumerator2
    Lifetime Max=&SelectFloatFromArray(
     Direct Array Index=&SpawnGroup()
     Float Selection Array=DataInterface() {propertyValues {FloatData [0.2,0.8,0.1]}}
    )
    Lifetime Min:NiagaraFloat=&SelectFloatFromArray(
     Direct Array Index=&SpawnGroup()
     Float Selection Array=DataInterface() {propertyValues {FloatData [0.05,0.25,0.05]}}
    )
    Lifetime Mode=::NewEnumerator1
    Mesh Scale=&self:ScaleMeshfromWheelDimensions(
     Mesh Unscaled Size:Vector3f=<98.5 50 98.5>
     Wheel Radius Array:NiagaraDataInterfaceArrayFloat=$User.Wheel_Radius_Array
     Wheel Width Array:NiagaraDataInterfaceArrayFloat=$User.Wheel_Width_Array
     Wheel Index:NiagaraInt32=&self:TogglesBoneSet(
      Partitions:NiagaraInt32=$Particles.PartitionParticles.Partition
      Front Left Tire:NiagaraBool=$System.Sparks_Wheel_FL
      Front Right Tire:NiagaraBool=$System.Sparks_Wheel_FR
      Rear Left Tire:NiagaraBool=$System.Sparks_Wheel_RL
      Rear Right Tire:NiagaraBool=$System.Sparks_Wheel_RR
      Middle Left Tire:NiagaraBool=$System.Sparks_Wheel_ML
      Middle Right Tire:NiagaraBool=$System.Sparks_Wheel_MR
     )
     Additional Scaling:Vector3f=&Lerp_Vector(A=<1.1 1.1 1.1>, B=<0.75 0.75 0.75>, Alpha=1)
    )
    Mesh Scale Mode=::NewEnumerator4
    Sprite Size Max=<2.5 30>
    Sprite Size Min=<1.25 20>
    Sprite Size Mode=::NewEnumerator2
    UsePositionOffset=false
    Write Material Random=true
   ) {module=InitializeParticle001}
   &SkeletalMeshLocation(
    Bone / Socket Index=&self:TogglesBoneSet(
     Partitions:NiagaraInt32=$Particles.PartitionParticles.Partition
     Front Left Tire:NiagaraBool=$System.Sparks_Wheel_FL
     Front Right Tire:NiagaraBool=$System.Sparks_Wheel_FR
     Rear Left Tire:NiagaraBool=$System.Sparks_Wheel_RL
     Rear Right Tire:NiagaraBool=$System.Sparks_Wheel_RR
     Middle Left Tire:NiagaraBool=$System.Sparks_Wheel_ML
     Middle Right Tire:NiagaraBool=$System.Sparks_Wheel_MR
    )
    "Inherit Velocity (Scale)":NiagaraFloat=&MaskFloatBySpawnGroup(Masked Float=-0.05, Passthrough Float=1, Spawn Group Mask=1)
    MaskSpawnGroup=false
    Mesh Sampling Type=::NewEnumerator5
    Sampled Position Offset=&MakeVector(
     X=&Multiply_Float(A=$User.Vehicle_Speed, B=-20)
     Y=&self:WheelWidthUniformRange(
      WheelWidthArray:NiagaraDataInterfaceArrayFloat=$User.Wheel_Width_Array
      WheelIndex:NiagaraInt32=&self:TogglesBoneSet(
       Partitions:NiagaraInt32=$Particles.PartitionParticles.Partition
       Front Left Tire:NiagaraBool=$System.Sparks_Wheel_FL
       Front Right Tire:NiagaraBool=$System.Sparks_Wheel_FR
       Rear Left Tire:NiagaraBool=$System.Sparks_Wheel_RL
       Rear Right Tire:NiagaraBool=$System.Sparks_Wheel_RR
       Middle Left Tire:NiagaraBool=$System.Sparks_Wheel_ML
       Middle Right Tire:NiagaraBool=$System.Sparks_Wheel_MR
      )
     )
     Z=&self:WheelRadiusOffset(
      Wheel Radius Array:NiagaraDataInterfaceArrayFloat=$User.Wheel_Radius_Array
      Wheel Index:NiagaraInt32=&self:TogglesBoneSet(
       Partitions:NiagaraInt32=$Particles.PartitionParticles.Partition
       Front Left Tire:NiagaraBool=$System.Sparks_Wheel_FL
       Front Right Tire:NiagaraBool=$System.Sparks_Wheel_FR
       Rear Left Tire:NiagaraBool=$System.Sparks_Wheel_RL
       Rear Right Tire:NiagaraBool=$System.Sparks_Wheel_RR
       Middle Left Tire:NiagaraBool=$System.Sparks_Wheel_ML
       Middle Right Tire:NiagaraBool=$System.Sparks_Wheel_MR
      )
      Additional Offset:NiagaraFloat=0
      Wheel Radius Addl Scale:NiagaraFloat=0.5
     )
    )
    Skeletal Mesh=$System.Vehicle_Selection_02
    Socket Sampling Mode=::NewEnumerator5
    Velocity Sampling=::NewEnumerator0
    VelocitySpeedLimit=false
   )
   /SetVariables(Particles.ParticleVelocityAdd:Vector3f=&RotateVectorByQuaternion(
     Vector To Rotate=&MakeVector(
      X=&Multiply_Float(
       A=&MaskFloatBySpawnGroup(
        Masked Float=&RandomRangeFloat(Minimum=-250, Maximum=-800)
        Passthrough Float=&RandomRangeFloat(Minimum=-800, Maximum=-1300)
        Spawn Group Mask=1
       )
       B=&Multiply_Float(
        A=HlslExpression() {hlslExpression="sign(User.Vehicle_Speed)"}
        B=&Lerp_Float(A=0.5, B=1, Alpha=&Abs_Float(Float:NiagaraFloat=$User.Vehicle_Speed))
       )
      )
      Y=&Add_Float(
       A=&Lerp_Float(
        A=-170
        B=170
        Alpha=&ModuloFloat(A=&MakeFloatFromInt(Int=$Particles.PartitionParticles.Partition), B=2)
       )
       B=&RandomRangeFloat(Minimum=-40, Maximum=40)
      )
      Z=&Multiply_Float(
       A=&RandomRangeFloat(Minimum=200, Maximum=500)
       B=&Abs_Float(Float:NiagaraFloat=$User.Vehicle_Speed)
      )
     )
     Quaternion=$Particles.SkeletalMeshLocation.SampledRotation
    )) {isSetParameters=true}
   /SetVariables(Particles.Velocity=&MaskVectorBySpawnGroup(
     Masked Vector=<0 0 0>
     Passthrough Vector=&Add_Vector(A=$Particles.Velocity, B=$Particles.ParticleVelocityAdd)
     Spawn Group Mask=2
    )) {isSetParameters=true}
   &DynamicMaterialParameters(
    Index 0 Param 1=5
    Index 0 Param 2=2
    Module.Index0Param0Write=true
    Module.Index0Param1Write=true
    Module.Index0Param2Write=false
    Module.Index0Param3Write=false
   )
  ]
  particleUpdateScript [
   &ParticleState()
   &ScaleColor(
    Scale Alpha=&MaskFloatBySpawnGroup(
     Masked Float=&FloatFromCurve(
      FloatCurve=DataInterface() {propertyValues {Curve {keys [<at -0.0046 1.0113 RCIM_Cubic>,<at 1.0011 0.002 RCIM_Cubic>]}}}
      Scale Curve=&ClampFloat(Float=&Multiply_Float(A=2, B=&Power(a=&Abs_Float(Float:NiagaraFloat=$User.Vehicle_Speed), b=2.5)))
     )
     Passthrough Float=&FloatFromCurve(FloatCurve=DataInterface() {propertyValues {Curve {keys [<at -0.0046 1.0113 RCIM_Cubic>,<at 1.0011 0.002 RCIM_Cubic>]}}})
     Spawn Group Mask=2
    )
    Scale RGB=&MaskVectorBySpawnGroup(Masked Vector=<100 100 100>, Passthrough Vector=<1 1 1>, Spawn Group Mask=2)
    ScaleA=true
    ScaleRGB=true
   )
   &Drag(
    Drag=&MaskFloatBySpawnGroup(Masked Float=1, Passthrough Float=0, Spawn Group Mask=1)
    Use Linear Drag=true
    Use Rotational Drag=false
   )
   &GravityForce(Gravity=&MaskVectorBySpawnGroup(Masked Vector=<0 0 -1800>, Passthrough Vector=<0 0 0>, Spawn Group Mask=1))
   &SolveForcesAndVelocity(Clamp Velocity=false, Limit Acceleration=false)
   &ScaleSpriteSize(
    Scale Factor=&Vector2DFromFloat(Value=&MaskFloatBySpawnGroup(
      Masked Float=&FloatFromCurve(
       FloatCurve=DataInterface() {propertyValues {Curve {keys [
           {arriveTangent=56.0592, interpMode=RCIM_Cubic, leaveTangent=56.0592, tangentMode=RCTM_User, time=0, value=0}
           {arriveTangent=-1.0335, interpMode=RCIM_Cubic, leaveTangent=-1.0335, tangentMode=RCTM_User, time=0.1674, value=2.5}
           <at 1 0>
          ]}}}
       Scale Curve=&Lerp_Float(A=0.2, B=1, Alpha=$User.Vehicle_Speed)
      )
      Passthrough Float=&FloatFromCurve(
       FloatCurve=DataInterface() {propertyValues {Curve {keys [
           {arriveTangent=3.0697, interpMode=RCIM_Cubic, leaveTangent=3.0697, tangentMode=RCTM_User, time=0, value=0}
           {arriveTangent=-1.0335, interpMode=RCIM_Cubic, leaveTangent=-1.0335, tangentMode=RCTM_User, time=0.1674, value=2.5}
           <at 1 0>
          ]}}}
       Scale Curve=&Lerp_Float(A=0.2, B=1, Alpha=$User.Vehicle_Speed)
      )
      Spawn Group Mask=1
     ))
    Scale Sprite Size Mode=::NewEnumerator2
   )
   &ScaleSpriteSizeBySpeed(
    Max Scale Factor=&MaskVector2dBySpawnGroup(Masked Vector2D=<0.5 0.6>, Passthrough Vector2D=<1 1>, Spawn Group Mask=1)
    Min Scale Factor=&MaskVector2dBySpawnGroup(Masked Vector2D=<1 0.05>, Passthrough Vector2D=<1 1>, Spawn Group Mask=1)
   )
   /SetVariables(Particles.SpriteAlignment=&MaskVectorBySpawnGroup(
     Masked Vector=$Particles.Velocity
     Passthrough Vector=&NormalizeVector(Vector To Normalize=$Particles.ParticleVelocityAdd)
     Spawn Group Mask=1
    )) {isSetParameters=true}
   &SkeletalMeshLocation(
    Bone / Socket Index=&self:TogglesBoneSet(
     Partitions:NiagaraInt32=$Particles.PartitionParticles.Partition
     Front Left Tire:NiagaraBool=$System.Sparks_Wheel_FL
     Front Right Tire:NiagaraBool=$System.Sparks_Wheel_FR
     Rear Left Tire:NiagaraBool=$System.Sparks_Wheel_RL
     Rear Right Tire:NiagaraBool=$System.Sparks_Wheel_RR
     Middle Left Tire:NiagaraBool=$System.Sparks_Wheel_ML
     Middle Right Tire:NiagaraBool=$System.Sparks_Wheel_MR
    )
    MaskSpawnGroup=true
    Mesh Sampling Type=::NewEnumerator5
    Skeletal Mesh=$System.Vehicle_Selection_02
    Socket Sampling Mode=::NewEnumerator5
    Spawn Group Mask=2
    Use Sampled Mesh Attribute as Ribbon ID=false
    Velocity Sampling=::NewEnumerator0
    VelocitySpeedLimit=false
   ) {module=SkeletalMeshLocation002}
   /SetVariables(Particles.MeshOrientation=&MultiplyQuaternion(
     Quaternion A=$Particles.SkeletalMeshLocation002.SampledRotation
     Quaternion B=&OrientMeshToVector(Look At Direction=&RandomRangeVector(Minimum=<0 0 0>, Maximum=<1 0 0>), Reference Axis Vector=<1 0 0>)
    )) {isSetParameters=true}
  ]
  rendererClasses [&NiagaraSpriteRendererProperties, &NiagaraMeshRendererProperties]
  renderers [
   {rendererClass=&NiagaraSpriteRendererProperties}
   {rendererClass=&NiagaraMeshRendererProperties, rendererIndex=1}
  ]
 }
}
summary {
 systemName=NS_Wheel_And_Tire_Interaction
 emitters [
  {
   emitterName=Dust_Tires_World
   simTarget=GPUComputeSim
   rendererClasses [&NiagaraSpriteRendererProperties, &NiagaraMeshRendererProperties]
  }
  {emitterName=Peel_Tires_Spikes, rendererClasses [&NiagaraSpriteRendererProperties]}
  {emitterName=Exhaust_Local, rendererClasses [&NiagaraSpriteRendererProperties, &NiagaraLightRendererProperties]}
  {
   emitterName=Sparks_Tires_Blown
   rendererClasses [&NiagaraSpriteRendererProperties, &NiagaraMeshRendererProperties]
  }
  {
   emitterName=LOW_Dust_Tires_World
   rendererClasses [&NiagaraSpriteRendererProperties, &NiagaraMeshRendererProperties]
  }
  {
   emitterName=High_Dust_Tires_World
   rendererClasses [&NiagaraSpriteRendererProperties, &NiagaraMeshRendererProperties]
  }
 ]
 userVariables [
  User.ExhaustColor {type=LinearColor, description="", defaultValue {struct=LinearColor, value=<rgba 0.6062 0.754 1 0.025>}}
  User.ExhaustScale {type=NiagaraFloat, defaultValue=1, description=""}
  User.IsElectric {type=NiagaraBool}
  User.PeelOutColorIndex {type=NiagaraInt32}
  User.Slide_Direction {type=NiagaraFloat, defaultValue=0.25, description=""}
  User.Vehicle_Speed {type=NiagaraFloat, defaultValue=0, description=""}
  User.VehicleFlags {type=NiagaraInt32}
  User.Wheel_Radius_Array {
   type=NiagaraDataInterfaceArrayFloat
   defaultValue {
    struct=NiagaraExt_VariableValue_DataInterface
    value {
     dataInterface=&self:NiagaraDataInterfaceArrayFloat_2
     dataInterfaceClass=&NiagaraDataInterfaceArrayFloat
    }
   }
  }
  User.Wheel_Width_Array {
   type=NiagaraDataInterfaceArrayFloat
   defaultValue {
    struct=NiagaraExt_VariableValue_DataInterface
    value {
     dataInterface=&self:NiagaraDataInterfaceArrayFloat_1
     dataInterfaceClass=&NiagaraDataInterfaceArrayFloat
    }
   }
  }
  User.WheelsTouchingArray {
   type=NiagaraDataInterfaceArrayFloat
   defaultValue {
    struct=NiagaraExt_VariableValue_DataInterface
    value {
     dataInterface=&self:NiagaraDataInterfaceArrayFloat_0
     dataInterfaceClass=&NiagaraDataInterfaceArrayFloat
    }
   }
  }
 ]
}
