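# Light function material for a Valet underglow light, compatible with the light function atlas, that can pulse to music.
# A time-scrolled sine along U gives a repeating idle pattern.
# When music amplitude is above zero and IsReactivityActive is set, it blends to the powered red channel of the audio MPC's music spectrum.
# A high-power falloff around the middle of V masks the result to a thin band.
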Material /Valet/Materials/M_LF_Underglow
graph {
 properties {
  bForceCompatibleWithLightFunctionAtlas=true
  materialDomain=MD_LightFunction
  pixelDepthOffsetMode=PDOM_Legacy
  usedShadingModels=Default Lit
 }
 nodes [
  n1 = TextureCoordinate()
  n2 = ComponentMask(None=@n1) {r=true}
  n3 = Multiply(A=Time()) {constB=0.05}
  n4 = Sine(None=ComponentMask(None=Add(A=@n2, B=@n3)) {r=true}) {period=0.05}
  n5 = MaterialFunctionCall(Float4=CollectionParameter() {collection=/Game/Athena/Events/AudioAnalysis/MPC_Audio_Reactive_Parameters, parameterName=Spectrum_Music}) {materialFunction=&BreakOutFloat4Components}
  n6 = Max(A=@n5.R, B=@n5.R)
  n7 = Ceil(None=CollectionParameter() {collection=&MPC_Audio_Reactive_Parameters, parameterName=Amplitude_Music})
  n8 = ScalarParameter() {parameterName=IsReactivityActive}
  n9 = ComponentMask(None=@n1) {g=true}
  n10 = Multiply(A=Abs(None=Subtract(A=@n9) {constB=0.5})) {constB=2}
  Multiply(
   A=LinearInterpolate(A=ConstantBiasScale(None=@n4) {bias=1.75, scale=0.25}, B=Power(Base=@n6), Alpha=Multiply(A=@n7, B=@n8)) {constAlpha=1}
   B=Power(Base=OneMinus(None=@n10)) {constExponent=128}
  )
 ]
}
