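# Master translucent, surface-lit vehicle glass material for windows, headlights and brake lights, ending in a vehicle de-rez transition function.
# Color and opacity come either from vertex-color masks per glass region or from window tint with Fresnel-driven opacity.
# Optional stages add depth-faded curvature from normal derivatives plus a normal map, then darken color and specular as fake reflection occlusion.
# A final switchable sorting fix adjusts blue-masked opacity by camera direction.
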Material /Valet/Materials/M_Valet_Glass_Master
graph {
 properties {
  bUseMaterialAttributes=true
  bUsedWithMeshDeformer=true
  bUsedWithSkeletalMesh=true
  blendMode=BLEND_Translucent
  pixelDepthOffsetMode=PDOM_Legacy
  translucencyLightingMode=TLM_SurfacePerPixelLighting
  usedShadingModels=Default Lit
 }
 nodes [
  n1 = VectorParameter() {group=1_Window, parameterName=ColorFront, defaultValue=<rgba 0.0677 0.0677 0.0677 1>}
  n2 = VectorParameter() {group=1_Window, parameterName=ColorSide, defaultValue=<rgba 0.0677 0.0677 0.0677 1>}
  n3 = VertexColor()
  n4 = LinearInterpolate(
   A=LinearInterpolate(
    A=LinearInterpolate(A=@n1.RGB, B=@n2.RGB, Alpha=@n3.A)
    B=VectorParameter() {group=2_HeadLights, parameterName=ColorHeadlights, defaultValue=<rgba 0.0677 0.0677 0.0677 1>}.RGB
    Alpha=@n3.R
   )
   B=VectorParameter() {group=3_BreakLights, parameterName=ColorBreaklights, defaultValue=<rgba 1 0 0.0158 1>}.RGB
   Alpha=@n3.G
  )
  n5 = VectorParameter() {group=4_Special, parameterName=ColorSpecial, defaultValue=<rgba 0.0677 0.0677 0.0677 1>}
  n6 = VectorParameter() {group=GLass Params, parameterName=Color_DarkTint, defaultValue=<rgba 0.001 0.001 0.001 1>}
  n7 = ScalarParameter() {group=1_Window, parameterName=Window Tint Amount, sortPriority=31}
  n8 = ScalarParameter() {group=1_Window, parameterName=Window Tint AmountFront}
  n9 = ScalarParameter() {group=1_Window, parameterName=Window Tint AmountSide}
  n10 = VertexColor()
  n11 = LinearInterpolate(
   A=LinearInterpolate(
    A=LinearInterpolate(A=@n8, B=@n9, Alpha=@n10.A)
    B=ScalarParameter() {group=2_HeadLights, parameterName=Window Tint Amount Headlights}
    Alpha=@n10.R
   )
   B=ScalarParameter() {group=3_BreakLights, parameterName=Window Tint Amount Breaklights}
   Alpha=@n10.G
  )
  n12 = ScalarParameter() {group=4_Special, parameterName=Window Tint Amount Special}
  n13 = LinearInterpolate(
   A=ScalarParameter() {group=GLass Params, parameterName=Fresnel Inner Transparency}
   B=ScalarParameter() {defaultValue=0.9, group=GLass Params, parameterName=Fresnel Inner Transparency Max Tint}
   Alpha=@n7
  )
  n14 = ScalarParameter() {defaultValue=0.98, group=GLass Params, parameterName=Fresnel Outer Transparency}
  n15 = ScalarParameter() {defaultValue=2, group=GLass Params, parameterName=Fresnel Exponent}
  n16 = Clamp(None=Multiply(A=Fresnel(ExponentIn=@n15), B=ScalarParameter() {defaultValue=1, group=GLass Params, parameterName=Glass thickness}))
  n17 = SetMaterialAttributes!(
   MaterialAttributes
   Base Color=StaticSwitchParameter(True=LinearInterpolate(A=@n4, B=@n5.RGB, Alpha=@n3.B), False=LinearInterpolate(A=@n1.RGB, B=@n6.RGB, Alpha=@n7)) {group=GLass Params, parameterName=UseVertexMaskTints}
   Opacity=StaticSwitchParameter(True=LinearInterpolate(A=@n11, B=@n12, Alpha=@n10.B), False=LinearInterpolate(A=@n13, B=@n14, Alpha=@n16)) {group=GLass Params, parameterName=UseVertexMaskTints}
   Roughness=ScalarParameter() {defaultValue=0.125, parameterName=Roughness}
   Specular=ScalarParameter() {defaultValue=0.8, parameterName=Spec}
   Metallic=ScalarParameter() {defaultValue=0.25, parameterName=Metalness?}
  ) {attributeSetTypes [
    #1
    #2
    #3
    #4
    #5
   ]}
  n18 = GetMaterialAttributes(None=@n17) {attributeGetTypes [#1, #2]}
  n19 = LinearInterpolate(
   A=ScalarParameter() {parameterName=CurvatureTexture_Add1}
   B=ScalarParameter() {parameterName=CurvatureTexture_Add2}
   Alpha=TextureSampleParameter2D() {parameterName=Curvature, samplerType=SAMPLERTYPE_Masks, texture=&T_BlackMask}.B
  )
  n20 = VertexNormalWS()
  n21 = DDX(Value=@n20)
  n22 = DDY(Value=@n20)
  n23 = Divide(
   A=MaterialFunctionCall(Vector 3=Add(A=@n21, B=@n22)) {materialFunction=/Engine/Functions/Engine_MaterialFunctions02/Utility/VectorLength}.V3 Length
   B=ScalarParameter() {defaultValue=0.2257, parameterName=Curvature}
  )
  n24 = ScalarParameter() {defaultValue=500, parameterName=FadeLength}
  n25 = ScalarParameter() {defaultValue=200, parameterName=FadeOffset}
  n26 = Saturate(None=LinearInterpolate(
    A=Add(A=@n19, B=@n23)
    Alpha=MaterialFunctionCall(Fade Length=@n24, Fade Offset=@n25) {materialFunction=&CameraDepthFade}.Result
   ) {constB=0})
  n27 = Add(
   A=@n18.Base Color
   B=LinearInterpolate(A=ScalarParameter() {defaultValue=1, parameterName=BaseAdd}, B=ScalarParameter() {defaultValue=1.4, parameterName=BaseAdd2}, Alpha=@n26)
  )
  n28 = Add(
   A=@n18.Opacity
   B=LinearInterpolate(
    A=ScalarParameter() {defaultValue=1, parameterName=OpacityAdd}
    B=ScalarParameter() {defaultValue=1.4, parameterName=OpacityAdd2}
    Alpha=@n26
   )
  )
  n29 = Constant3Vector() {constant=<rgba 0 0 1 0>}
  n30 = TextureSampleParameter2D() {parameterName=Normal, samplerType=SAMPLERTYPE_Normal, texture=/Engine/EngineMaterials/FlatNormal}
  n31 = Saturate(None=ChannelMaskParameter(None=VertexColor()) {parameterName=MaskNormal}.R)
  n32 = StaticSwitchParameter(
   True=SetMaterialAttributes!(
    MaterialAttributes=@n18.MaterialAttributes
    Base Color=Saturate(None=@n27)
    Opacity=Saturate(None=@n28)
    Normal=LinearInterpolate(A=@n29, B=@n30.RGB, Alpha=@n31)
   ) {attributeSetTypes [#1, #2, #6]}
   False=@n17
  ) {parameterName=UseCurvaturemap}
  n33 = GetMaterialAttributes(None=@n32) {attributeGetTypes [#1, #4]}
  n34 = PreSkinnedPosition()
  n35 = Divide(
   A=Add(A=ComponentMask(None=@n34) {r=true}, B=ScalarParameter() {group=ReflectionOcclusion, parameterName=OcclusionGradientAdd})
   B=ScalarParameter() {defaultValue=50, group=ReflectionOcclusion, parameterName=OcclusionGradientDistance}
  )
  n36 = VertexNormalWS()
  n37 = CameraVectorWS()
  n38 = LinearInterpolate(
   A=ScalarParameter() {defaultValue=3, group=ReflectionOcclusion, parameterName=OcclusionAngleExp1}
   B=ScalarParameter() {defaultValue=3, group=ReflectionOcclusion, parameterName=OcclusionAngleExp2}
   Alpha=Saturate(None=DotProduct(A=@n36, B=@n37))
  )
  n39 = ChannelMaskParameter(None=VertexColor()) {group=ReflectionOcclusion, maskChannel=Blue, parameterName=FakeReflectionOcclusionChannel, defaultValue=<rgba 0 0 1 0>}
  n40 = LinearInterpolate(
   B=LinearInterpolate(
    A=ScalarParameter() {group=ReflectionOcclusion, parameterName=OcclusionMin}
    B=ScalarParameter() {defaultValue=1, group=ReflectionOcclusion, parameterName=OcclusionMax}
    Alpha=VertexInterpolator(VS=@n35).PS
   )
   Alpha=Multiply(A=Saturate(None=@n38), B=@n39.R)
  ) {constA=1}
  n41 = StaticSwitchParameter(
   True=SetMaterialAttributes!(MaterialAttributes=@n33.MaterialAttributes, Base Color=Multiply(A=@n33.Base Color, B=@n40), Specular=Multiply(A=@n33.Specular, B=@n40)) {attributeSetTypes [#1, #4]}
   False=@n32
  ) {group=ReflectionOcclusion, parameterName=UseFakeReflectionOcclusion}
  n42 = GetMaterialAttributes(None=@n41) {attributeGetTypes [#2]}
  n43 = Subtract(
   A=CameraPositionWS()
   B=TransformPosition(None=VectorParameter() {parameterName=SortingCenter, defaultValue=<rgba 0 0 0 1>}.RGB) {transformType=TRANSFORMPOSSOURCE_World}
  )
  n44 = VectorParameter() {parameterName=Sorting Direction, defaultValue=<rgba 1 0 -1.1633 1>}
  n45 = LinearInterpolate(
   A=ScalarParameter() {defaultValue=1.2644, parameterName=SortingFix_1}
   B=ScalarParameter() {defaultValue=5.738, parameterName=SortingFix_2}
   Alpha=DotProduct(A=Normalize(VectorInput=@n43), B=Transform(None=@n44.RGB) {transformSourceType=TRANSFORMSOURCE_Local})
  )
  n46 = LinearInterpolate(
   A=@n42.Opacity
   B=ScalarParameter() {defaultValue=0.2286, parameterName=SortingFixOpacity}
   Alpha=Multiply(A=VertexColor().B, B=Saturate(None=@n45))
  )
  MaterialFunctionCall(In=StaticSwitchParameter(
    True=SetMaterialAttributes!(MaterialAttributes=@n42.MaterialAttributes, Opacity=@n46) {attributeSetTypes [#2]}
    False=@n41
   ) {group=ReflectionOcclusion, parameterName=UseSortingFix}) {materialFunction=&MF_VehicleDeRez_Transition}
 ]
}
