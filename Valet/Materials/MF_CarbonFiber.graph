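# Material function that outputs carbon fiber material attributes with clear coat and anisotropy, exposed through CarbonFiber parameters.
# Tiled weave masks blend base color, metallic and roughness, while a weave normal map feeds only a Fresnel term that shifts fiber tint and scales metallic and roughness.
# The remapped red mask alternates anisotropy between two values.
# The output normal comes from the input material attributes, defaulting to flat, so the weave adds no surface bump.
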MaterialFunction /Valet/Materials/MF_CarbonFiber
graph {nodes [
  n1 = VectorParameter() {group=CarbonFiber, parameterName=CarbonFiber1, defaultValue=<rgba 0.06 0.06 0.06 1>}
  n2 = VectorParameter() {group=CarbonFiber, parameterName=CarbonFiber2, defaultValue=<rgba 0.02 0.02 0.02 1>}
  n3 = Multiply(
   A=TextureCoordinate()
   B=AppendVector(
    A=ScalarParameter() {defaultValue=140, group=CarbonFiber, parameterName=CarbonFiber_Tiling_X}
    B=ScalarParameter() {defaultValue=140, group=CarbonFiber, parameterName=CarbonFiber_Tiling_Y}
   )
  )
  n4 = Fresnel(
   ExponentIn=ScalarParameter() {defaultValue=10, group=CarbonFiber, parameterName=CarbonFresnel}
   BaseReflectFractionIn=ScalarParameter() {defaultValue=-0.4, group=CarbonFiber, parameterName=CarbonFresnelBRF}
   Normal=TextureSample(UVs=@n3) {samplerType=SAMPLERTYPE_Normal, texture=/Valet/Materials/SharedTextures/T_CarbonFiberWeave_03_N}.RGB
  )
  n5 = TextureSample(UVs=@n3) {samplerType=SAMPLERTYPE_Masks, texture=/Valet/Materials/SharedTextures/T_CarbonFiberWeave_01_M}
  n6 = Multiply(A=@n5.G, B=ScalarParameter() {defaultValue=0.6, group=CarbonFiber, parameterName=CarbonFiber_Metallic})
  n7 = Multiply(A=@n5.B, B=ScalarParameter() {defaultValue=4, group=CarbonFiber, parameterName=CarbonFiber_Roughness1})
  n8 = SetMaterialAttributes!(MaterialAttributes, Normal=Constant3Vector() {constant=<rgba 0 0 1 0>}) {attributeSetTypes [#1]}
  n9 = LinearInterpolate(Alpha=@n5.R) {constA=-0.3, constB=10}
  n10 = SetMaterialAttributes!(
   MaterialAttributes
   BaseColor=LinearInterpolate(
    A=VectorParameter() {group=CarbonFiber, parameterName=CarbonFiber, defaultValue=<rgba 0.025 0.025 0.025 1>}.RGB
    B=LinearInterpolate(A=@n1.RGB, B=@n2.RGB, Alpha=@n4)
    Alpha=@n5.R
   )
   Metallic=LinearInterpolate(B=Multiply(A=@n4, B=@n6), Alpha=@n5.R)
   Specular=ScalarParameter() {defaultValue=0.5, group=CarbonFiber, parameterName=CarbonFiber_Specular}
   Roughness=LinearInterpolate(A=ScalarParameter() {defaultValue=0.3, group=CarbonFiber, parameterName=CarbonFiber_Roughness0}, B=Multiply(A=@n4, B=@n7), Alpha=@n5.R)
   Normal=GetMaterialAttributes(None=FunctionInput(Preview=@n8) {bUsePreviewValueAsDefault=true, inputType=FunctionInput_MaterialAttributes}) {attributeGetTypes [#1]}.Normal
   CustomData0=ScalarParameter() {defaultValue=1, group=CarbonFiber, parameterName=CarbonFiber_ClearCoat}
   CustomData1=ScalarParameter() {group=CarbonFiber, parameterName=CarbonFiber_ClearCoat_Roughness}
   Anisotropy=LinearInterpolate(
    A=VectorParameter() {group=CarbonFiber, parameterName=Aniso1, defaultValue=<rgba 0.0833 0.0833 0.0833 1>}.RGB
    B=VectorParameter() {group=CarbonFiber, parameterName=Aniso2, defaultValue=<rgba 0.7292 0.7292 0.7292 1>}.RGB
    Alpha=Saturate(None=@n9)
   )
  ) {attributeSetTypes [
    #2
    #3
    #4
    #5
    #1
    #6
    #7
    #8
   ]}
  FunctionOutput(None=@n10)
 ]}
