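# Material function that applies a simple surface layer over incoming Material attributes, blended by an Alpha input.
# Layer Color is scaled by a Gmap-driven brightness lerp, while metalness and specular come straight from inputs.
# The incoming roughness remaps between min/max inputs for both layer roughness and clearcoat roughness, preserving underlying variation.
# Layer Clearcoat feeds CustomData0 and switches the shading model to Clear Coat above a threshold.
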MaterialFunction /Valet/Materials/MF_Valet_SimpleLayer
graph {nodes [
  n1 = FunctionInput() {inputName=Material, inputType=FunctionInput_MaterialAttributes}
  n2 = FunctionInput() {inputName=Layer Color}
  n3 = LinearInterpolate(Alpha=FunctionInput() {inputName=Gmap, inputType=FunctionInput_Scalar}) {constA=0.2, constB=1.12}
  n4 = FunctionInput() {inputName=Layer Roughness Min, inputType=FunctionInput_Scalar}
  n5 = FunctionInput() {inputName=Layer Roughness Max, inputType=FunctionInput_Scalar, previewValue=<1 0 0 1>}
  n6 = GetMaterialAttributes(None=@n1) {attributeGetTypes [#1]}
  n7 = FunctionInput() {inputName=Layer Clearcoat, inputType=FunctionInput_Scalar}
  n8 = FunctionInput() {inputName=Layer Clearcoat Roughness Min, inputType=FunctionInput_Scalar}
  n9 = FunctionInput() {inputName=Layer Clearcoat Roughness Max, inputType=FunctionInput_Scalar, previewValue=<1 0 0 1>}
  n10 = ShadingModel()
  n11 = SetMaterialAttributes!(
   MaterialAttributes=@n1
   BaseColor=Saturate(None=Multiply(A=@n2, B=@n3))
   Metallic=FunctionInput() {inputName=Layer Metalness, inputType=FunctionInput_Scalar}
   Roughness=Saturate(None=LinearInterpolate(A=@n4, B=@n5, Alpha=@n6.Roughness))
   CustomData0=@n7
   CustomData1=Saturate(None=LinearInterpolate(A=@n8, B=@n9, Alpha=@n6.Roughness))
   ShadingModel=If(A=@n7, B=Constant() {r=0.5}, "A > B"=ShadingModel() {shadingModel=MSM_ClearCoat}, "A == B"=@n10, "A < B"=@n10)
   Specular=FunctionInput() {inputName=Layer Specular, inputType=FunctionInput_Scalar}
  ) {attributeSetTypes [
    #2
    #3
    #1
    #4
    #5
    #6
    #7
   ]}
  FunctionOutput(None=BlendMaterialAttributes(A=@n1, B=@n11, Alpha=FunctionInput() {inputName=Alpha, inputType=FunctionInput_Scalar}))
  LinearInterpolate(Alpha=@n7)
 ]}
