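# Base material function for Valet vehicles, switching statically between layered customization and realistic car paint.
# The layered path runs textured attributes, with Fresnel, time-of-day or diffuse-driven emissive, through three or four masked color layers, then scratch/grime.
# The paint path builds clearcoat with view-angle tints and optional second color, orange peel, normal blur, carbon fiber and distance-faded hex-grid overlay.
# Extra outputs expose scratch color, lights, Gmap and emissive mask.
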MaterialFunction /Valet/Materials/MF_Valet_BaseFunctionality
graph {nodes [
  n1 = TextureSampleParameter2D() {
   parameterName=Scratch/Grime/EMPTY
   samplerSource=SSM_Wrap_WorldGroupSettings
   samplerType=SAMPLERTYPE_Masks
   texture=&T_BlackMask
  }
  n2 = ScalarParameter() {group=Layer 04, parameterName=Layer 04 Clearcoat}
  n3 = ScalarParameter() {group=Layer 01, parameterName=Layer 01 Clearcoat}
  n4 = TextureSampleParameter2D() {
   group=Texture Input
   parameterName=Diffuse
   samplerSource=SSM_Wrap_WorldGroupSettings
   texture=&T_BlacksRGB
  }
  n5 = TextureSampleParameter2D() {
   group=Texture Input
   parameterName=Specular Mask
   samplerSource=SSM_Wrap_WorldGroupSettings
   samplerType=SAMPLERTYPE_Masks
   texture=&T_BlackMask
   channelNames {a=Unused, b=Specular, g=Roughness, r=Metallic}
  }
  n6 = TextureSampleParameter2D() {
   group=Texture Input
   parameterName=Gmap/Emissive/Lights
   samplerSource=SSM_Wrap_WorldGroupSettings
   samplerType=SAMPLERTYPE_Masks
   texture=&T_BlackMask
   channelNames {a=DiffuseToEmissiveMask, b=Lights, g=Emissive, r=Gmap}
  }
  n7 = Power(Base=@n6.Emissive, Exp=ScalarParameter() {defaultValue=1, group=Emissive, parameterName=EmissiveTexturePower})
  n8 = ScalarParameter() {group=Emissive, parameterName=Override Emissive Mask Full}
  n9 = PixelNormalWS()
  n10 = CameraVectorWS()
  n11 = Power(
   Base=Clamp(None=DotProduct(A=@n9, B=@n10))
   Exp=ScalarParameter() {defaultValue=2, group=Emissive, parameterName=View Dependant Falloff Emissive}
  )
  n12 = OneMinus(None=@n11)
  n13 = ScalarParameter() {group=Emissive, parameterName=Invert Emissive Fresnel}
  n14 = Multiply(
   A=Multiply(A=VectorParameter() {group=Emissive, parameterName=Emissive Color, defaultValue=<rgba 0 0 0 1>}.RGB, B=LinearInterpolate(A=@n7, Alpha=@n8))
   B=Saturate(None=LinearInterpolate(A=@n11, B=@n12, Alpha=@n13))
  )
  n15 = ScalarParameter() {defaultValue=1, group=Emissive, parameterName=EmissiveTODFactor_Morning, sortPriority=51}
  n16 = ScalarParameter() {defaultValue=1, group=Emissive, parameterName=EmissiveTODFactor_Day, sortPriority=52}
  n17 = ScalarParameter() {defaultValue=1, group=Emissive, parameterName=EmissiveTODFactor_Evening, sortPriority=53}
  n18 = ScalarParameter() {defaultValue=1, group=Emissive, parameterName=EmissiveTODFactor_Night, sortPriority=54}
  n19 = StaticSwitchParameter(
   True=Multiply(
    A=@n14
    B=MaterialFunctionCall(Value_Morning=@n15, Value_Day=@n16, Value_Evening=@n17, Value_Night=@n18) {materialFunction=&Scalar_Time_of_Day}.Result
   )
   False=@n14
  ) {desc=Multiplies emissive value by factor per time-of-day phase, group=Emissive, parameterName=UseEmissiveTODFactor, sortPriority=50}
  n20 = StaticSwitchParameter(
   True=Multiply(
    A=Multiply(A=@n4.RGB, B=@n6.DiffuseToEmissiveMask)
    B=VectorParameter() {group=Emissive, parameterName=DiffuseToEmissiveColorScale, defaultValue=<rgba 1 1 1 0>}.RGB
   )
   False=@n19
  ) {desc=Multiplies emissive value by factor per time-of-day phase, group=Emissive, parameterName=AddDiffuseToEmissive, sortPriority=50}
  n21 = TextureSampleParameter2D() {
   group=Texture Input
   parameterName=Normal
   samplerSource=SSM_Wrap_WorldGroupSettings
   samplerType=SAMPLERTYPE_Normal
   texture=&FlatNormalBlue
  }
  n22 = SetMaterialAttributes!(
   MaterialAttributes
   BaseColor=@n4.RGB
   Metallic=@n5.Metallic
   Specular=@n5.Metallic
   Roughness=@n5.Metallic
   EmissiveColor=@n20
   Normal=@n21.RGB
   ShadingModel=ShadingModel()
   CustomData0=ScalarParameter() {parameterName=CC_defaultlayer}
   CustomData1=ScalarParameter() {defaultValue=0.4, parameterName=CCR_defaultlayer}
  ) {attributeSetTypes [
    #1
    #2
    #3
    #4
    #5
    #6
    #7
    #8
    #9
   ]}
  n23 = VectorParameter() {group=Layer 01, parameterName=Layer 01 Color, defaultValue=<rgba 0 0 0 1>}
  n24 = ScalarParameter() {group=Layer 01, parameterName=Layer 01 Roughness Min}
  n25 = ScalarParameter() {group=Layer 01, parameterName=Layer 01 Metalness}
  n26 = TextureSampleParameter2D() {
   desc=Pri/Sec/Tri/Tet
   group=Texture Input
   parameterName=Mask
   samplerSource=SSM_Wrap_WorldGroupSettings
   samplerType=SAMPLERTYPE_Masks
   texture=&T_BlackMask
   channelNames {a=Layer 04, b=Layer 03, g=Layer 02, r=Layer 01}
  }
  n27 = ScalarParameter() {defaultValue=1, group=Layer 01, parameterName=Layer 01 Roughness Max}
  n28 = ScalarParameter() {group=Layer 01, parameterName=Layer 01 Clearcoat Roughness Min}
  n29 = ScalarParameter() {defaultValue=1, group=Layer 01, parameterName=Layer 01 Clearcoat Roughness Max}
  n30 = ScalarParameter() {defaultValue=1, group=Layer 01, parameterName=Layer 01 Specular, sliderMax=1}
  n31 = MaterialFunctionCall!(
   Layer Clearcoat=ScalarParameter() {group=Layer 03, parameterName=Layer 03 Clearcoat}
   Material=MaterialFunctionCall(
    Layer Clearcoat=ScalarParameter() {group=Layer 02, parameterName=Layer 02 Clearcoat}
    Material=MaterialFunctionCall(
     Layer Clearcoat=@n3
     Material=@n22
     Layer Color=@n23.RGB
     Layer Roughness Min=@n24
     Layer Metalness=@n25
     Alpha=@n26.Layer 01
     Gmap=@n6.Gmap
     Layer Roughness Max=@n27
     Layer Clearcoat Roughness Min=@n28
     Layer Clearcoat Roughness Max=@n29
     Layer Specular=@n30
    ) {materialFunction=&/MF_Valet_SimpleLayer}.Result
    Layer Color=VectorParameter() {group=Layer 02, parameterName=Layer 02 Color, defaultValue=<rgba 0 0 0 1>}.RGB
    Layer Roughness Min=ScalarParameter() {group=Layer 02, parameterName=Layer 02 Roughness Min}
    Layer Metalness=ScalarParameter() {group=Layer 02, parameterName=Layer 02 Metalness}
    Alpha=@n26.Layer 02
    Gmap=@n6.Gmap
    Layer Roughness Max=ScalarParameter() {defaultValue=1, group=Layer 02, parameterName=Layer 02 Roughness Max}
    Layer Clearcoat Roughness Min=ScalarParameter() {group=Layer 02, parameterName=Layer 02 Clearcoat Roughness Min}
    Layer Clearcoat Roughness Max=ScalarParameter() {defaultValue=1, group=Layer 02, parameterName=Layer 02 Clearcoat Roughness Max}
    Layer Specular=ScalarParameter() {defaultValue=1, group=Layer 02, parameterName=Layer 02 Specular, sliderMax=1}
   ) {materialFunction=&MF_Valet_SimpleLayer}.Result
   Layer Color=VectorParameter() {group=Layer 03, parameterName=Layer 03 Color, defaultValue=<rgba 0 0 0 1>}.RGB
   Layer Roughness Min=ScalarParameter() {group=Layer 03, parameterName=Layer 03 Roughness Min}
   Layer Metalness=ScalarParameter() {group=Layer 03, parameterName=Layer 03 Metalness}
   Alpha=@n26.Layer 03
   Gmap=@n6.Gmap
   Layer Roughness Max=ScalarParameter() {defaultValue=1, group=Layer 03, parameterName=Layer 03 Roughness Max}
   Layer Clearcoat Roughness Min=ScalarParameter() {group=Layer 03, parameterName=Layer 03 Clearcoat Roughness Min}
   Layer Clearcoat Roughness Max=ScalarParameter() {defaultValue=1, group=Layer 03, parameterName=Layer 03 Clearcoat Roughness Max}
   Layer Specular=ScalarParameter() {defaultValue=1, group=Layer 03, parameterName=Layer 03 Specular, sliderMax=1}
  ) {materialFunction=&MF_Valet_SimpleLayer}
  n32 = VectorParameter() {group=Layer 04, parameterName=Layer 04 Color, defaultValue=<rgba 0 0 0 1>}
  n33 = ScalarParameter() {group=Layer 04, parameterName=Layer 04 Roughness Min}
  n34 = ScalarParameter() {group=Layer 04, parameterName=Layer 04 Metalness}
  n35 = ScalarParameter() {defaultValue=1, group=Layer 04, parameterName=Layer 04 Roughness Max}
  n36 = ScalarParameter() {group=Layer 04, parameterName=Layer 04 Clearcoat Roughness Min}
  n37 = ScalarParameter() {defaultValue=1, group=Layer 04, parameterName=Layer 04 Clearcoat Roughness Max}
  n38 = ScalarParameter() {defaultValue=1, group=Layer 04, parameterName=Layer 04 Specular, sliderMax=1}
  n39 = MaterialFunctionCall!(
   Grime=@n1.G
   Scratch=@n1.G
   Gmap=@n6.Gmap
   Material=StaticSwitchParameter(
    True=MaterialFunctionCall(
     Layer Clearcoat=@n2
     Material=@n31.Result
     Layer Color=@n32.RGB
     Layer Roughness Min=@n33
     Layer Metalness=@n34
     Alpha=@n26.Layer 04
     Gmap=@n6.Gmap
     Layer Roughness Max=@n35
     Layer Clearcoat Roughness Min=@n36
     Layer Clearcoat Roughness Max=@n37
     Layer Specular=@n38
    ) {materialFunction=&MF_Valet_SimpleLayer}.Result
    False=@n31.Result
   ) {group=Layer 04, parameterName=Add 4th Customization Layer}
  ) {materialFunction=&/MF_Valet_ScratchGrime}
  n40 = TextureSampleParameter2D() {
   group=Texture Input
   parameterName=Diffuse
   samplerSource=SSM_Wrap_WorldGroupSettings
   texture=&T_BlacksRGB
  }
  n41 = TextureSampleParameter2D() {
   group=Texture Input
   parameterName=Specular Mask
   samplerSource=SSM_Wrap_WorldGroupSettings
   samplerType=SAMPLERTYPE_Masks
   texture=&T_BlackMask
  }
  n42 = TextureSampleParameter2D() {
   group=Texture Input
   parameterName=Gmap/Emissive/Lights
   samplerSource=SSM_Wrap_WorldGroupSettings
   samplerType=SAMPLERTYPE_Masks
   texture=&T_BlackMask
  }
  n43 = ScalarParameter() {group=Emissive, parameterName=Override Emissive Mask Full}
  n44 = TextureSampleParameter2D() {
   group=Texture Input
   parameterName=Normal
   samplerSource=SSM_Wrap_WorldGroupSettings
   samplerType=SAMPLERTYPE_Normal
   texture=&FlatNormalBlue
  }
  n45 = SetMaterialAttributes(
   BaseColor=@n40.RGB
   Metallic=@n41.G
   Specular=@n41.G
   Roughness=@n41.G
   EmissiveColor=Multiply(
    A=VectorParameter() {group=Emissive, parameterName=Emissive Color, defaultValue=<rgba 0 0 0 1>}.RGB
    B=LinearInterpolate(A=@n42.G, Alpha=@n43)
   )
   Normal=@n44.RGB
   ShadingModel=ShadingModel()
   CustomData0=ScalarParameter() {parameterName=CC_defaultlayer}
   CustomData1=ScalarParameter() {parameterName=CCR_defaultlayer}
  ) {attributeSetTypes [
    #1
    #2
    #3
    #4
    #5
    #6
    #7
    #8
    #9
   ]}
  n46 = ScalarParameter() {defaultValue=1, group=RealisticPaint, parameterName=EnergyConservation_low}
  n47 = VectorParameter() {group=RealisticPaint, parameterName=TintPaint_a, defaultValue=<rgba 0.6979 0.6979 0.6979 1>}
  n48 = WorldPosition()
  n49 = CameraPositionWS()
  n50 = DotProduct(A=VertexNormalWS(), B=Normalize(VectorInput=Subtract(A=@n48.XYZ, B=@n49)))
  n51 = Clamp(None=Power(Base=Abs(None=@n50), Exp=ScalarParameter() {defaultValue=0.2, group=RealisticPaint, parameterName=ExpPaintFresnel}) {constExponent=5})
  n52 = ScalarParameter() {defaultValue=1, group=RealisticPaint, parameterName=EnergyConservation_High}
  n53 = VectorParameter() {group=RealisticPaint, parameterName=TintPaint_b, defaultValue=<rgba 0.6979 0.6979 0.6979 1>}
  n54 = LinearInterpolate(
   A=LinearInterpolate(A=Multiply(A=@n46, B=@n47.RGB), B=@n47.RGB, Alpha=@n51)
   B=LinearInterpolate(A=Multiply(A=@n52, B=@n53.RGB), B=@n53.RGB, Alpha=@n51)
   Alpha=@n51
  )
  n55 = VectorParameter() {group=RealisticPaint2, parameterName=TintPaint2_a, defaultValue=<rgba 0.6979 0.6979 0.6979 1>}
  n56 = LinearInterpolate(
   A=Multiply(A=ScalarParameter() {defaultValue=1, group=RealisticPaint2, parameterName=EnergyConservation2_low}, B=@n55.RGB)
   B=@n55.RGB
   Alpha=@n51
  )
  n57 = VectorParameter() {group=RealisticPaint2, parameterName=TintPaint2_b, defaultValue=<rgba 0.6979 0.6979 0.6979 1>}
  n58 = LinearInterpolate(
   A=Multiply(A=ScalarParameter() {defaultValue=1, group=RealisticPaint2, parameterName=EnergyConservation2_High}, B=@n57.RGB)
   B=@n57.RGB
   Alpha=@n51
  )
  n59 = TextureSampleParameter2D() {
   group=Texture Input
   parameterName=MaskRealisticPaint
   samplerSource=SSM_Wrap_WorldGroupSettings
   texture=&T_BlacksRGB
  }
  n60 = StaticSwitchParameter(True=LinearInterpolate(A=@n54, B=LinearInterpolate(A=@n56, B=@n58, Alpha=@n51), Alpha=@n59.G), False=@n54) {group=RealisticPaint, parameterName=Use2PaintColors}
  n61 = TextureSampleParameter2D() {group=Texture Input, parameterName=AO, samplerSource=SSM_Wrap_WorldGroupSettings, texture=&T_BlacksRGB}
  n62 = ScalarParameter() {group=RealisticPaint, parameterName=AO_intensity}
  n63 = Multiply(A=@n60, B=Saturate(None=Add(A=@n61.R, B=@n62)))
  n64 = LinearInterpolate(
   A=ScalarParameter() {defaultValue=1, group=RealisticPaint, parameterName=Metallic_low}
   B=ScalarParameter() {defaultValue=1, group=RealisticPaint, parameterName=Metallic_high}
   Alpha=@n51
  )
  n65 = LinearInterpolate(
   A=ScalarParameter() {defaultValue=1, group=RealisticPaint2, parameterName=Metallic2_low}
   B=ScalarParameter() {defaultValue=1, group=RealisticPaint2, parameterName=Metallic2_high}
   Alpha=@n51
  )
  n66 = TextureSampleParameter2D(UVs=Multiply(A=TextureCoordinate(), B=ScalarParameter() {defaultValue=20, group=RealisticPaint, parameterName=OrangePeelSize})) {
   group=Texture Input
   parameterName=OrangePeelNormal
   samplerSource=SSM_Wrap_WorldGroupSettings
   samplerType=SAMPLERTYPE_Normal
   texture=&FlatNormalBlue
  }
  n67 = MaterialFunctionCall(Normal=@n66.RGB, Flatness=ScalarParameter() {defaultValue=0.5, group=RealisticPaint, parameterName=OrangePeelNormal}) {materialFunction=&FlattenNormal}
  n68 = TextureSampleParameter2D() {
   group=Texture Input
   parameterName=Normal
   samplerSource=SSM_Wrap_WorldGroupSettings
   samplerType=SAMPLERTYPE_Normal
   texture=&FlatNormalBlue
  }
  n69 = TextureCoordinate()
  n70 = Constant() {r=1}
  n71 = ScalarParameter() {defaultValue=2048, group=NormalBlur, parameterName=TextureRes}
  n72 = MaterialFunctionCall!(OffsetAmount=Multiply(A=Divide(A=@n70, B=@n71), B=ScalarParameter() {defaultValue=1, group=NormalBlur, parameterName=NormalBlurPixels})) {materialFunction=/Engine/Functions/Engine_MaterialFunctions01/ImageAdjustment/BlurSampleOffsets}
  n73 = TextureSampleParameter2D(UVs=Add(A=@n69, B=@n72."0,1")) {
   group=Texture Input
   parameterName=Normal
   samplerSource=SSM_Wrap_WorldGroupSettings
   samplerType=SAMPLERTYPE_Normal
   texture=&FlatNormalBlue
  }
  n74 = TextureSampleParameter2D(UVs=Add(A=@n69, B=@n72."1,0")) {
   group=Texture Input
   parameterName=Normal
   samplerSource=SSM_Wrap_WorldGroupSettings
   samplerType=SAMPLERTYPE_Normal
   texture=&FlatNormalBlue
  }
  n75 = TextureSampleParameter2D(UVs=Add(A=@n69, B=@n72."0,-1")) {
   group=Texture Input
   parameterName=Normal
   samplerSource=SSM_Wrap_WorldGroupSettings
   samplerType=SAMPLERTYPE_Normal
   texture=&FlatNormalBlue
  }
  n76 = TextureSampleParameter2D(UVs=Add(A=@n69, B=@n72."-1,0")) {
   group=Texture Input
   parameterName=Normal
   samplerSource=SSM_Wrap_WorldGroupSettings
   samplerType=SAMPLERTYPE_Normal
   texture=&FlatNormalBlue
  }
  n77 = Add(A=@n68.RGB, B=Add(A=Add(A=@n73.RGB, B=@n74.RGB), B=Add(A=@n75.RGB, B=@n76.RGB)))
  n78 = FeatureLevelSwitch(Default=StaticSwitchParameter(True=Divide(A=@n77) {constB=5}, False=@n68.RGB) {parameterName=UseNormalBlur}, ES2=@n68.RGB, ES3_1=@n68.RGB)
  n79 = SetMaterialAttributes(
   BaseColor=StaticSwitchParameter(
    True=Multiply(A=@n60, B=@n40.RGB)
    False=StaticSwitchParameter(True=@n63, False=@n60) {group=RealisticPaint, parameterName=UseAmbienOcclusionBake}
   ) {group=RealisticPaint, parameterName=UseDiffuseMultiply}
   Metallic=StaticSwitchParameter(True=LinearInterpolate(A=@n64, B=@n65, Alpha=@n59.G), False=@n64) {group=RealisticPaint, parameterName=Use2PaintColors}
   Specular=ScalarParameter() {defaultValue=0.5, group=RealisticPaint, parameterName=Specular}
   Roughness=LinearInterpolate(
    A=ScalarParameter() {defaultValue=0.35, group=RealisticPaint, parameterName=Roughness}
    B=ScalarParameter() {defaultValue=0.35, group=RealisticPaint, parameterName=Roughness2}
    Alpha=@n51
   )
   Normal=StaticSwitchParameter(
    True=MaterialFunctionCall(BaseNormal=@n78, AdditionalNormal=@n67.Result) {materialFunction=&BlendAngleCorrectedNormals}.Result
    False=@n78
   ) {group=RealisticPaint, parameterName=UseOrangePeel}
   ShadingModel=ShadingModel() {shadingModel=MSM_ClearCoat}
   CustomData0=ScalarParameter() {defaultValue=1, group=RealisticPaint, parameterName=CC}
   CustomData1=ScalarParameter() {group=RealisticPaint, parameterName=CCR}
  ) {attributeSetTypes [
    #1
    #2
    #3
    #4
    #5
    #6
    #7
    #8
    #9
   ]}
  n80 = StaticSwitchParameter(
   True=@n45
   False=StaticSwitchParameter(True=BlendMaterialAttributes(A=@n45, B=@n79, Alpha=@n59.R), False=@n79) {group=RealisticPaint, parameterName=UseRealisticPaintMask}
  ) {group=RealisticPaint, parameterName=NoPaint}
  n81 = MaterialFunctionCall!(In=@n80) {materialFunction=&/MF_CarbonFiber}
  n82 = TextureSampleParameter2D() {
   group=Texture Input
   parameterName=MaterialLayerMask
   samplerSource=SSM_Wrap_WorldGroupSettings
   texture=&T_BlacksRGB
  }
  n83 = StaticSwitchParameter(True=FeatureLevelSwitch(Default=BlendMaterialAttributes(A=@n80, B=@n81.Result, Alpha=@n82.R), ES3_1=@n80), False=@n80) {group=RealisticPaint, parameterName=UseCarbonFiber}
  n84 = Multiply(A=TextureCoordinate(), B=ScalarParameter() {defaultValue=102.7072, group=HexGrid, parameterName=Scale})
  n85 = ScalarParameter() {defaultValue=1000, group=HexGrid, parameterName=DistanceFade2}
  n86 = ScalarParameter() {defaultValue=500, group=HexGrid, parameterName=DistanceFadeOffset2}
  n87 = TextureSampleParameter2D!(
   UVs=@n84
   Bias=LinearInterpolate(
    B=ScalarParameter() {defaultValue=-4, group=HexGrid, parameterName=Bias}
    Alpha=MaterialFunctionCall(Fade Length=@n85, Fade Offset=@n86) {materialFunction=&CameraDepthFade}.Result
   )
   Apply View MipBias
  ) {
   group=HexGrid
   mipValueMode=TMVM_MipBias
   parameterName=HexGridMask
   samplerSource=SSM_Wrap_WorldGroupSettings
   texture=&T_BlacksRGB
  }
  n88 = ScalarParameter() {defaultValue=0.5, group=HexGrid, parameterName=Threshold}
  n89 = ScalarParameter() {defaultValue=0.05, group=HexGrid, parameterName=Threshold2}
  n90 = MaterialFunctionCall(
   Fade Length=ScalarParameter() {defaultValue=3000, group=HexGrid, parameterName=DistanceFade}
   Fade Offset=ScalarParameter() {group=HexGrid, parameterName=DistanceFadeOffset}
  ) {materialFunction=&CameraDepthFade}
  n91 = Floor(None=Divide(A=@n87.G, B=LinearInterpolate(A=@n88, B=@n89, Alpha=@n90.Result)))
  n92 = TextureSampleParameter2D!(UVs=@n84, Apply View MipBias) {
   group=HexGrid
   parameterName=HexGridNormal
   samplerSource=SSM_Wrap_WorldGroupSettings
   samplerType=SAMPLERTYPE_Normal
   texture=&FlatNormalBlue
  }
  n93 = Constant()
  n94 = SetMaterialAttributes(
   BaseColor=LinearInterpolate(B=VectorParameter() {group=HexGrid, parameterName=HexGridColor, defaultValue=<rgba 0 0 0 1>}.RGB, Alpha=@n91)
   Metallic=ScalarParameter() {group=HexGrid, parameterName=Metallic_Hexgrid}
   Specular=LinearInterpolate(
    A=ScalarParameter() {group=HexGrid, parameterName=Spec_Hexgrid_2}
    B=ScalarParameter() {defaultValue=0.5, group=HexGrid, parameterName=Spec_Hexgrid_1}
    Alpha=@n91
   )
   Roughness=ScalarParameter() {defaultValue=0.5, group=HexGrid, parameterName=Roughness_Hexgrid}
   Normal=@n92.RGB
   ShadingModel=ShadingModel()
   CustomData0=@n93
   CustomData1=@n93
  ) {attributeSetTypes [
    #1
    #2
    #3
    #4
    #5
    #6
    #7
    #8
    #9
   ]}
  n95 = ScalarParameter() {group=HexGrid, parameterName=Opacity_Hexgrid}
  n96 = BlendMaterialAttributes(A=@n83, B=@n94, Alpha=Multiply(A=LinearInterpolate(A=@n95, Alpha=@n91), B=@n59.B)) {vertexAttributeBlendType=UseA}
  FunctionOutput(None=StaticSwitchParameter(True=StaticSwitchParameter(True=@n96, False=@n83) {group=RealisticPaint, parameterName=UseDetailOverlay}, False=@n39.Result) {parameterName=IsRealisticCarPaint})
  FunctionOutput(None=@n39.Scratch Color) {outputName=Scratch Color}
  FunctionOutput(None=@n6.Lights) {outputName=Lights}
  FunctionOutput(None=@n6.Gmap) {outputName=Gmap}
  FunctionOutput(None=@n6.Emissive) {outputName=Emissive Mask}
  MaterialFunctionCall(Alpha Threshold=@n91) {materialFunction=&DitherTemporalAA}
 ]}
