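# Material function for Valet materials that layers an optional textured decal, with its own surface properties, over an incoming material attributes set.
# The decal texture samples UV0 or UV1 by static switch, supplying base color and the alpha that blends decal attributes over the input.
# Decal roughness, metallic and specular are parameters, while the normal comes from the underlying material.
# Decal Clearcoat feeds CustomData0 and selects Clear Coat shading above 0.5.
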MaterialFunction /Valet/Materials/MF_Valet_Decal
graph {nodes [
  Constant()
  n1 = FunctionInput() {inputName=Material, inputType=FunctionInput_MaterialAttributes}
  n2 = GetMaterialAttributes(None=@n1) {attributeGetTypes [#1, #2, #3]}
  n3 = TextureSampleParameter2D(UVs=StaticSwitchParameter(True=TextureCoordinate() {coordinateIndex=1}, False=TextureCoordinate()) {group=Decal, parameterName=use Second UV Channel for Decal, sortPriority=2}) {
   group=Decal
   parameterName=Decal
   samplerType=SAMPLERTYPE_LinearColor
   texture=/Game/Global/Textures/Default/Blanks/T_Blank_White_ZeroAlpha_LinearColor
  }
  n4 = ScalarParameter() {defaultValue=1, group=Decal, parameterName=Decal Roughness, sliderMax=4, sortPriority=3}
  n5 = ScalarParameter() {group=Decal, parameterName=Decal Clearcoat}
  n6 = SetMaterialAttributes!(
   MaterialAttributes
   BaseColor=@n3.RGB
   Roughness=@n4
   Metallic=ScalarParameter() {defaultValue=1, group=Decal, parameterName=Decal Metallic, sliderMax=4, sortPriority=5}
   Specular=ScalarParameter() {defaultValue=1, group=Decal, parameterName=Decal Specular, sliderMax=4, sortPriority=4}
   ShadingModel=If(A=@n5, B=Constant() {r=0.5}, "A > B"=ShadingModel() {shadingModel=MSM_ClearCoat}, "A < B"=ShadingModel())
   Normal=@n2.Normal
   CustomData0=@n5
   CustomData1=@n2.Normal
  ) {attributeSetTypes [
    #4
    #1
    #5
    #6
    #7
    #2
    #8
    #3
   ]}
  FunctionOutput(None=StaticSwitchParameter(True=BlendMaterialAttributes(A=@n2.MaterialAttributes, B=@n6, Alpha=@n3.A), False=@n1) {group=Decal, parameterName=Use Decal})
  LinearInterpolate(Alpha=@n5)
  Saturate(None=Multiply(A=@n2.Roughness, B=@n4))
 ]}
