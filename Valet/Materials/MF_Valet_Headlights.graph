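# Material function that adds headlight glow to the emissive channel of an incoming material attributes set, for Valet materials.
# An HDR Headlight Color fades to black toward grazing angles through a Fresnel term shaped by HeadlightFalloff and Power parameters.
# The scalar Mask input confines the glow, which is added to existing emissive.
# The headlightsFullyOn parameter blends between original and lit emissive before output.
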MaterialFunction /Valet/Materials/MF_Valet_Headlights
graph {nodes [
  n1 = GetMaterialAttributes(None=FunctionInput() {inputName=MA, inputType=FunctionInput_MaterialAttributes}) {attributeGetTypes [#1]}
  n2 = VectorParameter() {group=Headlights, parameterName=Headlight Color, defaultValue=<rgba 7 6.6166 5.861 1>}
  n3 = ScalarParameter() {defaultValue=0.1, group=Headlights, parameterName=HeadlightFalloff}
  n4 = Saturate(None=Power(Base=Fresnel(ExponentIn=@n3) {baseReflectFraction=0, exponent=0.04}, Exp=ScalarParameter() {defaultValue=2, parameterName=Power}))
  n5 = Add(
   A=@n1.EmissiveColor
   B=Multiply(A=FunctionInput() {inputName=Mask, inputType=FunctionInput_Scalar}, B=LinearInterpolate(A=@n2.RGB, Alpha=@n4) {constB=0})
  )
  n6 = ScalarParameter() {group=Headlights, parameterName=headlightsFullyOn}
  FunctionOutput(None=SetMaterialAttributes!(MaterialAttributes=@n1.MaterialAttributes, EmissiveColor=LinearInterpolate(A=@n1.EmissiveColor, B=@n5, Alpha=@n6)) {attributeSetTypes [#1]})
 ]}
